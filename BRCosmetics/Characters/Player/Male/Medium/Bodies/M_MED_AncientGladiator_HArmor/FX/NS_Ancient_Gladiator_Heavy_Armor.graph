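# Niagara system for the Ancient Gladiator heavy armor cosmetic, adding purple collar fire and pulsing eye glows that follow the character's skeletal mesh sockets or bones.
# Five burst-once sprite emitters place looping glows on chest, forearms, knees and shoulders, mirrored through spawn groups and signed offsets.
# The collar combines spawned decal and fire sprites, the fire blending slightly from socket-following toward spawn position, with one persistent flame mesh passing interpolated position and WPO to a dynamic parameter.
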NiagaraSystem /BRCosmetics/Characters/Player/Male/Medium/Bodies/M_MED_AncientGladiator_HArmor/FX/NS_Ancient_Gladiator_Heavy_Armor
dependencies {
 usedDataInterfaces [&NiagaraDataInterfaceColorCurve, &NiagaraDataInterfaceCurve, &NiagaraDataInterfaceVectorCurve]
 usedDynamicInputs [&MakeVectorFromLinearColorRGB, &VectorFromCurve]
 usedModules [&SystemState]
}
emitters {
 Collar_Decal {
  emitterUpdateScript [
   &EmitterState()
   &SpawnRate(SpawnRate=13, Use Spawn Probability=false)
  ]
  particleSpawnScript [
   &FN_SetSpawnGroup_v2(Number of Spawn Groups=4)
   /SetVariables(
    Particles.LateralSpawnGroup=HlslExpression() {hlslExpression="Particles.SpawnGroup < 2 ? 1 : -1 "}
    Particles.LessThanSpawnGroup=HlslExpression() {hlslExpression="Particles.SpawnGroup < 1 || Particles.SpawnGroup > 2 ? 1 : 0"}
   ) {isSetParameters=true, module=SetVariables_#1}
   /Niagara/Modules/Spawn/Initialization/InitializeParticle(
    Color=&RandomRangeLinearColor(Minimum=<rgba 0.489 0 0.75 1>, Maximum=<rgba 0.5454 0 1 1.65>)
    Module.Write Color=true
    Module.Write Lifetime=true
    Module.Write Mass=true
    Module.Write Material Random=true
    Module.Write Position=false
    Module.Write Scale=false
    Module.Write SpriteRotation=false
    Module.Write SpriteSize=true
    Position=&SimulationPosition()
    Sprite Size=&MakeVector2D(
     X=HlslExpression() {hlslExpression="Particles.SpawnGroup == 0 ? 46 : 30"}
     Y=HlslExpression() {hlslExpression="Particles.SpawnGroup == 0 ? 36 : 30"}
    )
   )
   &CameraOffset(Camera Offset Amount=9)
   /SetVariables(Particles.SpawnOffset:Vector3f=&MakeVector(
     X=&Lerp_Float(
      A=&Multiply_Float(A=$Particles.LateralSpawnGroup, B=12)
      B=&RandomRangeFloat(Minimum=-5, Maximum=2)
      Alpha=&MakeCustomFloatFromBool(Bool=$Particles.LessThanSpawnGroup, True Float=1, False Float=0)
     )
     Y=&Lerp_Float(
      A=&Multiply_Float(A=$Particles.LateralSpawnGroup, B=&RandomRangeFloat(Minimum=-10, Maximum=8))
      B=-14
      Alpha=&MakeCustomFloatFromBool(Bool=$Particles.LessThanSpawnGroup, True Float=1, False Float=0)
     )
     Z=&Lerp_Float(
      A=&Multiply_Float(A=$Particles.LateralSpawnGroup, B=15)
      B=&RandomRangeFloat(Minimum=-10, Maximum=3)
      Alpha=&MakeCustomFloatFromBool(Bool=$Particles.LessThanSpawnGroup, True Float=1, False Float=0)
     )
    )) {isSetParameters=true}
   /SetVariables(Particles.UVScale:Vector2f=&MakeVector2D(X=&UniformAOrBFloat(A=-1), Y=1)) {isSetParameters=true}
  ]
  particleUpdateScript [
   &ParticleState()
   &SkeletalMeshLocation(
    Apply Sampled Orientation to Mesh Particles=true
    Bone / Socket Index=&Add_Integer(A=HlslExpression() {hlslExpression="Particles.SpawnGroup == 3  ? 0 : Particles.SpawnGroup"}, B=8)
    Bone Sampling Mode=::NewEnumerator4
    Inherit Velocity=false
    Interpolate Mesh Sampling=false
    MaskSpawnGroup=false
    Sampled Position Offset=$Particles.SpawnOffset
    Sampled Position Offset Coordinate Space:ENiagaraMeshTransforms=::NewEnumerator3
    Skeletal Mesh=$User.SkeletalMesh
   )
   &DynamicMaterialParameters(
    Index 0 Param 1=1
    Index 0 Param 2=&Lerp_Float(A=0.5, B=1, Alpha=$Particles.NormalizedAge)
    Index 0 Param 3=&RandomRangeFloat(Minimum=0, Maximum=1)
    Index 0 Param 4=&RandomRangeFloat(Minimum=0.75, Maximum=1)
    Module.Param0WriteEnabled=true
    Module.Param0WriteEnabled1=true
    Module.Param1WriteEnabled=true
    Module.Param1WriteEnabled1=true
    Module.Param2WriteEnabled=true
    Module.Param2WriteEnabled1=true
    Module.Param3WriteEnabled=true
    Module.Param3WriteEnabled1=true
    Write Parameter Index 1=true
   )
   &ScaleColor(
    Scale Alpha=&FloatFromCurve(FloatCurve=DataInterface() {propertyValues {Curve {keys [
         <at 0 0 RCIM_Cubic>
         {arriveTangent=1.5906, interpMode=RCIM_Cubic, leaveTangent=1.5906, time=0.25, value=1}
         {arriveTangent=0.1798, interpMode=RCIM_Cubic, leaveTangent=0.1798, time=0.3824, value=0.6082}
         {arriveTangent=-0.9846, interpMode=RCIM_Cubic, leaveTangent=-0.9846, time=0.6232, value=1.0671}
         <at 1 0>
        ]}}})
    ScaleA=true
    ScaleRGB=true
   )
   &ScaleSpriteSize(Scale Factor=&Vector2DFromFloat(Value=&FloatFromCurve(FloatCurve=DataInterface() {propertyValues {Curve {keys [{arriveTangent=2.3225, interpMode=RCIM_Cubic, leaveTangent=2.3225, tangentMode=RCTM_User, time=0, value=0.25}, <at 1 1>]}}})))
  ]
  rendererClasses [&NiagaraSpriteRendererProperties]
  renderers [{rendererClass=&NiagaraSpriteRendererProperties}]
 }
 Collar_Flames {
  emitterUpdateScript [
   &EmitterState()
   &SpawnBurst_Instantaneous(Spawn Count=1, Spawn Probability=HlslExpression() {hlslExpression="Emitter.LoopCount == 0"}, Use Spawn Probability=true)
  ]
  particleSpawnScript [
   &InitializeParticle(
    Color=<rgba 0.9781 0 1.5 1.1>
    Mesh Scale=<1.1 1 1>
    Module.Write Color=true
    Module.Write Lifetime=true
    Module.Write Mass=true
    Module.Write Material Random=true
    Module.Write Position=false
    Module.Write Scale=true
    Module.Write SpriteRotation=false
    Module.Write SpriteSize=false
    Position=&SimulationPosition()
   )
   &DynamicMaterialParameters(
    Index 0 Param 3=10
    Module.Param0WriteEnabled=true
    Module.Param1WriteEnabled=true
    Module.Param2WriteEnabled=true
    Module.Param3WriteEnabled=true
   )
   /Niagara/Modules/Spawn/Orientation/V2/InitialMeshOrientation(Orientation Axis:Vector3f=<0 0 1>, Rotation=<-0.035 0 0.025>, Use Orientation Vector=true, Use Rotation Vector=true)
  ]
  particleUpdateScript [
   &ParticleState(Kill Particles When Lifetime Has Elapsed=false, Let Infinitely Lived Particles Die When Emitter Deactivates=true)
   &SkeletalMeshLocation(
    Apply Sampled Orientation to Mesh Particles=true
    Bone / Socket Index=8
    Bone Sampling Mode=::NewEnumerator4
    Inherit Velocity=false
    MaskSpawnGroup=false
    Sampled Position Offset=<-3.5 5 1>
    Sampled Position Offset Coordinate Space:ENiagaraMeshTransforms=::NewEnumerator3
    Skeletal Mesh=$User.SkeletalMesh
   )
   &InterpolateOverTime(
    Rate Of Change:NiagaraFloat=$User.Interpolate_Rate_of_Change
    Target Vector Value:NiagaraPosition=$Particles.Position
    Variable Type=::NewEnumerator2
   )
   /SetVariables(Particles.Interpolate Dynamic=&MakeFloat4FromVectorAndFloat("Vector (RGB)"=$Output.InterpolateOverTime.Moving Average, "Float (Alpha)"=$User.Interpolate_WPO)) {isSetParameters=true}
  ]
  rendererClasses [&NiagaraMeshRendererProperties]
  renderers [{rendererClass=&NiagaraMeshRendererProperties, rendererIndex=0}]
 }
 EyeGlow_on_Chest {
  emitterSpawnScript [&FN_PreserveSkydiveUserVariables() {module=PreserveUserVariables}]
  emitterUpdateScript [
   &EmitterState()
   &SpawnBurst_Instantaneous(Spawn Count=2, Spawn Probability=HlslExpression() {hlslExpression="Emitter.LoopCount == 0"}, Use Spawn Probability=true)
  ]
  particleSpawnScript [
   &FN_SetSpawnGroup_v2(Number of Spawn Groups=2)
   &InitializeParticle(
    Color=<rgba 0.8688 0 1 0.4>
    Lifetime=&RandomRangeFloat(Minimum=3, Maximum=4)
    Module.Write Color=true
    Module.Write Lifetime=true
    Module.Write Mass=false
    Module.Write Material Random=true
    Module.Write Position=false
    Module.Write Scale=true
    Module.Write SpriteRotation=false
    Module.Write SpriteSize=true
    Sprite Size=&Vector2DFromFloat(Value=14)
   )
   &CameraOffset(Camera Offset Amount=2)
   /Game/Effects/Niagara/Modules/Math/FN_EvenNumberToSign(Input:NiagaraInt32=$Particles.SpawnGroup, Output Float Attribute=true)
   /SetVariables(Particles.PositionOffset:Vector3f=&MakeVector(X=-3, Y=14, Z=&Multiply_Float(A=$Particles.Sign, B=2.5))) {isSetParameters=true, module=SetVariables_#2}
  ]
  particleUpdateScript [
   &ParticleState(Kill Particles When Lifetime Has Elapsed=false, Loop Particles Lifetime=true)
   &SkeletalMeshLocation(
    Apply Sampled Orientation to Mesh Particles=true
    Bone / Socket Index=3
    Bone Sampling Mode=::NewEnumerator4
    Inherit Velocity=false
    Interpolate Mesh Sampling=false
    MaskSpawnGroup=false
    Sampled Position Offset=$Particles.PositionOffset
    Sampled Position Offset Coordinate Space:ENiagaraMeshTransforms=::NewEnumerator3
    Skeletal Mesh=$User.SkeletalMesh
   )
   &ScaleColor(
    Scale Mode=::NewEnumerator1
    Scale RGBA=&MakeVector4_FromLinearColor(LinearColor=&ColorFromCurve(ColorCurve=DataInterface() {propertyValues {
        AlphaCurve {keys [<at 0 0.5>,<at 0.5 1>,<at 1 0.5>]}
        BlueCurve {keys [<at 0 1 RCIM_Cubic>,<at 0.5 2 RCIM_Cubic>,<at 1 1 RCIM_Cubic>]}
        GreenCurve {keys [<at 0 1 RCIM_Cubic>,<at 0.5 2 RCIM_Cubic>,<at 1 1 RCIM_Cubic>]}
        RedCurve {keys [<at 0 1 RCIM_Cubic>,<at 0.5 2 RCIM_Cubic>,<at 1 1 RCIM_Cubic>]}
       }}))
    ScaleRGBA=true
   )
  ]
  rendererClasses [&NiagaraSpriteRendererProperties]
  renderers [{rendererClass=&NiagaraSpriteRendererProperties}]
 }
 EyeGlow_on_Forearms {
  emitterSpawnScript [&FN_PreserveSkydiveUserVariables() {module=PreserveUserVariables}]
  emitterUpdateScript [
   &EmitterState()
   &SpawnBurst_Instantaneous(Spawn Count=2, Spawn Probability=HlslExpression() {hlslExpression="Emitter.LoopCount == 0"}, Use Spawn Probability=true)
  ]
  particleSpawnScript [
   &FN_SetSpawnGroup_v2(Number of Spawn Groups=2)
   &InitializeParticle(
    Color=<rgba 0.8688 0 1 0.4>
    Lifetime=&RandomRangeFloat(Minimum=3, Maximum=4)
    Module.Write Color=true
    Module.Write Lifetime=true
    Module.Write Mass=false
    Module.Write Material Random=true
    Module.Write Position=false
    Module.Write Scale=true
    Module.Write SpriteRotation=false
    Module.Write SpriteSize=true
    Sprite Size=&Vector2DFromFloat(Value=16)
   )
   &CameraOffset(Camera Offset Amount=2)
   /SetVariables(Particles.PositionOffset:Vector3f=&MakeVector(
     X=&MaskFloatBySpawnGroup(Masked Float=-2.5, Passthrough Float=2.5)
     Y=&MaskFloatBySpawnGroup(Masked Float=-0.25, Passthrough Float=0.25)
     Z=&MaskFloatBySpawnGroup(Masked Float=7, Passthrough Float=-7)
    )) {isSetParameters=true, module=SetVariables_#2}
  ]
  particleUpdateScript [
   &ParticleState(Kill Particles When Lifetime Has Elapsed=false, Loop Particles Lifetime=true)
   &SkeletalMeshLocation(
    Apply Sampled Orientation to Mesh Particles=true
    Bone / Socket Index=&MaskIntBySpawnGroup(Masked Int=6, Passthrough Int=7, Spawn Group Mask=0)
    Bone Sampling Mode=::NewEnumerator4
    Inherit Velocity=false
    Interpolate Mesh Sampling=false
    MaskSpawnGroup=false
    Sampled Position Offset=$Particles.PositionOffset
    Sampled Position Offset Coordinate Space:ENiagaraMeshTransforms=::NewEnumerator3
    Skeletal Mesh=$User.SkeletalMesh
   )
   &ScaleColor(
    Scale Mode=::NewEnumerator1
    Scale RGBA=&MakeVector4_FromLinearColor(LinearColor=&ColorFromCurve(ColorCurve=DataInterface() {propertyValues {
        AlphaCurve {keys [<at 0 0.5>,<at 0.5 1>,<at 1 0.5>]}
        BlueCurve {keys [<at 0 1 RCIM_Cubic>,<at 0.5 2 RCIM_Cubic>,<at 1 1 RCIM_Cubic>]}
        GreenCurve {keys [<at 0 1 RCIM_Cubic>,<at 0.5 2 RCIM_Cubic>,<at 1 1 RCIM_Cubic>]}
        RedCurve {keys [<at 0 1 RCIM_Cubic>,<at 0.5 2 RCIM_Cubic>,<at 1 1 RCIM_Cubic>]}
       }}))
    ScaleRGBA=true
   )
  ]
  rendererClasses [&NiagaraSpriteRendererProperties]
  renderers [{rendererClass=&NiagaraSpriteRendererProperties}]
 }
 EyeGlow_on_Knees {
  emitterSpawnScript [&FN_PreserveSkydiveUserVariables() {module=PreserveUserVariables}]
  emitterUpdateScript [
   &EmitterState()
   &SpawnBurst_Instantaneous(Spawn Count=4, Spawn Probability=HlslExpression() {hlslExpression="Emitter.LoopCount == 0"}, Use Spawn Probability=true)
  ]
  particleSpawnScript [
   &InitializeParticle(
    Color=<rgba 0.8688 0 1 0.15>
    Lifetime=&RandomRangeFloat(Minimum=3, Maximum=4)
    Module.Write Color=true
    Module.Write Lifetime=true
    Module.Write Mass=false
    Module.Write Material Random=true
    Module.Write Position=false
    Module.Write Scale=true
    Module.Write SpriteRotation=false
    Module.Write SpriteSize=true
    Sprite Size=&Vector2DFromFloat(Value=14)
   )
   &FN_SetSpawnGroup_v2(Number of Spawn Groups=4)
   &CameraOffset(Camera Offset Amount=4)
   /SetVariables(
    Particles.LateralSpawnGroup=HlslExpression() {hlslExpression="Particles.SpawnGroup < 2 ? 1 : -1 "}
    Particles.LessThanSpawnGroup=HlslExpression() {hlslExpression="Particles.SpawnGroup < 2 ? 1 : 0"}
   ) {isSetParameters=true, module=SetVariables_#1}
   &FN_EvenNumberToSign(Input:NiagaraInt32=$Particles.SpawnGroup, Output Float Attribute=true)
  ]
  particleUpdateScript [
   &ParticleState(Kill Particles When Lifetime Has Elapsed=false, Loop Particles Lifetime=true)
   &SkeletalMeshLocation(
    Apply Sampled Orientation to Mesh Particles=true
    Bone / Socket Index=&MakeCustomIntFromBool(Bool=$Particles.LessThanSpawnGroup, True INT=1, False INT=2)
    Bone Sampling Mode=::NewEnumerator4
    Inherit Velocity=false
    Interpolate Mesh Sampling=false
    MaskSpawnGroup=false
    Sampled Position Offset=&MakeVector(
     X=&Multiply_Float(A=$Particles.LateralSpawnGroup, B=-1)
     Y=&Multiply_Float(A=$Particles.LateralSpawnGroup, B=8)
     Z=&Add_Float(
      A=&Multiply_Float(A=$Particles.LateralSpawnGroup, B=-1.5)
      B=&Multiply_Float(A=$Particles.Sign, B=2.25)
     )
    )
    Sampled Position Offset Coordinate Space:ENiagaraMeshTransforms=::NewEnumerator3
    Skeletal Mesh=$User.SkeletalMesh
   )
   &ScaleColor(
    Scale Mode=::NewEnumerator1
    Scale RGBA=&MakeVector4_FromLinearColor(LinearColor=&ColorFromCurve(ColorCurve=DataInterface() {propertyValues {
        AlphaCurve {keys [<at 0 0.5>,<at 0.5 1>,<at 1 0.5>]}
        BlueCurve {keys [<at 0 1 RCIM_Cubic>,<at 0.5 2 RCIM_Cubic>,<at 1 1 RCIM_Cubic>]}
        GreenCurve {keys [<at 0 1 RCIM_Cubic>,<at 0.5 2 RCIM_Cubic>,<at 1 1 RCIM_Cubic>]}
        RedCurve {keys [<at 0 1 RCIM_Cubic>,<at 0.5 2 RCIM_Cubic>,<at 1 1 RCIM_Cubic>]}
       }}))
    ScaleRGBA=true
   )
  ]
  rendererClasses [&NiagaraSpriteRendererProperties]
  renderers [{rendererClass=&NiagaraSpriteRendererProperties}]
 }
 EyeGlow_on_Shoulders_Back {
  emitterSpawnScript [&FN_PreserveSkydiveUserVariables() {module=PreserveUserVariables}]
  emitterUpdateScript [
   &EmitterState()
   &SpawnBurst_Instantaneous(Spawn Count=4, Spawn Probability=HlslExpression() {hlslExpression="Emitter.LoopCount == 0"}, Use Spawn Probability=true)
  ]
  particleSpawnScript [
   &InitializeParticle(
    Color=<rgba 0.8688 0 1 0.3>
    Lifetime=&RandomRangeFloat(Minimum=3, Maximum=4)
    Module.Write Color=true
    Module.Write Lifetime=true
    Module.Write Mass=false
    Module.Write Material Random=true
    Module.Write Position=false
    Module.Write Scale=true
    Module.Write SpriteRotation=false
    Module.Write SpriteSize=true
    Sprite Size=&Vector2DFromFloat(Value=12)
   )
   &FN_SetSpawnGroup_v2(Number of Spawn Groups=4)
   &CameraOffset(Camera Offset Amount=4)
   /SetVariables(
    Particles.LateralSpawnGroup=HlslExpression() {hlslExpression="Particles.SpawnGroup < 2 ? 1 : -1 "}
    Particles.LessThanSpawnGroup=HlslExpression() {hlslExpression="Particles.SpawnGroup < 2 ? 1 : 0"}
   ) {isSetParameters=true, module=SetVariables_#1}
   &FN_EvenNumberToSign(Input:NiagaraInt32=$Particles.SpawnGroup, Output Float Attribute=true)
  ]
  particleUpdateScript [
   &ParticleState(Kill Particles When Lifetime Has Elapsed=false, Loop Particles Lifetime=true)
   &SkeletalMeshLocation(
    Apply Sampled Orientation to Mesh Particles=true
    Bone / Socket Index=&MakeCustomIntFromBool(Bool=$Particles.LessThanSpawnGroup, True INT=4, False INT=5)
    Bone Sampling Mode=::NewEnumerator4
    Inherit Velocity=false
    Interpolate Mesh Sampling=false
    MaskSpawnGroup=false
    Sampled Position Offset=&MakeVector(
     X=&Multiply_Float(A=$Particles.LateralSpawnGroup, B=2)
     Y=&Multiply_Float(A=$Particles.LateralSpawnGroup, B=-21)
     Z=&Multiply_Float(A=$Particles.Sign, B=2)
    )
    Sampled Position Offset Coordinate Space:ENiagaraMeshTransforms=::NewEnumerator3
    Skeletal Mesh=$User.SkeletalMesh
   )
   &ScaleColor(
    Scale Mode=::NewEnumerator1
    Scale RGBA=&MakeVector4_FromLinearColor(LinearColor=&ColorFromCurve(ColorCurve=DataInterface() {propertyValues {
        AlphaCurve {keys [<at 0 0.5>,<at 0.5 1>,<at 1 0.5>]}
        BlueCurve {keys [<at 0 1 RCIM_Cubic>,<at 0.5 2 RCIM_Cubic>,<at 1 1 RCIM_Cubic>]}
        GreenCurve {keys [<at 0 1 RCIM_Cubic>,<at 0.5 2 RCIM_Cubic>,<at 1 1 RCIM_Cubic>]}
        RedCurve {keys [<at 0 1 RCIM_Cubic>,<at 0.5 2 RCIM_Cubic>,<at 1 1 RCIM_Cubic>]}
       }}))
    ScaleRGBA=true
   )
  ]
  rendererClasses [&NiagaraSpriteRendererProperties]
  renderers [{rendererClass=&NiagaraSpriteRendererProperties}]
 }
 EyeGlow_on_Shoulders_Front {
  emitterSpawnScript [&FN_PreserveSkydiveUserVariables() {module=PreserveUserVariables}]
  emitterUpdateScript [
   &EmitterState()
   &SpawnBurst_Instantaneous(Spawn Count=4, Spawn Probability=HlslExpression() {hlslExpression="Emitter.LoopCount == 0"}, Use Spawn Probability=true)
  ]
  particleSpawnScript [
   &InitializeParticle(
    Color=<rgba 0.8688 0 1 0.3>
    Lifetime=&RandomRangeFloat(Minimum=3, Maximum=4)
    Module.Write Color=true
    Module.Write Lifetime=true
    Module.Write Mass=false
    Module.Write Material Random=true
    Module.Write Position=false
    Module.Write Scale=true
    Module.Write SpriteRotation=false
    Module.Write SpriteSize=true
    Sprite Size=&Vector2DFromFloat(Value=12)
   )
   &FN_SetSpawnGroup_v2(Number of Spawn Groups=4)
   &CameraOffset(Camera Offset Amount=2)
   /SetVariables(
    Particles.LateralSpawnGroup=HlslExpression() {hlslExpression="Particles.SpawnGroup < 2 ? 1 : -1 "}
    Particles.LessThanSpawnGroup=HlslExpression() {hlslExpression="Particles.SpawnGroup < 2 ? 1 : 0"}
   ) {isSetParameters=true, module=SetVariables_#1}
   &FN_EvenNumberToSign(Input:NiagaraInt32=$Particles.SpawnGroup, Output Float Attribute=true)
  ]
  particleUpdateScript [
   &ParticleState(Kill Particles When Lifetime Has Elapsed=false, Loop Particles Lifetime=true)
   &SkeletalMeshLocation(
    Apply Sampled Orientation to Mesh Particles=true
    Bone / Socket Index=&MakeCustomIntFromBool(Bool=$Particles.LessThanSpawnGroup, True INT=4, False INT=5)
    Bone Sampling Mode=::NewEnumerator4
    Inherit Velocity=false
    Interpolate Mesh Sampling=false
    MaskSpawnGroup=false
    Sampled Position Offset=&MakeVector(
     X=&Multiply_Float(A=$Particles.LateralSpawnGroup, B=-8)
     Y=&Multiply_Float(A=$Particles.LateralSpawnGroup, B=3)
     Z=&Multiply_Float(A=$Particles.Sign, B=2)
    )
    Sampled Position Offset Coordinate Space:ENiagaraMeshTransforms=::NewEnumerator3
    Skeletal Mesh=$User.SkeletalMesh
   )
   &ScaleColor(
    Scale Mode=::NewEnumerator1
    Scale RGBA=&MakeVector4_FromLinearColor(LinearColor=&ColorFromCurve(ColorCurve=DataInterface() {propertyValues {
        AlphaCurve {keys [<at 0 0.5>,<at 0.5 1>,<at 1 0.5>]}
        BlueCurve {keys [<at 0 1 RCIM_Cubic>,<at 0.5 2 RCIM_Cubic>,<at 1 1 RCIM_Cubic>]}
        GreenCurve {keys [<at 0 1 RCIM_Cubic>,<at 0.5 2 RCIM_Cubic>,<at 1 1 RCIM_Cubic>]}
        RedCurve {keys [<at 0 1 RCIM_Cubic>,<at 0.5 2 RCIM_Cubic>,<at 1 1 RCIM_Cubic>]}
       }}))
    ScaleRGBA=true
   )
  ]
  rendererClasses [&NiagaraSpriteRendererProperties]
  renderers [{rendererClass=&NiagaraSpriteRendererProperties}]
 }
 Fire_Collar {
  emitterUpdateScript [
   &EmitterState()
   &SpawnRate(SpawnRate=20, Use Spawn Probability=false)
  ]
  particleSpawnScript [
   &InitializeParticle(
    Color=&RandomRangeLinearColor(Minimum=<rgba 0.325 0 0.65 0.5>, Maximum=<rgba 1.4333 0 2 1>)
    Lifetime=&RandomRangeFloat(Minimum=0.75, Maximum=1)
    Module.Write Color=true
    Module.Write Lifetime=true
    Module.Write Mass=false
    Module.Write Material Random=true
    Module.Write Position=false
    Module.Write Scale=false
    Module.Write SpriteRotation=true
    Module.Write SpriteSize=true
    Position=&SimulationPosition()
    Sprite Rotation=&NormalizedAngleToDegrees(NormalizedAngle=&RandomRangeFloat(Minimum=-1, Maximum=1))
    Sprite Size=&Vector2DFromFloat(Value=&RandomRangeFloat(Minimum=30, Maximum=32))
   )
   &SkeletalMeshLocation(
    Bone / Socket Index=11
    Bone Sampling Mode=::NewEnumerator4
    Inherit Velocity=false
    Interpolate Mesh Sampling=false
    MaskSpawnGroup=false
    Sampled Position Offset Coordinate Space:ENiagaraMeshTransforms=::NewEnumerator3
    Skeletal Mesh=$User.SkeletalMesh
   ) {module=SkeletalMeshLocation001}
   /SetVariables(Particles.SpawnVector:Vector3f=&MakeVector(
     X=&RandomRangeFloat(
      Minimum=&FloatFromCurve(
       FloatCurve=DataInterface() {propertyValues {Curve {keys [
           {arriveTangent=22.9185, interpMode=RCIM_Cubic, leaveTangent=21.9686, tangentMode=RCTM_User, time=0, value=-1}
           {arriveTangent=-4.178, interpMode=RCIM_Cubic, leaveTangent=-4.1779, tangentMode=RCTM_User, time=1, value=-2}
          ]}}}
       CurveIndex=$Particles.MaterialRandom
      )
      Maximum=&FloatFromCurve(
       FloatCurve=DataInterface() {propertyValues {Curve {keys [
           {arriveTangent=11.2213, interpMode=RCIM_Cubic, leaveTangent=8.9102, tangentMode=RCTM_User, time=0, value=-5}
           {arriveTangent=-4.178, interpMode=RCIM_Cubic, leaveTangent=-4.1779, tangentMode=RCTM_User, time=1, value=-2}
          ]}}}
       CurveIndex=$Particles.MaterialRandom
      )
     )
     Y=&FloatFromCurve(FloatCurve=DataInterface() {propertyValues {Curve {keys [<at 0 -14>,<at 1 8>]}}}, CurveIndex=$Particles.MaterialRandom)
     Z=&UniformAOrBFloat(
      A=&FloatFromCurve(
       FloatCurve=DataInterface() {propertyValues {Curve {keys [
           {arriveTangent=60.1659, interpMode=RCIM_Cubic, leaveTangent=60.1658, tangentMode=RCTM_User, time=0, value=2}
           {interpMode=RCIM_Cubic, leaveTangent=34.0317, time=1, value=14}
          ]}}}
       CurveIndex=$Particles.MaterialRandom
      )
      B=&FloatFromCurve(
       FloatCurve=DataInterface() {propertyValues {Curve {keys [
           {arriveTangent=-74.1897, interpMode=RCIM_Cubic, leaveTangent=-74.1898, tangentMode=RCTM_User, time=0, value=-2}
           {arriveTangent=-13.1006, interpMode=RCIM_Cubic, leaveTangent=-13.1005, tangentMode=RCTM_User, time=1, value=-14}
          ]}}}
       CurveIndex=$Particles.MaterialRandom
      )
     )
    )) {isSetParameters=true}
   /SetVariables(Particles.SpawnPosition:Vector3f=&RotateVectorByQuaternion(
     Vector To Rotate=&Add_Vector(A=$Particles.SpawnVector, B=<14 0 0>)
     Quaternion=$Output.SkeletalMeshLocation001.SampledRotation
    )) {isSetParameters=true}
   &CameraOffset(Camera Offset Amount=&RandomRangeFloat(Minimum=-2, Maximum=2))
  ]
  particleUpdateScript [
   &ParticleState()
   &SkeletalMeshLocation(
    Apply To Particle Position=false
    Bone / Socket Index=11
    Bone Sampling Mode=::NewEnumerator4
    Inherit Velocity=false
    MaskSpawnGroup=false
    Sampled Position Offset=<14 0 0>
    Sampled Position Offset Coordinate Space:ENiagaraMeshTransforms=::NewEnumerator3
    Skeletal Mesh=$User.SkeletalMesh
   )
   &ScaleColor(
    Scale Mode=::NewEnumerator1
    Scale RGBA=&Vector4FromCurve(Vector4Curve=DataInterface() {propertyValues {
       AlphaCurve {keys [
         <at 0 0 RCIM_Cubic>
         {arriveTangent=3.3333, interpMode=RCIM_Cubic, leaveTangent=3.3333, time=0.3, value=1}
         {arriveTangent=-1.6667, interpMode=RCIM_Cubic, leaveTangent=-1.6667, time=0.6, value=2}
         {arriveTangent=-0.4881, interpMode=RCIM_Cubic, leaveTangent=-0.4881, tangentMode=RCTM_User, time=0.9, value=0}
        ]}
       BlueCurve {keys [<at 0.206 1>,<at 0.9 2>,<at 1.2198 0.4062>]}
       GreenCurve {keys [<at 0.206 1>,<at 0.9 2>,<at 1.2198 0.4062>]}
       RedCurve {keys [<at 0.206 1>,<at 0.9 2>,<at 1.2198 0.4062>]}
      }})
    ScaleRGBA=true
   )
   &ScaleSpriteSize(Scale Factor=&Vector2DFromFloat(Value=&FloatFromCurve(FloatCurve=DataInterface() {propertyValues {Curve {keys [
          {arriveTangent=-0.0804, interpMode=RCIM_Cubic, leaveTangent=-0.0804, tangentMode=RCTM_User, time=0, value=0.75}
          {interpMode=RCIM_Cubic, tangentMode=RCTM_User, time=0.4, value=1}
          {arriveTangent=-4.5991, interpMode=RCIM_Cubic, leaveTangent=-12.9762, tangentMode=RCTM_User, time=1, value=0}
         ]}}})))
   &DynamicMaterialParameters(
    Index 0 Param 1=&FloatFromCurve(FloatCurve=DataInterface() {propertyValues {Curve {keys [{arriveTangent=2.7551, interpMode=RCIM_Cubic, leaveTangent=2.7551, tangentMode=RCTM_User, time=0, value=0}, <at 0.6 1>]}}})
    Index 0 Param 2=&RandomRangeFloat(Minimum=0.33, Maximum=1)
    Index 0 Param 3=&FloatFromCurve(FloatCurve=DataInterface() {propertyValues {Curve {keys [
         {arriveTangent=6.3026, interpMode=RCIM_Cubic, time=0, value=1}
         {arriveTangent=2.4291, interpMode=RCIM_Cubic, leaveTangent=6.3584, tangentMode=RCTM_User, time=1, value=2}
        ]}}})
    Index 0 Param 4=&RandomRangeFloat(Minimum=1, Maximum=2)
    Module.Param0WriteEnabled=true
    Module.Param1WriteEnabled=true
    Module.Param2WriteEnabled=true
    Module.Param3WriteEnabled=true
   ) {module=DynamicMaterialParameters001}
   &SpriteRotationRate(Rotation Rate=20)
   &SubUVAnimation()
   &self:Location_LERP_-_Source_to_Destination(
    DestinationPosition:Vector3f=&Add_Vector(A=Linked() {linkedVariable {name=Particles.Initial.Position, type=NiagaraPosition}}, B=$Particles.SpawnPosition)
    DestinationPosition Offset:Vector3f=&RotateVectorByQuaternion(
     Vector To Rotate=&Multiply_VectorByFloat(Vector=&RandomRangeVector(Minimum=<5 0 0>, Maximum=<10 0 0>), Float=$Particles.NormalizedAge)
     Quaternion=$Output.SkeletalMeshLocation.SampledRotation
    )
    Normalized Influence:NiagaraFloat=&Lerp_Float(A=0, B=&RandomRangeFloat(Minimum=0, Maximum=0.125), Alpha=$Particles.NormalizedAge)
    SourcePosition:Vector3f=&Add_Vector(A=$Output.SkeletalMeshLocation.SampledPosition, B=$Particles.SpawnPosition)
    SourcePositionOffset:Vector3f=&RotateVectorByQuaternion(Vector To Rotate=<1 -2 0>, Quaternion=$Output.SkeletalMeshLocation.SampledRotation)
   ) {module=Location_LERP_-_Source_to_Destination}
  ]
  rendererClasses [&NiagaraSpriteRendererProperties]
  renderers [{rendererClass=&NiagaraSpriteRendererProperties}]
 }
}
summary {
 systemName=NS_Ancient_Gladiator_Heavy_Armor
 emitters [
  {emitterName=EyeGlow_on_Knees, rendererClasses [&NiagaraSpriteRendererProperties]}
  {emitterName=EyeGlow_on_Chest, rendererClasses [&NiagaraSpriteRendererProperties]}
  {emitterName=EyeGlow_on_Shoulders_Front, rendererClasses [&NiagaraSpriteRendererProperties]}
  {emitterName=EyeGlow_on_Forearms, rendererClasses [&NiagaraSpriteRendererProperties]}
  {emitterName=EyeGlow_on_Shoulders_Back, rendererClasses [&NiagaraSpriteRendererProperties]}
  {emitterName=Collar_Decal, rendererClasses [&NiagaraSpriteRendererProperties]}
  {emitterName=Fire_Collar, rendererClasses [&NiagaraSpriteRendererProperties]}
  {emitterName=Collar_Flames, rendererClasses [&NiagaraMeshRendererProperties]}
 ]
 userVariables [
  User.bIsFrontEndPreview {type=NiagaraBool}
  User.Interpolate_Rate_of_Change {type=NiagaraFloat, defaultValue=7, description=""}
  User.Interpolate_WPO {type=NiagaraFloat, defaultValue=1, description=""}
  User.PlayerSpeed {type=NiagaraFloat, defaultValue=0, description=""}
  User.SkeletalMesh {
   type=NiagaraDataInterfaceSkeletalMesh
   defaultValue {
    struct=NiagaraExt_VariableValue_DataInterface
    value {dataInterface=&self:NiagaraDataInterfaceSkeletalMesh_1}
   }
  }
 ]
}
