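# Control Rig for the male medium Ancient Gladiator armor body, posing its loincloth, flap, shoulder pad and fringe bones from body motion.
# A knee-aim bone between the calves steers the loincloth; weighted parent constraints blend armor bones across spine, arm and thigh bones.
# Remapped thigh and upper-arm angles, plus crouching and skydivingINV variables, weight additive corrective offsets.
# Parallel TransformConstraint chains duplicate many constraints but are disconnected from execution.
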ControlRigBlueprint /BRCosmetics/Characters/Player/Male/Medium/Bodies/M_MED_AncientGladiator_HArmor/Meshes/M_MED_AncientGladiator_H_CtrlRig
graphs [
 RigVMFunctionLibrary {links [], nodes []}
 RigVMModel {nodes [
   RigUnit_BeginExecution = &RigUnit_BeginExecution()
   /Script/ControlRig.RigUnit_PositionConstraintLocalSpaceOffset(
    Child=(Type=Bone,Name="dyn_kneeAim_l")
    Filter=(bX=True,bY=True,bZ=True)
    Parents=((Item=(Type=Bone,Name="calf_l"),Weight=1),(Item=(Type=Bone,Name="calf_r"),Weight=1))
    Weight=1
    bMaintainOffset=True
    ExecutePin=@RigUnit_BeginExecution.ExecutePin
   ) {is=struct RigVMUnitNode}
   > /Script/ControlRig.RigUnit_DrawContainerSetTransform(Transform=&RigUnit_GetBoneTransform(Bone=dyn_kneeAim_l).Transform) {is=struct RigVMUnitNode}
   RigVMStructNode_0 = &RigUnit_GetBoneTransform(Bone=calf_r)
   > /Script/ControlRig.RigUnit_AimBone(
    Bone=dyn_kneeAim_l
    DebugSettings=(bEnabled=true,Scale=10,WorldOffset=(Rotation=(X=0,Y=0,Z=0,W=1),Translation=(X=0,Y=0,Z=0),Scale3D=(X=1,Y=1,Z=1)))
    Primary=(Weight=1,Axis=(X=0,Y=-0.1,Z=1),Target=(X=1,Y=0,Z=0),Kind=Location)
    Secondary=(Weight=1,Axis=(X=1,Y=-0.11,Z=-0.085),Target=(X=1,Y=0,Z=0),Kind=Location)
    Weight=1
    bPropagateToChildren=True
    Secondary.Target=&RigUnit_GetBoneTransform(Bone=pelvis).Transform.Translation
    Primary.Target=@RigVMStructNode_0.Transform.Translation
   ) {is=struct RigVMUnitNode}
   > &RigUnit_ParentConstraint(
    AdvancedSettings=(InterpolationType=Average,RotationOrderForFilter=XZY)
    Child=(Type=Bone,Name="dyn_loinClothBack_01_Base")
    Parents=((Item=(Type=Bone,Name="dyn_kneeAim_l"),Weight=0.2),(Item=(Type=Bone,Name="dyn_loinClothBack_01_Base"),Weight=0.8))
    Weight=1
   )
   > &RigUnit_ParentConstraint(
    AdvancedSettings=(InterpolationType=Average,RotationOrderForFilter=XZY)
    Child=(Type=Bone,Name="dyn_loinClothBack_01_Base")
    Parents=((Item=(Type=Bone,Name="dyn_loinClothBack_01_Base"),Weight=1))
    Weight=1
   )
   n1 = RigVMVariableNode() {variable=crouching}
   n2 = &RigVMFunction_MathFloatRemap(SourceMaximum=1, SourceMinimum=0, TargetMaximum=0.2, TargetMinimum=0.5, bClamp=true, Value=@n1.Value)
   n3 = &RigVMFunction_MathFloatRemap(SourceMaximum=1, SourceMinimum=0, TargetMaximum=0.8, TargetMinimum=0.5, bClamp=true, Value=@n1.Value)
   > &RigUnit_ParentConstraint(
    AdvancedSettings=(InterpolationType=Average,RotationOrderForFilter=XZY)
    Child=(Type=Bone,Name="dyn_loinClothFront_01_c_Base")
    Parents=((Item=(Type=Bone,Name="dyn_kneeAim_l"),Weight=1),(Item=(Type=Bone,Name="dyn_loinClothFront_01_c_Base"),Weight=1))
    Weight=1
    Parents.0.Weight=@n2.Result
    Parents.1.Weight=@n3.Result
   )
   > &RigUnit_ParentConstraint(
    AdvancedSettings=(InterpolationType=Average,RotationOrderForFilter=XZY)
    Child=(Type=Bone,Name="dyn_loinClothFront_01_Base_l")
    Parents=((Item=(Type=Bone,Name="thigh_twist_01_l"),Weight=0.5),(Item=(Type=Bone,Name="dyn_loinClothFront_01_c_Base"),Weight=0.5))
    Weight=1
   )
   > &RigUnit_ParentConstraint(
    AdvancedSettings=(InterpolationType=Average,RotationOrderForFilter=XZY)
    Child=(Type=Bone,Name="dyn_loinClothFront_01_Base_r")
    Parents=((Item=(Type=Bone,Name="thigh_twist_01_r"),Weight=0.5),(Item=(Type=Bone,Name="dyn_loinClothFront_01_c_Base"),Weight=0.5))
    Weight=1
   )
   > /Script/ControlRig.RigUnit_ModifyBoneTransforms(
    BoneToModify=((Bone="dyn_loinClothFront_01_c_Base",Transform=(Rotation=(X=0,Y=0,Z=0,W=1),Translation=(X=-5,Y=-8,Z=0),Scale3D=(X=1,Y=1,Z=1))))
    Mode=AdditiveLocal
    WeightMaximum=1
    Weight=@n1.Value
   ) {is=struct RigVMUnitNode}
   n4 = &RigVMFunction_MathFloatRemap(
    SourceMaximum=-105
    SourceMinimum=-20
    bClamp=true
    Value=&RigVMFunction_MathQuaternionToEuler(RotationOrder=ZYX, Value=&RigUnit_GetRelativeBoneTransform(Bone=thigh_l, Space=pelvis).Transform.Rotation).Result.Z
   )
   > &RigUnit_ModifyBoneTransforms(
    BoneToModify=((Bone="dyn_loinClothFront_01_Base_l",Transform=(Rotation=(X=0,Y=0,Z=0,W=1),Translation=(X=0,Y=-7,Z=0),Scale3D=(X=1,Y=1,Z=1))))
    Mode=AdditiveLocal
    WeightMaximum=1
    Weight=@n4.Result
    BoneToModify.0.Transform.Rotation=&RigVMFunction_MathQuaternionFromEuler(Euler=(X=-60,Y=0,Z=0)).Result
   )
   > &RigUnit_ModifyBoneTransforms(
    BoneToModify=((Bone="dyn_cutFlap_Base_l",Transform=(Rotation=(X=0,Y=0,Z=0,W=1),Translation=(X=2,Y=0,Z=3),Scale3D=(X=1,Y=1,Z=1))))
    Mode=AdditiveLocal
    WeightMaximum=1
    Weight=@n4.Result
   )
   n5 = &RigVMFunction_MathQuaternionFromEuler(Euler=(X=-30,Y=0,Z=0))
   > &RigUnit_ModifyBoneTransforms(
    BoneToModify=((Bone="dyn_loinClothFront_01_Base_l",Transform=(Rotation=(X=0,Y=0,Z=0,W=1),Translation=(X=0,Y=-5,Z=-1),Scale3D=(X=1,Y=1,Z=1))))
    Mode=AdditiveLocal
    WeightMaximum=1
    WeightMinimum=0
    Weight=@n4.Result
    BoneToModify.0.Transform.Rotation=@n5.Result
   )
   n6 = &RigUnit_GetRelativeBoneTransform(Bone=thigh_r, Space=pelvis)
   n7 = &RigVMFunction_MathFloatRemap(
    SourceMaximum=72
    SourceMinimum=150
    TargetMinimum=0
    bClamp=true
    Value=/Script/RigVM.RigVMFunction_MathFloatAbs(Value=&RigVMFunction_MathQuaternionToEuler(RotationOrder=ZYX, Value=@n6.Transform.Rotation).Result.Z) {is=struct RigVMUnitNode, template="Absolute::Execute(in Value,out Result)"}.Result
   )
   > &RigUnit_ModifyBoneTransforms(
    BoneToModify=((Bone="dyn_cutFlap_r",Transform=(Rotation=(X=0,Y=0,Z=0,W=1),Translation=(X=-2,Y=2,Z=-7),Scale3D=(X=1,Y=1,Z=1))))
    Mode=AdditiveLocal
    WeightMaximum=1
    WeightMinimum=0
    Weight=@n7.Result
   )
   > &RigUnit_ModifyBoneTransforms(
    BoneToModify=((Bone="dyn_loinClothFront_01_Base_r",Transform=(Rotation=(X=0,Y=0,Z=0,W=1),Translation=(X=-3,Y=-10,Z=-2),Scale3D=(X=1,Y=1,Z=1))))
    Mode=AdditiveLocal
    WeightMaximum=1
    WeightMinimum=0
    Weight=@n7.Result
    BoneToModify.0.Transform.Rotation=&RigVMFunction_MathQuaternionFromEuler(Euler=(X=0,Y=15,Z=0)).Result
   )
   > &RigUnit_ParentConstraint(
    AdvancedSettings=(InterpolationType=Average,RotationOrderForFilter=XZY)
    Child=(Type=Bone,Name="dyn_mainShoulderpadArmBase_r")
    Parents=((Item=(Type=Bone,Name="spine_03"),Weight=0.1),(Item=(Type=Bone,Name="spine_04"),Weight=0.05),(Item=(Type=Bone,Name="spine_05"),Weight=0.1),(Item=(Type=Bone,Name="clavicle_r"),Weight=0.5),(Item=(Type=Bone,Name="upperarm_twist_01_r"),Weight=0.25))
    Weight=1
   )
   > &RigUnit_ParentConstraint(
    AdvancedSettings=(InterpolationType=Average,RotationOrderForFilter=XZY)
    Child=(Type=Bone,Name="dyn_mainShoulderpadArmBase_l")
    Parents=((Item=(Type=Bone,Name="spine_03"),Weight=0.1),(Item=(Type=Bone,Name="spine_04"),Weight=0.05),(Item=(Type=Bone,Name="spine_05"),Weight=0.1),(Item=(Type=Bone,Name="clavicle_l"),Weight=0.5),(Item=(Type=Bone,Name="upperarm_twist_01_l"),Weight=0.25))
    Weight=1
   )
   > &RigUnit_ParentConstraint(
    AdvancedSettings=(InterpolationType=Average,RotationOrderForFilter=XZY)
    Child=(Type=Bone,Name="dyn_fringeA_01_Base_l")
    Parents="((Item=(Type=Bone,Name=\"spine_03\"),Weight=0.0822),(Item=(Type=Bone,Name=\"spine_04\"),Weight=0.0354),(Item=(Type=Bone,Name=\"spine_05\"),Weight=0.0543),(Item=(Type=Bone,Name=\"clavicle_l\"),Weight=0.418),(Item=(Type=Bone,Name=\"upperarm_twist_01_l\"),Weight=0.3778 ...(truncated)"
    Weight=1
   )
   > &RigUnit_ParentConstraint(
    AdvancedSettings=(InterpolationType=Average,RotationOrderForFilter=XZY)
    Child=(Type=Bone,Name="dyn_fringeH_01_Base_l")
    Parents="((Item=(Type=Bone,Name=\"spine_03\"),Weight=0.0822),(Item=(Type=Bone,Name=\"spine_04\"),Weight=0.0354),(Item=(Type=Bone,Name=\"spine_05\"),Weight=0.0543),(Item=(Type=Bone,Name=\"clavicle_l\"),Weight=0.418),(Item=(Type=Bone,Name=\"upperarm_twist_01_l\"),Weight=0.3778 ...(truncated)"
    Weight=1
   )
   > &RigUnit_ParentConstraint(
    AdvancedSettings=(InterpolationType=Average,RotationOrderForFilter=XZY)
    Child=(Type=Bone,Name="dyn_fringeB_01_Base_l")
    Parents="((Item=(Type=Bone,Name=\"spine_03\"),Weight=0.0575),(Item=(Type=Bone,Name=\"spine_04\"),Weight=0.0424),(Item=(Type=Bone,Name=\"spine_05\"),Weight=0.0314),(Item=(Type=Bone,Name=\"clavicle_l\"),Weight=0.2066),(Item=(Type=Bone,Name=\"upperarm_twist_01_l\"),Weight=0.414 ...(truncated)"
    Weight=1
   )
   > &RigUnit_ParentConstraint(
    AdvancedSettings=(InterpolationType=Average,RotationOrderForFilter=XZY)
    Child=(Type=Bone,Name="dyn_fringeC_01_Base_l")
    Parents="((Item=(Type=Bone,Name=\"spine_03\"),Weight=0.0575),(Item=(Type=Bone,Name=\"spine_04\"),Weight=0.0424),(Item=(Type=Bone,Name=\"spine_05\"),Weight=0.0314),(Item=(Type=Bone,Name=\"clavicle_l\"),Weight=0.2066),(Item=(Type=Bone,Name=\"upperarm_twist_01_l\"),Weight=0.414 ...(truncated)"
    Weight=1
   )
   > &RigUnit_ParentConstraint(
    AdvancedSettings=(InterpolationType=Average,RotationOrderForFilter=XZY)
    Child=(Type=Bone,Name="dyn_fringeD_01_Base_l")
    Parents="((Item=(Type=Bone,Name=\"spine_03\"),Weight=0.0575),(Item=(Type=Bone,Name=\"spine_04\"),Weight=0.0424),(Item=(Type=Bone,Name=\"spine_05\"),Weight=0.0314),(Item=(Type=Bone,Name=\"clavicle_l\"),Weight=0.2066),(Item=(Type=Bone,Name=\"upperarm_twist_01_l\"),Weight=0.414 ...(truncated)"
    Weight=1
   )
   > &RigUnit_ParentConstraint(
    AdvancedSettings=(InterpolationType=Average,RotationOrderForFilter=XZY)
    Child=(Type=Bone,Name="dyn_fringeE_01_Base_l")
    Parents="((Item=(Type=Bone,Name=\"spine_03\"),Weight=0.0575),(Item=(Type=Bone,Name=\"spine_04\"),Weight=0.0424),(Item=(Type=Bone,Name=\"spine_05\"),Weight=0.0314),(Item=(Type=Bone,Name=\"clavicle_l\"),Weight=0.2066),(Item=(Type=Bone,Name=\"upperarm_twist_01_l\"),Weight=0.414 ...(truncated)"
    Weight=1
   )
   > &RigUnit_ParentConstraint(
    AdvancedSettings=(InterpolationType=Average,RotationOrderForFilter=XZY)
    Child=(Type=Bone,Name="dyn_fringeF_01_Base_l")
    Parents="((Item=(Type=Bone,Name=\"spine_03\"),Weight=0.0575),(Item=(Type=Bone,Name=\"spine_04\"),Weight=0.0424),(Item=(Type=Bone,Name=\"spine_05\"),Weight=0.0314),(Item=(Type=Bone,Name=\"clavicle_l\"),Weight=0.2066),(Item=(Type=Bone,Name=\"upperarm_twist_01_l\"),Weight=0.414 ...(truncated)"
    Weight=1
   )
   > &RigUnit_ParentConstraint(
    AdvancedSettings=(InterpolationType=Average,RotationOrderForFilter=XZY)
    Child=(Type=Bone,Name="dyn_fringeG_01_Base_l")
    Parents="((Item=(Type=Bone,Name=\"spine_03\"),Weight=0.0575),(Item=(Type=Bone,Name=\"spine_04\"),Weight=0.0424),(Item=(Type=Bone,Name=\"spine_05\"),Weight=0.0314),(Item=(Type=Bone,Name=\"clavicle_l\"),Weight=0.2066),(Item=(Type=Bone,Name=\"upperarm_twist_01_l\"),Weight=0.414 ...(truncated)"
    Weight=1
   )
   > &RigUnit_ParentConstraint(
    AdvancedSettings=(InterpolationType=Average,RotationOrderForFilter=XZY)
    Child=(Type=Bone,Name="dyn_fringeA_01_Base_r")
    Parents="((Item=(Type=Bone,Name=\"spine_03\"),Weight=0.0822),(Item=(Type=Bone,Name=\"spine_04\"),Weight=0.0354),(Item=(Type=Bone,Name=\"spine_05\"),Weight=0.0543),(Item=(Type=Bone,Name=\"clavicle_r\"),Weight=0.418),(Item=(Type=Bone,Name=\"upperarm_twist_01_r\"),Weight=0.3778 ...(truncated)"
    Weight=1
   )
   > &RigUnit_ParentConstraint(
    AdvancedSettings=(InterpolationType=Average,RotationOrderForFilter=XZY)
    Child=(Type=Bone,Name="dyn_fringeH_01_Base_r")
    Parents="((Item=(Type=Bone,Name=\"spine_03\"),Weight=0.0822),(Item=(Type=Bone,Name=\"spine_04\"),Weight=0.0354),(Item=(Type=Bone,Name=\"spine_05\"),Weight=0.0543),(Item=(Type=Bone,Name=\"clavicle_r\"),Weight=0.418),(Item=(Type=Bone,Name=\"upperarm_twist_01_r\"),Weight=0.3778 ...(truncated)"
    Weight=1
   )
   > &RigUnit_ParentConstraint(
    AdvancedSettings=(InterpolationType=Average,RotationOrderForFilter=XZY)
    Child=(Type=Bone,Name="dyn_fringeB_01_Base_r")
    Parents="((Item=(Type=Bone,Name=\"spine_03\"),Weight=0.0575),(Item=(Type=Bone,Name=\"spine_04\"),Weight=0.0424),(Item=(Type=Bone,Name=\"spine_05\"),Weight=0.0314),(Item=(Type=Bone,Name=\"clavicle_r\"),Weight=0.2066),(Item=(Type=Bone,Name=\"upperarm_twist_01_r\"),Weight=0.414 ...(truncated)"
    Weight=1
   )
   > &RigUnit_ParentConstraint(
    AdvancedSettings=(InterpolationType=Average,RotationOrderForFilter=XZY)
    Child=(Type=Bone,Name="dyn_fringeC_01_Base_r")
    Parents="((Item=(Type=Bone,Name=\"spine_03\"),Weight=0.0575),(Item=(Type=Bone,Name=\"spine_04\"),Weight=0.0424),(Item=(Type=Bone,Name=\"spine_05\"),Weight=0.0314),(Item=(Type=Bone,Name=\"clavicle_r\"),Weight=0.2066),(Item=(Type=Bone,Name=\"upperarm_twist_01_r\"),Weight=0.414 ...(truncated)"
    Weight=1
   )
   > &RigUnit_ParentConstraint(
    AdvancedSettings=(InterpolationType=Average,RotationOrderForFilter=XZY)
    Child=(Type=Bone,Name="dyn_fringeD_01_Base_r")
    Parents="((Item=(Type=Bone,Name=\"spine_03\"),Weight=0.0575),(Item=(Type=Bone,Name=\"spine_04\"),Weight=0.0424),(Item=(Type=Bone,Name=\"spine_05\"),Weight=0.0314),(Item=(Type=Bone,Name=\"clavicle_r\"),Weight=0.2066),(Item=(Type=Bone,Name=\"upperarm_twist_01_r\"),Weight=0.414 ...(truncated)"
    Weight=1
   )
   > &RigUnit_ParentConstraint(
    AdvancedSettings=(InterpolationType=Average,RotationOrderForFilter=XZY)
    Child=(Type=Bone,Name="dyn_fringeE_01_Base_r")
    Parents="((Item=(Type=Bone,Name=\"spine_03\"),Weight=0.0575),(Item=(Type=Bone,Name=\"spine_04\"),Weight=0.0424),(Item=(Type=Bone,Name=\"spine_05\"),Weight=0.0314),(Item=(Type=Bone,Name=\"clavicle_r\"),Weight=0.2066),(Item=(Type=Bone,Name=\"upperarm_twist_01_r\"),Weight=0.414 ...(truncated)"
    Weight=1
   )
   > &RigUnit_ParentConstraint(
    AdvancedSettings=(InterpolationType=Average,RotationOrderForFilter=XZY)
    Child=(Type=Bone,Name="dyn_fringeF_01_Base_r")
    Parents="((Item=(Type=Bone,Name=\"spine_03\"),Weight=0.0575),(Item=(Type=Bone,Name=\"spine_04\"),Weight=0.0424),(Item=(Type=Bone,Name=\"spine_05\"),Weight=0.0314),(Item=(Type=Bone,Name=\"clavicle_r\"),Weight=0.2066),(Item=(Type=Bone,Name=\"upperarm_twist_01_r\"),Weight=0.414 ...(truncated)"
    Weight=1
   )
   > &RigUnit_ParentConstraint(
    AdvancedSettings=(InterpolationType=Average,RotationOrderForFilter=XZY)
    Child=(Type=Bone,Name="dyn_fringeG_01_Base_r")
    Parents="((Item=(Type=Bone,Name=\"spine_03\"),Weight=0.0575),(Item=(Type=Bone,Name=\"spine_04\"),Weight=0.0424),(Item=(Type=Bone,Name=\"spine_05\"),Weight=0.0314),(Item=(Type=Bone,Name=\"clavicle_r\"),Weight=0.2066),(Item=(Type=Bone,Name=\"upperarm_twist_01_r\"),Weight=0.414 ...(truncated)"
    Weight=1
   )
   n8 = &RigVMFunction_MathQuaternionToEuler(RotationOrder=ZYX, Value=&RigUnit_GetRelativeBoneTransform(Bone=upperarm_r, Space=spine_05).Transform.Rotation)
   n9 = &RigVMFunction_MathFloatAbs(Value=@n8.Result.Z) {template="Absolute::Execute(in Value,out Result)"}
   n10 = RigVMVariableNode() {variable=skydivingINV}
   > &RigUnit_ModifyBoneTransforms(
    BoneToModify=((Bone="dyn_mainShoulderpadArmBase_r",Transform=(Rotation=(X=0,Y=0,Z=0,W=1),Translation=(X=4,Y=-1,Z=-1),Scale3D=(X=0.75,Y=1,Z=1))))
    Mode=AdditiveLocal
    WeightMaximum=1
    Weight=&RigVMFunction_MathFloatMul(
     A=&RigVMFunction_MathFloatRemap(SourceMaximum=123, SourceMinimum=15, TargetMinimum=0, bClamp=true, Value=@n9.Result).Result
     B=@n10.Value
    ) {template="Multiply::Execute(in A,in B,out Result)"}.Result
    BoneToModify.0.Transform.Rotation=&RigVMFunction_MathQuaternionFromEuler(Euler=(X=0,Y=10,Z=20)).Result
   )
   n11 = &RigVMFunction_MathFloatRemap(
    SourceMaximum=109
    SourceMinimum=142
    TargetMinimum=0
    bClamp=true
    Value=&RigVMFunction_MathFloatAbs(Value=@n8.Result.X) {template="Absolute::Execute(in Value,out Result)"}.Result
   )
   > &RigUnit_ModifyBoneTransforms(
    BoneToModify=((Bone="dyn_mainShoulderpadArmBase_r",Transform=(Rotation=(X=0,Y=0,Z=0,W=1),Translation=(X=0,Y=0,Z=0),Scale3D=(X=1,Y=1,Z=1))))
    Mode=AdditiveLocal
    WeightMaximum=1
    WeightMinimum=0
    Weight=@n11.Result
    BoneToModify.0.Transform.Rotation=&RigVMFunction_MathQuaternionFromEuler(Euler=(X=0,Y=0,Z=0)).Result
   )
   > &RigUnit_ModifyBoneTransforms(
    BoneToModify=((Bone="dyn_mainShoulderpadArmBase_r",Transform=(Rotation=(X=0,Y=0,Z=0,W=1),Translation=(X=0,Y=-1.5,Z=0),Scale3D=(X=1,Y=1,Z=1))))
    Mode=AdditiveLocal
    WeightMaximum=1
    Weight=&RigVMFunction_MathFloatMul(
     A=@n11.Result
     B=&RigVMFunction_MathFloatRemap(SourceMaximum=20, SourceMinimum=80, TargetMinimum=0, bClamp=true, Value=@n9.Result).Result
    ) {template="Multiply::Execute(in A,in B,out Result)"}.Result
    BoneToModify.0.Transform.Rotation=&RigVMFunction_MathQuaternionFromEuler(Euler=(X=5,Y=10,Z=0)).Result
   )
   n12 = &RigVMFunction_MathFloatRemap(SourceMaximum=65, SourceMinimum=10, TargetMinimum=0, bClamp=true, Value=@n8.Result.Z)
   > &RigUnit_ModifyBoneTransforms(
    BoneToModify=((Bone="dyn_fringeF_01_Base_r",Transform=(Rotation=(X=0,Y=0,Z=0,W=1),Translation=(X=4,Y=-4,Z=-8),Scale3D=(X=1,Y=1,Z=1))))
    Mode=AdditiveLocal
    WeightMaximum=1
    Weight=@n12.Result
    BoneToModify.0.Transform.Rotation=&RigVMFunction_MathQuaternionFromEuler(Euler=(X=0,Y=60,Z=0)).Result
   )
   > &RigUnit_ModifyBoneTransforms(
    BoneToModify=((Bone="dyn_fringeE_01_Base_r",Transform=(Rotation=(X=0,Y=0,Z=0,W=1),Translation=(X=0,Y=-2,Z=0),Scale3D=(X=1,Y=1,Z=1))))
    Mode=AdditiveLocal
    WeightMaximum=1
    Weight=@n12.Result
   )
   > &RigUnit_ModifyBoneTransforms(
    BoneToModify=((Bone="dyn_fringeD_01_Base_r",Transform=(Rotation=(X=0,Y=0,Z=0,W=1),Translation=(X=0,Y=-2,Z=0),Scale3D=(X=1,Y=1,Z=1))))
    Mode=AdditiveLocal
    WeightMaximum=1
    Weight=@n12.Result
   )
   > &RigUnit_ModifyBoneTransforms(
    BoneToModify=((Bone="dyn_fringeH_01_Base_r",Transform=(Rotation=(X=0,Y=0,Z=0,W=1),Translation=(X=0,Y=0,Z=0),Scale3D=(X=1,Y=1,Z=1))))
    Mode=AdditiveLocal
    WeightMaximum=1
    Weight=@n12.Result
   )
   n13 = &RigVMFunction_MathFloatAbs(Value=&RigVMFunction_MathQuaternionToEuler(RotationOrder=ZYX, Value=&RigUnit_GetRelativeBoneTransform(Bone=upperarm_r, Space=spine_05).Transform.Rotation).Result.Z) {template="Absolute::Execute(in Value,out Result)"}
   > &RigUnit_ModifyBoneTransforms(
    BoneToModify=((Bone="dyn_mainShoulderpad_01_l",Transform=(Rotation=(X=0,Y=0,Z=0,W=1),Translation=(X=-4,Y=0,Z=1.5),Scale3D=(X=1,Y=1,Z=1))))
    Mode=AdditiveLocal
    WeightMaximum=1
    Weight=&RigVMFunction_MathFloatMul(
     B=&RigVMFunction_MathFloatRemap(SourceMaximum=123, SourceMinimum=15, TargetMinimum=0, bClamp=true, Value=@n13.Result).Result
     A=@n10.Value
    ) {template="Multiply::Execute(in A,in B,out Result)"}.Result
    BoneToModify.0.Transform.Rotation=&RigVMFunction_MathQuaternionFromEuler(Euler=(X=0,Y=0,Z=0)).Result
   )
   /Script/ControlRig.RigUnit_TransformConstraint(
    BaseTransform=(Rotation=(X=0,Y=0,Z=0,W=1),Translation=(X=0,Y=0,Z=0),Scale3D=(X=1,Y=1,Z=1))
    BaseTransformSpace=GlobalSpace
    Bone=dyn_mainShoulderpadArmBase_r
    ExecutePin=()
    Targets="((Transform=(Rotation=(X=0,Y=0,Z=0,W=1),Translation=(X=0,Y=0,Z=0),Scale3D=(X=1,Y=1,Z=1)),Weight=0.1,bMaintainOffset=True,Filter=(TranslationFilter=(bX=True,bY=True,bZ=True),RotationFilter=(bX=True,bY=True,bZ=True),ScaleFilter=(bX=True,bY=True,bZ=True))),(T ...(truncated)"
    bUseInitialTransforms=True
    Targets.0.Transform=&RigUnit_GetBoneTransform(Bone=spine_03).Transform
    Targets.1.Transform=&RigUnit_GetBoneTransform(Bone=spine_04).Transform
    Targets.3.Transform=&RigUnit_GetBoneTransform(Bone=clavicle_r).Transform
    Targets.2.Transform=&RigUnit_GetBoneTransform(Bone=spine_05).Transform
    Targets.4.Transform=&RigUnit_GetBoneTransform(Bone=upperarm_twist_01_r).Transform
   ) {is=struct RigVMUnitNode}
   > &RigUnit_TransformConstraint(
    BaseTransform=(Rotation=(X=0,Y=0,Z=0,W=1),Translation=(X=0,Y=0,Z=0),Scale3D=(X=1,Y=1,Z=1))
    BaseTransformSpace=GlobalSpace
    Bone=dyn_mainShoulderpadArmBase_l
    Targets="((Transform=(Rotation=(X=0,Y=0,Z=0,W=1),Translation=(X=0,Y=0,Z=0),Scale3D=(X=1,Y=1,Z=1)),Weight=0.1,bMaintainOffset=True,Filter=(TranslationFilter=(bX=True,bY=True,bZ=True),RotationFilter=(bX=True,bY=True,bZ=True),ScaleFilter=(bX=True,bY=True,bZ=True))),(T ...(truncated)"
    bUseInitialTransforms=True
    Targets.0.Transform=&RigUnit_GetBoneTransform(Bone=spine_03).Transform
    Targets.2.Transform=&RigUnit_GetBoneTransform(Bone=spine_05).Transform
    Targets.3.Transform=&RigUnit_GetBoneTransform(Bone=clavicle_l).Transform
    Targets.1.Transform=&RigUnit_GetBoneTransform(Bone=spine_04).Transform
    Targets.4.Transform=&RigUnit_GetBoneTransform(Bone=upperarm_twist_01_l).Transform
   )
   &RigUnit_TransformConstraint(
    BaseTransform=(Rotation=(X=0,Y=0,Z=0,W=1),Translation=(X=0,Y=0,Z=0),Scale3D=(X=1,Y=1,Z=1))
    BaseTransformSpace=GlobalSpace
    Bone=dyn_fringeA_01_Base_l
    ExecutePin=()
    Targets="((Transform=(Rotation=(X=0,Y=0,Z=0,W=1),Translation=(X=0,Y=0,Z=0),Scale3D=(X=1,Y=1,Z=1)),Weight=0.0822,bMaintainOffset=True,Filter=(TranslationFilter=(bX=True,bY=True,bZ=True),RotationFilter=(bX=True,bY=True,bZ=True),ScaleFilter=(bX=True,bY=True,bZ=True))) ...(truncated)"
    bUseInitialTransforms=True
    Targets.0.Transform=&RigUnit_GetBoneTransform(Bone=spine_03).Transform
    Targets.2.Transform=&RigUnit_GetBoneTransform(Bone=spine_05).Transform
    Targets.3.Transform=&RigUnit_GetBoneTransform(Bone=clavicle_l).Transform
    Targets.1.Transform=&RigUnit_GetBoneTransform(Bone=spine_04).Transform
    Targets.4.Transform=&RigUnit_GetBoneTransform(Bone=upperarm_twist_01_l).Transform
    Targets.5.Transform=&RigUnit_GetBoneTransform(Bone=upperarm_twist_02_l).Transform
    Targets.6.Transform=&RigUnit_GetBoneTransform(Bone=upperarm_l).Transform
   )
   > &RigUnit_TransformConstraint(
    BaseTransform=(Rotation=(X=0,Y=0,Z=0,W=1),Translation=(X=0,Y=0,Z=0),Scale3D=(X=1,Y=1,Z=1))
    BaseTransformSpace=GlobalSpace
    Bone=dyn_fringeH_01_Base_l
    Targets="((Transform=(Rotation=(X=0,Y=0,Z=0,W=1),Translation=(X=0,Y=0,Z=0),Scale3D=(X=1,Y=1,Z=1)),Weight=0.0822,bMaintainOffset=True,Filter=(TranslationFilter=(bX=True,bY=True,bZ=True),RotationFilter=(bX=True,bY=True,bZ=True),ScaleFilter=(bX=True,bY=True,bZ=True))) ...(truncated)"
    bUseInitialTransforms=True
    Targets.0.Transform=&RigUnit_GetBoneTransform(Bone=spine_03).Transform
    Targets.2.Transform=&RigUnit_GetBoneTransform(Bone=spine_05).Transform
    Targets.3.Transform=&RigUnit_GetBoneTransform(Bone=clavicle_l).Transform
    Targets.1.Transform=&RigUnit_GetBoneTransform(Bone=spine_04).Transform
    Targets.6.Transform=&RigUnit_GetBoneTransform(Bone=upperarm_l).Transform
    Targets.5.Transform=&RigUnit_GetBoneTransform(Bone=upperarm_twist_02_l).Transform
    Targets.4.Transform=&RigUnit_GetBoneTransform(Bone=upperarm_twist_01_l).Transform
   )
   n14 = &RigUnit_GetBoneTransform(Bone=spine_03)
   n15 = &RigUnit_GetBoneTransform(Bone=spine_05)
   n16 = &RigUnit_GetBoneTransform(Bone=clavicle_l)
   n17 = &RigUnit_GetBoneTransform(Bone=spine_04)
   n18 = &RigUnit_GetBoneTransform(Bone=upperarm_l)
   n19 = &RigUnit_GetBoneTransform(Bone=upperarm_twist_02_l)
   n20 = &RigUnit_GetBoneTransform(Bone=upperarm_twist_01_l)
   > &RigUnit_TransformConstraint(
    BaseTransform=(Rotation=(X=0,Y=0,Z=0,W=1),Translation=(X=0,Y=0,Z=0),Scale3D=(X=1,Y=1,Z=1))
    BaseTransformSpace=GlobalSpace
    Bone=dyn_fringeB_01_Base_l
    Targets="((Transform=(Rotation=(X=0,Y=0,Z=0,W=1),Translation=(X=0,Y=0,Z=0),Scale3D=(X=1,Y=1,Z=1)),Weight=0.0575,bMaintainOffset=True,Filter=(TranslationFilter=(bX=True,bY=True,bZ=True),RotationFilter=(bX=True,bY=True,bZ=True),ScaleFilter=(bX=True,bY=True,bZ=True))) ...(truncated)"
    bUseInitialTransforms=True
    Targets.0.Transform=@n14.Transform
    Targets.2.Transform=@n15.Transform
    Targets.3.Transform=@n16.Transform
    Targets.1.Transform=@n17.Transform
    Targets.6.Transform=@n18.Transform
    Targets.5.Transform=@n19.Transform
    Targets.4.Transform=@n20.Transform
   )
   > &RigUnit_TransformConstraint(
    BaseTransform=(Rotation=(X=0,Y=0,Z=0,W=1),Translation=(X=0,Y=0,Z=0),Scale3D=(X=1,Y=1,Z=1))
    BaseTransformSpace=GlobalSpace
    Bone=dyn_fringeC_01_Base_l
    Targets="((Transform=(Rotation=(X=0,Y=0,Z=0,W=1),Translation=(X=0,Y=0,Z=0),Scale3D=(X=1,Y=1,Z=1)),Weight=0.0575,bMaintainOffset=True,Filter=(TranslationFilter=(bX=True,bY=True,bZ=True),RotationFilter=(bX=True,bY=True,bZ=True),ScaleFilter=(bX=True,bY=True,bZ=True))) ...(truncated)"
    bUseInitialTransforms=True
    Targets.0.Transform=@n14.Transform
    Targets.2.Transform=@n15.Transform
    Targets.3.Transform=@n16.Transform
    Targets.1.Transform=@n17.Transform
    Targets.6.Transform=@n18.Transform
    Targets.5.Transform=@n19.Transform
    Targets.4.Transform=@n20.Transform
   )
   > &RigUnit_TransformConstraint(
    BaseTransform=(Rotation=(X=0,Y=0,Z=0,W=1),Translation=(X=0,Y=0,Z=0),Scale3D=(X=1,Y=1,Z=1))
    BaseTransformSpace=GlobalSpace
    Bone=dyn_fringeD_01_Base_l
    Targets="((Transform=(Rotation=(X=0,Y=0,Z=0,W=1),Translation=(X=0,Y=0,Z=0),Scale3D=(X=1,Y=1,Z=1)),Weight=0.0575,bMaintainOffset=True,Filter=(TranslationFilter=(bX=True,bY=True,bZ=True),RotationFilter=(bX=True,bY=True,bZ=True),ScaleFilter=(bX=True,bY=True,bZ=True))) ...(truncated)"
    bUseInitialTransforms=True
    Targets.0.Transform=@n14.Transform
    Targets.2.Transform=@n15.Transform
    Targets.3.Transform=@n16.Transform
    Targets.1.Transform=@n17.Transform
    Targets.6.Transform=@n18.Transform
    Targets.5.Transform=@n19.Transform
    Targets.4.Transform=@n20.Transform
   )
   > &RigUnit_TransformConstraint(
    BaseTransform=(Rotation=(X=0,Y=0,Z=0,W=1),Translation=(X=0,Y=0,Z=0),Scale3D=(X=1,Y=1,Z=1))
    BaseTransformSpace=GlobalSpace
    Bone=dyn_fringeE_01_Base_l
    Targets="((Transform=(Rotation=(X=0,Y=0,Z=0,W=1),Translation=(X=0,Y=0,Z=0),Scale3D=(X=1,Y=1,Z=1)),Weight=0.0575,bMaintainOffset=True,Filter=(TranslationFilter=(bX=True,bY=True,bZ=True),RotationFilter=(bX=True,bY=True,bZ=True),ScaleFilter=(bX=True,bY=True,bZ=True))) ...(truncated)"
    bUseInitialTransforms=True
    Targets.0.Transform=@n14.Transform
    Targets.2.Transform=@n15.Transform
    Targets.3.Transform=@n16.Transform
    Targets.1.Transform=@n17.Transform
    Targets.6.Transform=@n18.Transform
    Targets.5.Transform=@n19.Transform
    Targets.4.Transform=@n20.Transform
   )
   > &RigUnit_TransformConstraint(
    BaseTransform=(Rotation=(X=0,Y=0,Z=0,W=1),Translation=(X=0,Y=0,Z=0),Scale3D=(X=1,Y=1,Z=1))
    BaseTransformSpace=GlobalSpace
    Bone=dyn_fringeF_01_Base_l
    Targets="((Transform=(Rotation=(X=0,Y=0,Z=0,W=1),Translation=(X=0,Y=0,Z=0),Scale3D=(X=1,Y=1,Z=1)),Weight=0.0575,bMaintainOffset=True,Filter=(TranslationFilter=(bX=True,bY=True,bZ=True),RotationFilter=(bX=True,bY=True,bZ=True),ScaleFilter=(bX=True,bY=True,bZ=True))) ...(truncated)"
    bUseInitialTransforms=True
    Targets.0.Transform=@n14.Transform
    Targets.2.Transform=@n15.Transform
    Targets.3.Transform=@n16.Transform
    Targets.1.Transform=@n17.Transform
    Targets.6.Transform=@n18.Transform
    Targets.5.Transform=@n19.Transform
    Targets.4.Transform=@n20.Transform
   )
   > &RigUnit_TransformConstraint(
    BaseTransform=(Rotation=(X=0,Y=0,Z=0,W=1),Translation=(X=0,Y=0,Z=0),Scale3D=(X=1,Y=1,Z=1))
    BaseTransformSpace=GlobalSpace
    Bone=dyn_fringeG_01_Base_l
    Targets="((Transform=(Rotation=(X=0,Y=0,Z=0,W=1),Translation=(X=0,Y=0,Z=0),Scale3D=(X=1,Y=1,Z=1)),Weight=0.0575,bMaintainOffset=True,Filter=(TranslationFilter=(bX=True,bY=True,bZ=True),RotationFilter=(bX=True,bY=True,bZ=True),ScaleFilter=(bX=True,bY=True,bZ=True))) ...(truncated)"
    bUseInitialTransforms=True
    Targets.0.Transform=@n14.Transform
    Targets.2.Transform=@n15.Transform
    Targets.3.Transform=@n16.Transform
    Targets.1.Transform=@n17.Transform
    Targets.6.Transform=@n18.Transform
    Targets.5.Transform=@n19.Transform
    Targets.4.Transform=@n20.Transform
   )
   > &RigUnit_TransformConstraint(
    BaseTransform=(Rotation=(X=0,Y=0,Z=0,W=1),Translation=(X=0,Y=0,Z=0),Scale3D=(X=1,Y=1,Z=1))
    BaseTransformSpace=GlobalSpace
    Bone=dyn_fringeA_01_Base_r
    Targets="((Transform=(Rotation=(X=0,Y=0,Z=0,W=1),Translation=(X=0,Y=0,Z=0),Scale3D=(X=1,Y=1,Z=1)),Weight=0.0822,bMaintainOffset=True,Filter=(TranslationFilter=(bX=True,bY=True,bZ=True),RotationFilter=(bX=True,bY=True,bZ=True),ScaleFilter=(bX=True,bY=True,bZ=True))) ...(truncated)"
    bUseInitialTransforms=True
    Targets.0.Transform=&RigUnit_GetBoneTransform(Bone=spine_03).Transform
    Targets.2.Transform=&RigUnit_GetBoneTransform(Bone=spine_05).Transform
    Targets.3.Transform=&RigUnit_GetBoneTransform(Bone=clavicle_r).Transform
    Targets.1.Transform=&RigUnit_GetBoneTransform(Bone=spine_04).Transform
    Targets.6.Transform=&RigUnit_GetBoneTransform(Bone=upperarm_r).Transform
    Targets.5.Transform=&RigUnit_GetBoneTransform(Bone=upperarm_twist_02_r).Transform
    Targets.4.Transform=&RigUnit_GetBoneTransform(Bone=upperarm_twist_01_r).Transform
   )
   > &RigUnit_TransformConstraint(
    BaseTransform=(Rotation=(X=0,Y=0,Z=0,W=1),Translation=(X=0,Y=0,Z=0),Scale3D=(X=1,Y=1,Z=1))
    BaseTransformSpace=GlobalSpace
    Bone=dyn_fringeH_01_Base_r
    Targets="((Transform=(Rotation=(X=0,Y=0,Z=0,W=1),Translation=(X=0,Y=0,Z=0),Scale3D=(X=1,Y=1,Z=1)),Weight=0.0822,bMaintainOffset=True,Filter=(TranslationFilter=(bX=True,bY=True,bZ=True),RotationFilter=(bX=True,bY=True,bZ=True),ScaleFilter=(bX=True,bY=True,bZ=True))) ...(truncated)"
    bUseInitialTransforms=True
    Targets.0.Transform=&RigUnit_GetBoneTransform(Bone=spine_03).Transform
    Targets.2.Transform=&RigUnit_GetBoneTransform(Bone=spine_05).Transform
    Targets.3.Transform=&RigUnit_GetBoneTransform(Bone=clavicle_r).Transform
    Targets.1.Transform=&RigUnit_GetBoneTransform(Bone=spine_04).Transform
    Targets.6.Transform=&RigUnit_GetBoneTransform(Bone=upperarm_r).Transform
    Targets.5.Transform=&RigUnit_GetBoneTransform(Bone=upperarm_twist_02_r).Transform
    Targets.4.Transform=&RigUnit_GetBoneTransform(Bone=upperarm_twist_01_r).Transform
   )
   n21 = &RigUnit_GetBoneTransform(Bone=spine_03)
   n22 = &RigUnit_GetBoneTransform(Bone=spine_05)
   n23 = &RigUnit_GetBoneTransform(Bone=clavicle_r)
   n24 = &RigUnit_GetBoneTransform(Bone=spine_04)
   n25 = &RigUnit_GetBoneTransform(Bone=upperarm_r)
   n26 = &RigUnit_GetBoneTransform(Bone=upperarm_twist_02_r)
   n27 = &RigUnit_GetBoneTransform(Bone=upperarm_twist_01_r)
   > &RigUnit_TransformConstraint(
    BaseTransform=(Rotation=(X=0,Y=0,Z=0,W=1),Translation=(X=0,Y=0,Z=0),Scale3D=(X=1,Y=1,Z=1))
    BaseTransformSpace=GlobalSpace
    Bone=dyn_fringeB_01_Base_r
    Targets="((Transform=(Rotation=(X=0,Y=0,Z=0,W=1),Translation=(X=0,Y=0,Z=0),Scale3D=(X=1,Y=1,Z=1)),Weight=0.0575,bMaintainOffset=True,Filter=(TranslationFilter=(bX=True,bY=True,bZ=True),RotationFilter=(bX=True,bY=True,bZ=True),ScaleFilter=(bX=True,bY=True,bZ=True))) ...(truncated)"
    bUseInitialTransforms=True
    Targets.0.Transform=@n21.Transform
    Targets.2.Transform=@n22.Transform
    Targets.3.Transform=@n23.Transform
    Targets.1.Transform=@n24.Transform
    Targets.6.Transform=@n25.Transform
    Targets.5.Transform=@n26.Transform
    Targets.4.Transform=@n27.Transform
   )
   > &RigUnit_TransformConstraint(
    BaseTransform=(Rotation=(X=0,Y=0,Z=0,W=1),Translation=(X=0,Y=0,Z=0),Scale3D=(X=1,Y=1,Z=1))
    BaseTransformSpace=GlobalSpace
    Bone=dyn_fringeC_01_Base_r
    Targets="((Transform=(Rotation=(X=0,Y=0,Z=0,W=1),Translation=(X=0,Y=0,Z=0),Scale3D=(X=1,Y=1,Z=1)),Weight=0.0575,bMaintainOffset=True,Filter=(TranslationFilter=(bX=True,bY=True,bZ=True),RotationFilter=(bX=True,bY=True,bZ=True),ScaleFilter=(bX=True,bY=True,bZ=True))) ...(truncated)"
    bUseInitialTransforms=True
    Targets.0.Transform=@n21.Transform
    Targets.2.Transform=@n22.Transform
    Targets.3.Transform=@n23.Transform
    Targets.1.Transform=@n24.Transform
    Targets.6.Transform=@n25.Transform
    Targets.5.Transform=@n26.Transform
    Targets.4.Transform=@n27.Transform
   )
   > &RigUnit_TransformConstraint(
    BaseTransform=(Rotation=(X=0,Y=0,Z=0,W=1),Translation=(X=0,Y=0,Z=0),Scale3D=(X=1,Y=1,Z=1))
    BaseTransformSpace=GlobalSpace
    Bone=dyn_fringeD_01_Base_r
    Targets="((Transform=(Rotation=(X=0,Y=0,Z=0,W=1),Translation=(X=0,Y=0,Z=0),Scale3D=(X=1,Y=1,Z=1)),Weight=0.0575,bMaintainOffset=True,Filter=(TranslationFilter=(bX=True,bY=True,bZ=True),RotationFilter=(bX=True,bY=True,bZ=True),ScaleFilter=(bX=True,bY=True,bZ=True))) ...(truncated)"
    bUseInitialTransforms=True
    Targets.0.Transform=@n21.Transform
    Targets.2.Transform=@n22.Transform
    Targets.3.Transform=@n23.Transform
    Targets.1.Transform=@n24.Transform
    Targets.6.Transform=@n25.Transform
    Targets.5.Transform=@n26.Transform
    Targets.4.Transform=@n27.Transform
   )
   > &RigUnit_TransformConstraint(
    BaseTransform=(Rotation=(X=0,Y=0,Z=0,W=1),Translation=(X=0,Y=0,Z=0),Scale3D=(X=1,Y=1,Z=1))
    BaseTransformSpace=GlobalSpace
    Bone=dyn_fringeE_01_Base_r
    Targets="((Transform=(Rotation=(X=0,Y=0,Z=0,W=1),Translation=(X=0,Y=0,Z=0),Scale3D=(X=1,Y=1,Z=1)),Weight=0.0575,bMaintainOffset=True,Filter=(TranslationFilter=(bX=True,bY=True,bZ=True),RotationFilter=(bX=True,bY=True,bZ=True),ScaleFilter=(bX=True,bY=True,bZ=True))) ...(truncated)"
    bUseInitialTransforms=True
    Targets.0.Transform=@n21.Transform
    Targets.2.Transform=@n22.Transform
    Targets.3.Transform=@n23.Transform
    Targets.1.Transform=@n24.Transform
    Targets.6.Transform=@n25.Transform
    Targets.5.Transform=@n26.Transform
    Targets.4.Transform=@n27.Transform
   )
   > &RigUnit_TransformConstraint(
    BaseTransform=(Rotation=(X=0,Y=0,Z=0,W=1),Translation=(X=0,Y=0,Z=0),Scale3D=(X=1,Y=1,Z=1))
    BaseTransformSpace=GlobalSpace
    Bone=dyn_fringeF_01_Base_r
    Targets="((Transform=(Rotation=(X=0,Y=0,Z=0,W=1),Translation=(X=0,Y=0,Z=0),Scale3D=(X=1,Y=1,Z=1)),Weight=0.0575,bMaintainOffset=True,Filter=(TranslationFilter=(bX=True,bY=True,bZ=True),RotationFilter=(bX=True,bY=True,bZ=True),ScaleFilter=(bX=True,bY=True,bZ=True))) ...(truncated)"
    bUseInitialTransforms=True
    Targets.0.Transform=@n21.Transform
    Targets.2.Transform=@n22.Transform
    Targets.3.Transform=@n23.Transform
    Targets.1.Transform=@n24.Transform
    Targets.6.Transform=@n25.Transform
    Targets.5.Transform=@n26.Transform
    Targets.4.Transform=@n27.Transform
   )
   > &RigUnit_TransformConstraint(
    BaseTransform=(Rotation=(X=0,Y=0,Z=0,W=1),Translation=(X=0,Y=0,Z=0),Scale3D=(X=1,Y=1,Z=1))
    BaseTransformSpace=GlobalSpace
    Bone=dyn_fringeG_01_Base_r
    Targets="((Transform=(Rotation=(X=0,Y=0,Z=0,W=1),Translation=(X=0,Y=0,Z=0),Scale3D=(X=1,Y=1,Z=1)),Weight=0.0575,bMaintainOffset=True,Filter=(TranslationFilter=(bX=True,bY=True,bZ=True),RotationFilter=(bX=True,bY=True,bZ=True),ScaleFilter=(bX=True,bY=True,bZ=True))) ...(truncated)"
    bUseInitialTransforms=True
    Targets.0.Transform=@n21.Transform
    Targets.2.Transform=@n22.Transform
    Targets.3.Transform=@n23.Transform
    Targets.1.Transform=@n24.Transform
    Targets.6.Transform=@n25.Transform
    Targets.5.Transform=@n26.Transform
    Targets.4.Transform=@n27.Transform
   )
   n28 = &RigUnit_GetBoneTransform(Bone=dyn_loinClothFront_01_c_Base)
   &RigUnit_TransformConstraint(
    BaseTransform=(Rotation=(X=0,Y=0,Z=0,W=1),Translation=(X=0,Y=0,Z=0),Scale3D=(X=1,Y=1,Z=1))
    BaseTransformSpace=GlobalSpace
    Bone=dyn_loinClothFront_01_Base_l
    ExecutePin=()
    Targets="((Transform=(Rotation=(X=0,Y=0,Z=0,W=1),Translation=(X=0,Y=0,Z=0),Scale3D=(X=1,Y=1,Z=1)),Weight=0.5,bMaintainOffset=True,Filter=(TranslationFilter=(bX=True,bY=True,bZ=True),RotationFilter=(bX=True,bY=True,bZ=True),ScaleFilter=(bX=True,bY=True,bZ=True))),(T ...(truncated)"
    bUseInitialTransforms=True
    Targets.0.Transform=&RigUnit_GetBoneTransform(Bone=thigh_twist_01_l).Transform
    Targets.1.Transform=@n28.Transform
   )
   > &RigUnit_TransformConstraint(
    BaseTransform=(Rotation=(X=0,Y=0,Z=0,W=1),Translation=(X=0,Y=0,Z=0),Scale3D=(X=1,Y=1,Z=1))
    BaseTransformSpace=GlobalSpace
    Bone=dyn_loinClothFront_01_Base_r
    Targets="((Transform=(Rotation=(X=0,Y=0,Z=0,W=1),Translation=(X=0,Y=0,Z=0),Scale3D=(X=1,Y=1,Z=1)),Weight=0.5,bMaintainOffset=True,Filter=(TranslationFilter=(bX=True,bY=True,bZ=True),RotationFilter=(bX=True,bY=True,bZ=True),ScaleFilter=(bX=True,bY=True,bZ=True))),(T ...(truncated)"
    bUseInitialTransforms=True
    Targets.0.Transform=&RigUnit_GetBoneTransform(Bone=thigh_twist_01_r).Transform
    Targets.1.Transform=@n28.Transform
   )
   &RigUnit_GetBoneTransform(Bone=thigh_twist_01_r)
   &RigUnit_GetBoneTransform(Bone=thigh_twist_01_l)
   &RigUnit_ModifyBoneTransforms(
    BoneToModify=((Bone="dyn_cutFlap_Base_l",Transform=(Rotation=(X=0,Y=0,Z=0,W=1),Translation=(X=-5,Y=-9.5,Z=2),Scale3D=(X=1,Y=1,Z=1))))
    ExecutePin=()
    Mode=AdditiveLocal
    WeightMaximum=1
    Weight=&RigVMFunction_MathFloatMul(A=@n4.Result, B=@n1.Value) {template="Multiply::Execute(in A,in B,out Result)"}.Result
    BoneToModify.0.Transform.Rotation=@n5.Result
   )
   &RigUnit_TransformConstraint(
    BaseTransform=(Rotation=(X=0,Y=0,Z=0,W=1),Translation=(X=0,Y=0,Z=0),Scale3D=(X=1,Y=1,Z=1))
    BaseTransformSpace=GlobalSpace
    Bone=dyn_loinClothFront_01_c_Base
    ExecutePin=()
    Targets="((Transform=(Rotation=(X=0,Y=0,Z=0,W=1),Translation=(X=0,Y=0,Z=0),Scale3D=(X=1,Y=1,Z=1)),Weight=0,bMaintainOffset=True,Filter=(TranslationFilter=(bX=True,bY=True,bZ=True),RotationFilter=(bX=True,bY=True,bZ=True),ScaleFilter=(bX=True,bY=True,bZ=True))),(Tra ...(truncated)"
    bUseInitialTransforms=True
    Targets.0.Transform.Rotation=&RigUnit_GetBoneTransform(Bone=dyn_kneeAim_l).Transform.Rotation
    Targets.1.Transform=@n28.Transform
    Targets.0.Transform.Translation=@n28.Transform.Translation
    Targets.0.Weight=@n2.Result
    Targets.1.Weight=@n3.Result
   )
   n29 = &RigUnit_GetBoneTransform(Bone=dyn_loinClothBack_01_Base)
   &RigUnit_TransformConstraint(
    BaseBone=None
    BaseTransform=(Rotation=(X=0,Y=0,Z=0,W=1),Translation=(X=0,Y=0,Z=0),Scale3D=(X=1,Y=1,Z=1))
    BaseTransformSpace=GlobalSpace
    Bone=dyn_loinClothBack_01_Base
    ExecutePin=()
    Targets="((Transform=(Rotation=(X=0,Y=0,Z=0,W=1),Translation=(X=0,Y=0,Z=0),Scale3D=(X=1,Y=1,Z=1)),Weight=0.2,bMaintainOffset=True,Filter=(TranslationFilter=(bX=True,bY=True,bZ=True),RotationFilter=(bX=True,bY=True,bZ=True),ScaleFilter=(bX=True,bY=True,bZ=True))),(T ...(truncated)"
    bUseInitialTransforms=True
    Targets.0.Transform.Rotation=&RigUnit_GetBoneTransform(Bone=dyn_kneeAim_l).Transform.Rotation
    Targets.1.Transform=@n29.Transform
    Targets.0.Transform.Translation=@n29.Transform.Translation
   )
   n30 = &RigUnit_GetBoneTransform(Bone=dyn_loinClothBack_01_Base)
   &RigUnit_TransformConstraint(
    BaseBone=None
    BaseTransform=(Rotation=(X=0,Y=0,Z=0,W=1),Translation=(X=0,Y=0,Z=0),Scale3D=(X=1,Y=1,Z=1))
    BaseTransformSpace=GlobalSpace
    Bone=dyn_loinClothBack_01_Base
    ExecutePin=()
    Targets="((Transform=(Rotation=(X=0,Y=0,Z=0,W=1),Translation=(X=0,Y=0,Z=0),Scale3D=(X=1,Y=1,Z=1)),Weight=0,bMaintainOffset=True,Filter=(TranslationFilter=(bX=True,bY=True,bZ=True),RotationFilter=(bX=True,bY=True,bZ=True),ScaleFilter=(bX=True,bY=True,bZ=True))),(Tra ...(truncated)"
    bUseInitialTransforms=True
    Targets.0.Transform.Translation=&RigUnit_GetBoneTransform(Bone=thigh_twist_01_l).Transform.Translation
    Targets.1.Transform.Translation=&RigUnit_GetBoneTransform(Bone=thigh_twist_01_r).Transform.Translation
    Targets.2.Transform=@n30.Transform
    Targets.0.Transform.Rotation=@n30.Transform.Rotation
    Targets.1.Transform.Rotation=@n30.Transform.Rotation
   )
   &RigUnit_TransformConstraint(
    BaseTransform=(Rotation=(X=0,Y=0,Z=0,W=1),Translation=(X=0,Y=0,Z=0),Scale3D=(X=1,Y=1,Z=1))
    BaseTransformSpace=GlobalSpace
    Bone=dyn_kneeAim_l
    ExecutePin=()
    Targets="((Transform=(Rotation=(X=0,Y=0,Z=0,W=1),Translation=(X=0,Y=0,Z=0),Scale3D=(X=1,Y=1,Z=1)),Weight=1,bMaintainOffset=True,Filter=(TranslationFilter=(bX=True,bY=True,bZ=True),RotationFilter=(bX=True,bY=True,bZ=True),ScaleFilter=(bX=True,bY=True,bZ=True))),(Tra ...(truncated)"
    bUseInitialTransforms=True
    Targets.0.Transform.Translation=&RigUnit_GetBoneTransform(Bone=calf_l).Transform.Translation
    Targets.1.Transform.Translation=&RigUnit_GetBoneTransform(Bone=calf_r).Transform.Translation
   )
  ]}
]
hierarchy [
 root {loc=<0 0 0>, rot=<0 0 0 1>}
 pelvis {parent=root, loc=<0 -0.7071 93.1191>, rot=<0.0127 0.707 -0.0127 -0.707>}
 spine_01 {parent=pelvis, loc=<6.7494 -0.6388 0>, rot=<0 0 0.034 -0.9994>}
 spine_02 {parent=spine_01, loc=<7.2518 -0.035 0>, rot=<0 0 0.0238 -0.9997>}
 spine_03 {parent=spine_02, loc=<6.0104 0 0>, rot=<0 0 0.0149 -0.9999>}
 spine_04 {parent=spine_03, loc=<8.1024 0 0>, rot=<0 0 -0.0132 -0.9999>}
 spine_05 {parent=spine_04, loc=<10.0146 0 0>, rot=<0 0 -0.0015 -1>}
 clavicle_l {parent=spine_05, loc=<4.3575 -0.1466 -2.3223>, rot=<0.0465 -0.7289 -0.009 -0.6829>}
 upperarm_l {parent=clavicle_l, loc=<15.5111 0 0>, rot=<0.0226 -0.3234 -0.0088 -0.9459>}
 lowerarm_l {parent=upperarm_l, loc=<26.6514 0 0>, rot=<0.0391 -0.0099 -0.2453 -0.9686>}
 hand_l {parent=lowerarm_l, loc=<22.3491 0 0>, rot=<-0.7081 -0.0507 0.0975 -0.6975>}
 index_metacarpal_l {parent=hand_l, loc=<1.6535 0.0183 -2.96>, rot=<0.0158 -0.0124 -0.0173 -0.9996>}
 index_01_l {parent=index_metacarpal_l, loc=<8.3008 0.0357 -0.0524>, rot=<-0.0044 0.0485 0.0924 -0.9945>}
 index_02_l {parent=index_01_l, loc=<3.3587 -0.3004 0.0452>, rot=<-0.002 0.0027 0.3225 -0.9466>}
 index_03_l {parent=index_02_l, loc=<2.5578 -0.123 -0.0252>, rot=<-0.002 0.0067 0.2558 -0.9667>}
 middle_metacarpal_l {parent=hand_l, loc=<2.1856 -0.1725 -0.5984>, rot=<0.006 0.0168 -0.0329 -0.9993>}
 middle_01_l {parent=middle_metacarpal_l, loc=<7.3275 -0.2447 0.0129>, rot=<-0.0208 0.0042 0.0858 -0.9961>}
 middle_02_l {parent=middle_01_l, loc=<4.0123 0.1401 0.1175>, rot=<-0.0218 0.0258 0.3512 -0.9357>}
 middle_03_l {parent=middle_02_l, loc=<2.6077 0.0534 0.1373>, rot=<0.0082 0.0188 0.1708 -0.9851>}
 pinky_metacarpal_l {parent=hand_l, loc=<1.7616 -0.9753 3.3814>, rot=<-0.0397 0.0814 -0.069 -0.9935>}
 pinky_01_l {parent=pinky_metacarpal_l, loc=<7.1709 -0.4831 -0.1094>, rot=<-0.0654 -0.0035 0.2242 -0.9723>}
 pinky_02_l {parent=pinky_01_l, loc=<2.8274 0.0893 -0.1043>, rot=<0.0037 0.0011 0.2854 -0.9584>}
 pinky_03_l {parent=pinky_02_l, loc=<2.2046 0.0307 -0.0078>, rot=<-0.0002 0.0098 0.1445 -0.9895>}
 ring_metacarpal_l {parent=hand_l, loc=<2.2074 -0.3625 1.3638>, rot=<-0.0284 0.0722 -0.0458 -0.9959>}
 ring_01_l {parent=ring_metacarpal_l, loc=<7.0295 -0.6197 -0.1446>, rot=<-0.0489 -0.0252 0.1892 -0.9804>}
 ring_02_l {parent=ring_01_l, loc=<3.6174 0.4646 -0.1027>, rot=<0 0 0.2806 -0.9598>}
 ring_03_l {parent=ring_02_l, loc=<2.4152 0.1141 0.0199>, rot=<0 0 0.1189 -0.9929>}
 thumb_01_l {parent=hand_l, loc=<1.3118 -0.3191 -2.459>, rot=<0.6546 -0.4042 0.0664 -0.6354>}
 thumb_02_l {parent=thumb_01_l, loc=<5.9336 -0.0448 0.0401>, rot=<-0.0037 0.0213 0.298 -0.9543>}
 thumb_03_l {parent=thumb_02_l, loc=<2.2223 0.3682 -0.0907>, rot=<0.0118 -0.013 0.0066 -0.9998>}
 lowerarm_twist_01_l {parent=lowerarm_l, loc=<17.8549 0 0>, rot=<0 0 0 1>}
 lowerarm_twist_02_l {parent=lowerarm_l, loc=<9.6199 0 0>, rot=<0 0 0 1>}
 upperarm_twist_01_l {parent=upperarm_l, loc=<0.0041 0 0>, rot=<0 0 0 -1>}
 upperarm_twist_02_l {parent=upperarm_l, loc=<13.7941 0 0>, rot=<0 0 0 -1>}
 dyn_fringeB_01_Base_l {parent=upperarm_l, loc=<6.2508 -9.944 0.0607>, rot=<0.3681 0.1129 0.9226 0.0237>}
 dyn_fringeB_01_l {parent=dyn_fringeB_01_Base_l, loc=<0 0 0>, rot=<0 0 0 -1>}
 dyn_fringeB_02_l {parent=dyn_fringeB_01_l, loc=<-6.8325 0 0>, rot=<0.0168 -0.0204 -0.0507 -0.9984>}
 dyn_fringeC_01_Base_l {parent=upperarm_l, loc=<8.0522 -8.3595 3.4216>, rot=<0.1874 0.2556 0.9483 0.0146>}
 dyn_fringeC_01_l {parent=dyn_fringeC_01_Base_l, loc=<0 0 0>, rot=<0 0 0 1>}
 dyn_fringeC_02_l {parent=dyn_fringeC_01_l, loc=<-6.5825 0.0002 0.0001>, rot=<0 0 -0.0523 -0.9986>}
 dyn_fringeD_01_Base_l {parent=upperarm_l, loc=<10.9231 -4.3324 5.9247>, rot=<0.053 0.6092 0.7913 0.0027>}
 dyn_fringeD_01_l {parent=dyn_fringeD_01_Base_l, loc=<0 0 0>, rot=<0 0 0 -1>}
 dyn_fringeD_02_l {parent=dyn_fringeD_01_l, loc=<-3.4439 0 0>, rot=<-0.0222 -0.0152 -0.0366 -0.999>}
 dyn_fringeE_01_Base_l {parent=upperarm_l, loc=<10.4155 2.6992 5.8153>, rot=<0.0186 -0.8359 -0.5467 0.045>}
 dyn_fringeE_01_l {parent=dyn_fringeE_01_Base_l, loc=<0 0 0>, rot=<0 0 0 -1>}
 dyn_fringeE_02_l {parent=dyn_fringeE_01_l, loc=<-3.5737 0 0>, rot=<0.0059 0.0045 -0.0047 -1>}
 dyn_fringeF_01_Base_l {parent=upperarm_l, loc=<10.4657 6.8331 2.54>, rot=<0.0875 -0.9536 -0.229 0.1747>}
 dyn_fringeF_01_l {parent=dyn_fringeF_01_Base_l, loc=<0 0 0>, rot=<0 0 0 1>}
 dyn_fringeF_02_l {parent=dyn_fringeF_01_l, loc=<-3.5094 0 0>, rot=<-0.0176 -0.0194 -0.0437 -0.9987>}
 dyn_fringeG_01_Base_l {parent=upperarm_l, loc=<7.646 8.5125 -1.3246>, rot=<0.0581 -0.925 -0.0654 0.3699>}
 dyn_fringeG_01_l {parent=dyn_fringeG_01_Base_l, loc=<0 0 0>, rot=<0 0 0 -1>}
 dyn_fringeG_02_l {parent=dyn_fringeG_01_l, loc=<-3.726 0 0>, rot=<0 0 0 -1>}
 dyn_mainShoulderpadArmBase_l {parent=clavicle_l, loc=<15.511 0 0>, rot=<0.0226 -0.3234 -0.0088 -0.9459>}
 dyn_mainShoulderpad_01_Base_l {parent=dyn_mainShoulderpadArmBase_l, loc=<-6.0581 0.1483 8.5265>, rot=<0.5306 -0.6959 -0.2854 0.3908>}
 dyn_mainShoulderpad_01_l {parent=dyn_mainShoulderpad_01_Base_l, loc=<0 0 0>, rot=<0 0 0 1>}
 dyn_mainShoulderpad_02_l {parent=dyn_mainShoulderpad_01_l, loc=<-7.6637 0.0002 0>, rot=<0.0075 -0.0032 -0.4052 -0.9142>}
 dyn_fringeA_01_Base_l {parent=clavicle_l, loc=<15.8777 -12.1202 -3.2418>, rot=<0.6848 0.0681 0.7253 -0.019>}
 dyn_fringeA_01_l {parent=dyn_fringeA_01_Base_l, loc=<0 0 0>, rot=<0 0 0 1>}
 dyn_fringeA_02_l {parent=dyn_fringeA_01_l, loc=<-6.9922 -0.0001 0>, rot=<0.0642 0.0142 -0.0426 -0.9969>}
 dyn_fringeH_01_Base_l {parent=clavicle_l, loc=<14.7513 10.0439 -3.5651>, rot=<-0.0012 0.7203 0.0583 -0.6912>}
 dyn_fringeH_01_l {parent=dyn_fringeH_01_Base_l, loc=<0 0 0>, rot=<0 0 0 -1>}
 dyn_fringeH_02_l {parent=dyn_fringeH_01_l, loc=<-7.0437 0 0>, rot=<-0.0584 -0.0164 -0.0106 -0.9981>}
 clavicle_r {parent=spine_05, loc=<4.3575 -0.1466 2.3223>, rot=<0.7289 0.0465 -0.6829 0.009>}
 upperarm_r {parent=clavicle_r, loc=<-15.5111 0 0>, rot=<0.0226 -0.3234 -0.0088 -0.9459>}
 lowerarm_r {parent=upperarm_r, loc=<-26.6514 0 0>, rot=<0.0391 -0.0099 -0.2453 -0.9686>}
 hand_r {parent=lowerarm_r, loc=<-22.3492 0 0>, rot=<-0.7081 -0.0507 0.0975 -0.6975>}
 index_metacarpal_r {parent=hand_r, loc=<-1.6534 -0.0181 2.9601>, rot=<0.0158 -0.0124 -0.0173 -0.9996>}
 index_01_r {parent=index_metacarpal_r, loc=<-8.3008 -0.0357 0.0524>, rot=<-0.0044 0.0485 0.0924 -0.9945>}
 index_02_r {parent=index_01_r, loc=<-3.3589 0.3003 -0.0453>, rot=<-0.002 0.0027 0.3225 -0.9466>}
 index_03_r {parent=index_02_r, loc=<-2.5578 0.1231 0.0251>, rot=<-0.002 0.0067 0.2558 -0.9667>}
 middle_metacarpal_r {parent=hand_r, loc=<-2.1856 0.1726 0.5984>, rot=<0.006 0.0168 -0.0329 -0.9993>}
 middle_01_r {parent=middle_metacarpal_r, loc=<-7.3273 0.2447 -0.0129>, rot=<-0.0208 0.0042 0.0858 -0.9961>}
 middle_02_r {parent=middle_01_r, loc=<-4.0142 -0.1411 -0.07>, rot=<-0.0217 0.0259 0.3512 -0.9357>}
 middle_03_r {parent=middle_02_r, loc=<-2.6074 -0.0534 -0.1312>, rot=<0.0083 0.0188 0.1708 -0.9851>}
 pinky_metacarpal_r {parent=hand_r, loc=<-1.7617 0.9753 -3.3814>, rot=<-0.0397 0.0814 -0.069 -0.9935>}
 pinky_01_r {parent=pinky_metacarpal_r, loc=<-7.1707 0.4832 0.1094>, rot=<-0.0654 -0.0035 0.2242 -0.9723>}
 pinky_02_r {parent=pinky_01_r, loc=<-2.8274 -0.0892 0.1043>, rot=<0.0037 0.0011 0.2854 -0.9584>}
 pinky_03_r {parent=pinky_02_r, loc=<-2.2045 -0.031 0.0078>, rot=<-0.0002 0.0098 0.1445 -0.9895>}
 ring_metacarpal_r {parent=hand_r, loc=<-2.2073 0.3627 -1.3638>, rot=<-0.0284 0.0722 -0.0458 -0.9959>}
 ring_01_r {parent=ring_metacarpal_r, loc=<-7.0297 0.6197 0.1446>, rot=<-0.0489 -0.0252 0.1892 -0.9804>}
 ring_02_r {parent=ring_01_r, loc=<-3.6172 -0.4648 0.1027>, rot=<0 0 0.2806 -0.9598>}
 ring_03_r {parent=ring_02_r, loc=<-2.4152 -0.1141 -0.0199>, rot=<0 0 0.1189 -0.9929>}
 thumb_01_r {parent=hand_r, loc=<-1.3119 0.3192 2.459>, rot=<0.6546 -0.4042 0.0664 -0.6354>}
 thumb_02_r {parent=thumb_01_r, loc=<-5.9336 0.0447 -0.0401>, rot=<-0.0037 0.0213 0.298 -0.9543>}
 thumb_03_r {parent=thumb_02_r, loc=<-2.2223 -0.3682 0.0907>, rot=<0.0118 -0.013 0.0066 -0.9998>}
 lowerarm_twist_01_r {parent=lowerarm_r, loc=<-17.855 0 0>, rot=<0 0 0 1>}
 lowerarm_twist_02_r {parent=lowerarm_r, loc=<-9.6199 0 0>, rot=<0 0 0 1>}
 upperarm_twist_01_r {parent=upperarm_r, loc=<-0.0041 0 0>, rot=<0 0 0 1>}
 upperarm_twist_02_r {parent=upperarm_r, loc=<-13.7942 0 0>, rot=<0 0 0 1>}
 dyn_fringeB_01_Base_r {parent=upperarm_r, loc=<-6.2511 9.944 -0.0604>, rot=<0.3681 0.1129 0.9226 0.0237>}
 dyn_fringeB_01_r {parent=dyn_fringeB_01_Base_r, loc=<0 0 0>, rot=<0 0 0 1>}
 dyn_fringeB_02_r {parent=dyn_fringeB_01_r, loc=<6.8323 0 0>, rot=<0.0168 -0.0204 -0.0507 -0.9984>}
 dyn_fringeG_01_Base_r {parent=upperarm_r, loc=<-7.646 -8.5125 1.3246>, rot=<0.0581 -0.925 -0.0654 0.3699>}
 dyn_fringeG_01_r {parent=dyn_fringeG_01_Base_r, loc=<0 0 0>, rot=<0 0 0 1>}
 dyn_fringeG_02_r {parent=dyn_fringeG_01_r, loc=<3.7262 0 -0.0001>, rot=<0 0 0 1>}
 dyn_fringeC_01_Base_r {parent=upperarm_r, loc=<-8.0521 8.3595 -3.4216>, rot=<0.1874 0.2556 0.9483 0.0146>}
 dyn_fringeC_01_r {parent=dyn_fringeC_01_Base_r, loc=<0 0 0>, rot=<0 0 0 -1>}
 dyn_fringeC_02_r {parent=dyn_fringeC_01_r, loc=<6.5825 -0.0002 -0.0001>, rot=<0 0 -0.0523 -0.9986>}
 dyn_fringeD_01_Base_r {parent=upperarm_r, loc=<-10.923 4.3325 -5.9247>, rot=<0.053 0.6092 0.7913 0.0027>}
 dyn_fringeD_01_r {parent=dyn_fringeD_01_Base_r, loc=<0 0 0>, rot=<0 0 0 1>}
 dyn_fringeD_02_r {parent=dyn_fringeD_01_r, loc=<3.4439 0 0>, rot=<-0.0222 -0.0152 -0.0366 -0.999>}
 dyn_fringeE_01_Base_r {parent=upperarm_r, loc=<-10.4154 -2.6992 -5.8153>, rot=<0.0186 -0.8359 -0.5467 0.045>}
 dyn_fringeE_01_r {parent=dyn_fringeE_01_Base_r, loc=<0 0 0>, rot=<0 0 0 -1>}
 dyn_fringeE_02_r {parent=dyn_fringeE_01_r, loc=<3.5736 -0.0002 0.0001>, rot=<0.0059 0.0045 -0.0047 -1>}
 dyn_fringeF_01_Base_r {parent=upperarm_r, loc=<-10.4656 -6.8331 -2.5399>, rot=<0.0875 -0.9536 -0.229 0.1747>}
 dyn_fringeF_01_r {parent=dyn_fringeF_01_Base_r, loc=<0 0 0>, rot=<0 0 0 1>}
 dyn_fringeF_02_r {parent=dyn_fringeF_01_r, loc=<3.5096 0.0001 -0.0001>, rot=<-0.0176 -0.0194 -0.0437 -0.9987>}
 dyn_mainShoulderpadArmBase_r {parent=clavicle_r, loc=<-15.5111 0 0>, rot=<0.0226 -0.3234 -0.0088 -0.9459>}
 dyn_mainShoulderpad_01_Base_r {parent=dyn_mainShoulderpadArmBase_r, loc=<6.0581 -0.1483 -8.5265>, rot=<0.5306 -0.6959 -0.2854 0.3908>}
 dyn_mainShoulderpad_01_r {parent=dyn_mainShoulderpad_01_Base_r, loc=<0 0 0>, rot=<0 0 0 -1>}
 dyn_mainShoulderpad_02_r {parent=dyn_mainShoulderpad_01_r, loc=<7.6637 0 0>, rot=<0.0075 -0.0032 -0.4052 -0.9142>}
 dyn_fringeA_01_Base_r {parent=clavicle_r, loc=<-15.8777 12.1202 3.2422>, rot=<0.6848 0.0681 0.7253 -0.019>}
 dyn_fringeA_01_r {parent=dyn_fringeA_01_Base_r, loc=<0 0 0>, rot=<0 0 0 -1>}
 dyn_fringeA_02_r {parent=dyn_fringeA_01_r, loc=<6.9915 0 0>, rot=<0.0642 0.0142 -0.0426 -0.9969>}
 dyn_fringeH_01_Base_r {parent=clavicle_r, loc=<-14.7513 -10.0439 3.5656>, rot=<-0.0012 0.7203 0.0583 -0.6912>}
 dyn_fringeH_01_r {parent=dyn_fringeH_01_Base_r, loc=<0 0 0>, rot=<0 0 0 -1>}
 dyn_fringeH_02_r {parent=dyn_fringeH_01_r, loc=<7.0436 0 0>, rot=<-0.0584 -0.0164 -0.0106 -0.9981>}
 neck_01 {parent=spine_05, loc=<11.3138 0.1529 0>, rot=<0 0 -0.2003 -0.9797>}
 neck_02 {parent=neck_01, loc=<4.5627 0 0>, rot=<0 0 0.0504 -0.9987>}
 head {parent=neck_02, loc=<5.3914 0 0>, rot=<0 0 0.115 -0.9934>}
 thigh_l {parent=pelvis, loc=<-2.5693 -0.0109 -7.7501>, rot=<0.0012 0.0572 0.0155 -0.9982>}
 calf_l {parent=thigh_l, loc=<-39.0484 0 0>, rot=<0.0006 -0.0161 -0.0391 -0.9991>}
 calf_twist_01_l {parent=calf_l, loc=<-35.8131 -0.0967 0.0397>, rot=<0.0237 0.0095 -0.0035 -0.9997>}
 calf_twist_02_l {parent=calf_l, loc=<-18.8064 -0.0967 0.0397>, rot=<0.0237 0.0095 -0.0035 -0.9997>}
 foot_l {parent=calf_l, loc=<-42.561 0.1077 0.4292>, rot=<0.0418 -0.0455 0.0343 -0.9975>}
 ball_l {parent=foot_l, loc=<-6.1599 13.8264 0.6806>, rot=<0.0001 -0.0001 -0.7063 -0.7079>}
 dyn_kneeAim_l {parent=calf_l, loc=<0 0 0>, rot=<0 0 0 1>}
 thigh_twist_01_l {parent=thigh_l, loc=<-21.6303 0 0>, rot=<0.0013 0 0 -1>}
 dyn_hipFlap_01_Base_l {parent=thigh_l, loc=<-4.0957 -1.1625 -10.2287>, rot=<0.0245 -0.7196 -0.6926 -0.0447>}
 dyn_hipFlap_01_l {parent=dyn_hipFlap_01_Base_l, loc=<0 0 0>, rot=<0 0 0 1>}
 dyn_hipFlap_02_l {parent=dyn_hipFlap_01_l, loc=<8.9487 0 0>, rot=<-0.023 0.0048 0.0848 -0.9961>}
 dyn_cutFlap_Base_l {parent=thigh_l, loc=<-6.386 9.0572 -4.5487>, rot=<0.1995 -0.0375 -0.0079 -0.9791>}
 dyn_cutFlap_l {parent=dyn_cutFlap_Base_l, loc=<0 0 0>, rot=<0 0 0 -1>}
 thigh_r {parent=pelvis, loc=<-2.5691 -0.0109 7.7501>, rot=<0.0572 -0.0012 0.9982 0.0155>}
 calf_r {parent=thigh_r, loc=<39.0487 0 0>, rot=<0.0006 -0.0161 -0.0391 -0.9991>}
 calf_twist_01_r {parent=calf_r, loc=<35.8132 0.0967 -0.0397>, rot=<0.0237 0.0095 -0.0035 -0.9997>}
 calf_twist_02_r {parent=calf_r, loc=<18.8064 0.0967 -0.0397>, rot=<0.0237 0.0095 -0.0035 -0.9997>}
 foot_r {parent=calf_r, loc=<42.5611 -0.1077 -0.4292>, rot=<0.0418 -0.0455 0.0343 -0.9975>}
 ball_r {parent=foot_r, loc=<6.1598 -13.8264 -0.6806>, rot=<0.0001 -0.0001 -0.7063 -0.7079>}
 thigh_twist_01_r {parent=thigh_r, loc=<21.6303 0 0>, rot=<0.0013 0 0 -1>}
 dyn_hipFlap_01_Base_r {parent=thigh_r, loc=<4.0959 1.1625 10.2287>, rot=<0.0245 -0.7196 -0.6926 -0.0447>}
 dyn_hipFlap_01_r {parent=dyn_hipFlap_01_Base_r, loc=<0 0 0>, rot=<0 0 -0.0322 -0.9995>}
 dyn_hipFlap_02_r {parent=dyn_hipFlap_01_r, loc=<-8.9488 0 0>, rot=<-0.023 0.0048 0.0848 -0.9961>}
 dyn_cutFlap_Base_r {parent=thigh_r, loc=<6.3861 -9.0573 4.5487>, rot=<0.1995 -0.0375 -0.0079 -0.9791>}
 dyn_cutFlap_r {parent=dyn_cutFlap_Base_r, loc=<0 0 0>, rot=<0 0 0 1>}
 dyn_loinClothFront_01_c_Base {parent=pelvis, loc=<-5.8218 11.1881 0.0322>, rot=<0 0 0.9998 0.0179>}
 dyn_loinClothFront_02_c {parent=dyn_loinClothFront_01_c_Base, loc=<0 0 0>, rot=<-0.0242 -0.0003 0.0128 -0.9996>}
 dyn_loinClothFront_03_c {parent=dyn_loinClothFront_02_c, loc=<8.3749 0.7667 -0.0321>, rot=<0.0209 0.0007 0.0078 -0.9998>}
 dyn_loinClothFront_04_c {parent=dyn_loinClothFront_03_c, loc=<8.9625 0.7094 -0.0216>, rot=<0.0142 0.0001 -0.0308 -0.9994>}
 dyn_loinClothFront_05_c {parent=dyn_loinClothFront_04_c, loc=<9.4167 0.1461 -0.0029>, rot=<0.0073 -0.0003 -0.0259 -0.9996>}
 dyn_loinClothFront_06_c {parent=dyn_loinClothFront_05_c, loc=<8.6386 -0.2588 0>, rot=<0 0 0 1>}
 dyn_loinClothBack_01_Base {parent=pelvis, loc=<-7.353 -14.2636 -0.0547>, rot=<0 0 1 0>}
 dyn_loinClothBack_01 {parent=dyn_loinClothBack_01_Base, loc=<0 0 0>, rot=<0 0 0 1>}
 dyn_loinClothBack_02 {parent=dyn_loinClothBack_01, loc=<7.8997 0 0>, rot=<-0.048 0.0003 0.0288 -0.9984>}
 dyn_loinClothBack_03 {parent=dyn_loinClothBack_02, loc=<6.4769 0 0>, rot=<0 0.0041 -0.0055 -1>}
 dyn_loinClothBack_04 {parent=dyn_loinClothBack_03, loc=<7.0532 0 0>, rot=<0 0 0 1>}
 dyn_dagger_01_Base {parent=pelvis, loc=<12.109 -5.5315 -15.4733>, rot=<-0.1862 -0.7297 -0.5855 -0.3001>}
 dyn_dagger_01 {parent=dyn_dagger_01_Base, loc=<0 0 0>, rot=<0 0 0 1>}
 dyn_dagger_02 {parent=dyn_dagger_01, loc=<5.8676 0 0>, rot=<0.098 -0.5489 0.0453 -0.8289>}
 dyn_backFringeA_01_Base {parent=pelvis, loc=<11.0883 -8.7926 -11.4288>, rot=<0.1479 -0.9467 -0.286 -0.0134>}
 dyn_backFringeA_01 {parent=dyn_backFringeA_01_Base, loc=<0 0 0>, rot=<0 0 0 -1>}
 dyn_backFringeA_02 {parent=dyn_backFringeA_01, loc=<4.6726 0 0>, rot=<-0.0041 0.0398 -0.0407 -0.9984>}
 dyn_backFringeB_01_Base {parent=pelvis, loc=<8.0468 -12.1068 -7.3817>, rot=<0.125 -0.9799 -0.1535 -0.0252>}
 dyn_backFringeB_01 {parent=dyn_backFringeB_01_Base, loc=<0 0 0>, rot=<0 0 0 -1>}
 dyn_backFringeB_02 {parent=dyn_backFringeB_01, loc=<6.095 0 0>, rot=<-0.0423 0.0115 -0.0599 -0.9972>}
 dyn_backFringeC_01_Base {parent=pelvis, loc=<7.6876 -13.1396 -2.4931>, rot=<0.1257 -0.9905 -0.0543 -0.0139>}
 dyn_backFringeC_01 {parent=dyn_backFringeC_01_Base, loc=<0 0 0>, rot=<0 0 0 1>}
 dyn_backFringeC_02 {parent=dyn_backFringeC_01, loc=<5.8186 0 0>, rot=<-0.0104 -0.001 -0.0511 -0.9986>}
 dyn_backFringeC_03 {parent=dyn_backFringeC_02, loc=<5.5026 0 0>, rot=<-0.0077 -0.0203 -0.0063 -0.9997>}
 dyn_backFringeD_01_Base {parent=pelvis, loc=<7.6876 -13.1396 2.4931>, rot=<0.1251 -0.9918 0.0197 0.0183>}
 dyn_backFringeD_01 {parent=dyn_backFringeD_01_Base, loc=<0 0 0>, rot=<0 0 0 -1>}
 dyn_backFringeD_02 {parent=dyn_backFringeD_01, loc=<5.8186 0 0>, rot=<-0.0418 0.0019 -0.0511 -0.9978>}
 dyn_backFringeD_03 {parent=dyn_backFringeD_02, loc=<5.5026 0 0>, rot=<0.0437 0.0183 -0.0207 -0.9987>}
 dyn_backFringeE_01_Base {parent=pelvis, loc=<8.0468 -12.1068 7.3817>, rot=<0.1241 -0.9846 0.1192 0.0295>}
 dyn_backFringeE_01 {parent=dyn_backFringeE_01_Base, loc=<0 0 0>, rot=<0 0 0 1>}
 dyn_backFringeE_02 {parent=dyn_backFringeE_01, loc=<6.095 0 0>, rot=<0.0099 -0.012 -0.0605 -0.998>}
 dyn_backFringeF_01_Base {parent=pelvis, loc=<11.0883 -8.7926 11.4288>, rot=<0.1481 -0.9415 0.3024 0.0108>}
 dyn_backFringeF_01 {parent=dyn_backFringeF_01_Base, loc=<0 0 0>, rot=<0 0 0 -1>}
 dyn_backFringeF_02 {parent=dyn_backFringeF_01, loc=<4.6726 0 0>, rot=<0.0041 -0.0412 -0.0393 -0.9984>}
 dyn_loinClothFront_01_Base_l {parent=pelvis, loc=<-8.9128 10.5018 -7.9768>, rot=<0.023 0.0009 0.9969 0.075>}
 dyn_loinClothFront_01_l {parent=dyn_loinClothFront_01_Base_l, loc=<0 0 0>, rot=<0 0 0 1>}
 dyn_loinClothFront_02_l {parent=dyn_loinClothFront_01_l, loc=<5.9744 0 0>, rot=<0.0016 -0.0215 -0.0418 -0.9989>}
 dyn_loinClothFront_03_l {parent=dyn_loinClothFront_02_l, loc=<8.3794 0 0>, rot=<0 0 0 1>}
 dyn_loinClothFront_04_l {parent=dyn_loinClothFront_03_l, loc=<9.4882 0.3242 1.3447>, rot=<0 0 0 1>}
 dyn_loinClothFront_05_l {parent=dyn_loinClothFront_04_l, loc=<7.9773 0.5867 0.5768>, rot=<0 0 0 1>}
 dyn_loinClothFront_01_Base_r {parent=pelvis, loc=<-8.9557 10.7913 8.0417>, rot=<-0.0251 -0.0008 0.9974 0.067>}
 dyn_loinClothFront_01_r {parent=dyn_loinClothFront_01_Base_r, loc=<0 0 0>, rot=<0 0 0 1>}
 dyn_loinClothFront_02_r {parent=dyn_loinClothFront_01_r, loc=<5.9555 0 0>, rot=<-0.0011 0.0227 -0.0483 -0.9986>}
 dyn_loinClothFront_03_r {parent=dyn_loinClothFront_02_r, loc=<8.3684 0 0>, rot=<0 0 0 1>}
 dyn_loinClothFront_04_r {parent=dyn_loinClothFront_03_r, loc=<9.4882 0.3242 -1.345>, rot=<0 0 0 1>}
 dyn_loinClothFront_05_r {parent=dyn_loinClothFront_04_r, loc=<7.9773 0.5866 -0.5002>, rot=<0 0 0 1>}
 dyn_frontStrap_01_Base_c {parent=pelvis, loc=<-6.2888 11.4456 0>, rot=<0 0 0.9986 0.0528>}
 dyn_frontStrap_01_c {parent=dyn_frontStrap_01_Base_c, loc=<0 0 0>, rot=<0 0 0 1>}
 dyn_frontStrap_02_c {parent=dyn_frontStrap_01_c, loc=<8.8597 0 0>, rot=<0 0 -0.0044 -1>}
 dyn_frontStrap_03_c {parent=dyn_frontStrap_02_c, loc=<7.799 0 0>, rot=<-0.007 0.0009 -0.0195 -0.9998>}
 dyn_frontStrap_04_c {parent=dyn_frontStrap_03_c, loc=<9.2451 0 0>, rot=<0 0 0 1>}
 dyn_frontStrap_01_Base_r {parent=pelvis, loc=<-2.2342 11.0957 4.1081>, rot=<-0.0003 0.0555 0.9976 0.0413>}
 dyn_frontStrap_01_r {parent=dyn_frontStrap_01_Base_r, loc=<0 0 0>, rot=<0 0 0 1>}
 dyn_frontStrap_02_r {parent=dyn_frontStrap_01_r, loc=<6.6199 0 0>, rot=<0.0111 0.0005 -0.0037 -0.9999>}
 dyn_frontStrap_03_r {parent=dyn_frontStrap_02_r, loc=<6.162 0 0>, rot=<0.034 -0.0014 -0.0006 -0.9994>}
 dyn_frontStrap_04_r {parent=dyn_frontStrap_03_r, loc=<8.0594 0 0>, rot=<-0.017 0 -0.0088 -0.9998>}
 dyn_frontStrap_05_r {parent=dyn_frontStrap_04_r, loc=<8.2286 0 0>, rot=<0 0 0 1>}
 dyn_frontStrap_01_Base_l {parent=pelvis, loc=<-2.2342 11.0957 -4.1081>, rot=<0.0003 -0.0555 0.9976 0.0413>}
 dyn_frontStrap_01_l {parent=dyn_frontStrap_01_Base_l, loc=<0 0 0>, rot=<0 0 0 1>}
 dyn_frontStrap_02_l {parent=dyn_frontStrap_01_l, loc=<6.6199 0 0>, rot=<-0.0111 -0.0005 -0.0037 -0.9999>}
 dyn_frontStrap_03_l {parent=dyn_frontStrap_02_l, loc=<6.162 0 0>, rot=<0.0009 0.0014 -0.0005 -1>}
 dyn_frontStrap_04_l {parent=dyn_frontStrap_03_l, loc=<8.0594 0 0>, rot=<-0.0029 -0.0006 -0.0069 -1>}
 dyn_frontStrap_05_l {parent=dyn_frontStrap_04_l, loc=<6.8198 0 0>, rot=<0 0 0 1>}
]
variables [
 crouching {type=double, default=0}
 skydivingINV {type=double, default=0}
]
