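# Physics asset for the male medium Ancient Gladiator heavy armor, covering its hanging front straps and front loincloth.
# Three strap chains and three loincloth chains of capsules descend from box base bodies jointed to the pelvis, alongside pelvis and thigh capsules.
# Most chain joints have narrow swing and twist limits, while pelvis joints are angularly free.
# The left loincloth also cross-links to the center chain, and one right-side joint is left free.
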PhysicsAsset /BRCosmetics/Characters/Player/Male/Medium/Bodies/M_MED_AncientGladiator_HArmor/Meshes/M_MED_AncientGladiator_Heavy_CRStrapPhysics
bodies {
 dyn_frontStrap_01_Base_c [{shapeType=Box, extentX=1.01, extentY=1.01, extentZ=1.01, rotation=<rot 0 -90 0>}]
 dyn_frontStrap_01_Base_l [{shapeType=Box, extentX=1.01, extentY=1.01, extentZ=1.01, rotation=<rot 0 -90 0>}]
 dyn_frontStrap_01_Base_r [{shapeType=Box, extentX=1.01, extentY=1.01, extentZ=1.01, rotation=<rot 0 -90 0>}]
 dyn_frontStrap_01_c [{shapeType=Capsule, length=8.6582, radius=1.5042, center=<3.2996 -0.785 0>, rotation=<rot 90 0 0>}]
 dyn_frontStrap_01_l [{shapeType=Capsule, length=6.4747, radius=1.5042, center=<3.2996 -0.785 0>, rotation=<rot 90 0 0>}]
 dyn_frontStrap_01_r [{shapeType=Capsule, length=6.4747, radius=1.5042, center=<3.2996 -0.785 0>, rotation=<rot 90 0 0>}]
 dyn_frontStrap_02_c [{shapeType=Capsule, length=8.6582, radius=1.5042, center=<3.2996 -0.785 0>, rotation=<rot 90 0 0>}]
 dyn_frontStrap_02_l [{shapeType=Capsule, length=6.4747, radius=1.5042, center=<3.2996 -0.785 0>, rotation=<rot 90 0 0>}]
 dyn_frontStrap_02_r [{shapeType=Capsule, length=6.4747, radius=1.5042, center=<3.2996 -0.785 0>, rotation=<rot 90 0 0>}]
 dyn_frontStrap_03_c [{shapeType=Capsule, length=8.6582, radius=1.5042, center=<3.2996 -0.785 0>, rotation=<rot 90 0 0>}]
 dyn_frontStrap_03_l [{shapeType=Capsule, length=6.4747, radius=1.5042, center=<3.2996 -0.785 0>, rotation=<rot 90 0 0>}]
 dyn_frontStrap_03_r [{shapeType=Capsule, length=6.4747, radius=1.5042, center=<3.2996 -0.785 0>, rotation=<rot 90 0 0>}]
 dyn_frontStrap_04_c [{shapeType=Capsule, length=8.6582, radius=1.5042, center=<3.2996 -0.785 0>, rotation=<rot 90 0 0>}]
 dyn_frontStrap_04_l [{shapeType=Capsule, length=6.4747, radius=1.5042, center=<3.2996 -0.785 0>, rotation=<rot 90 0 0>}]
 dyn_frontStrap_04_r [{shapeType=Capsule, length=6.4747, radius=1.5042, center=<3.2996 -0.785 0>, rotation=<rot 90 0 0>}]
 dyn_frontStrap_05_l [{shapeType=Capsule, length=6.4747, radius=1.5042, center=<3.2996 -0.785 0>, rotation=<rot 90 0 0>}]
 dyn_frontStrap_05_r [{shapeType=Capsule, length=6.4747, radius=1.5042, center=<3.2996 -0.785 0>, rotation=<rot 90 0 0>}]
 dyn_loinClothFront_01_Base_l [{shapeType=Box, extentX=6.3713, extentY=1.01, extentZ=0.8834, center=<0 2.7577 0>, rotation=<rot 0 -89.9995 -180>}]
 dyn_loinClothFront_01_Base_r [{shapeType=Box, extentX=6.3713, extentY=1.01, extentZ=0.8834, center=<0 2.7577 0>, rotation=<rot 0 -90 0>}]
 dyn_loinClothFront_01_c_Base [{shapeType=Box, extentX=1.01, extentY=1.01, extentZ=1.01, rotation=<rot 0 -90 0>}]
 dyn_loinClothFront_01_l [{shapeType=Capsule, length=12.0605, radius=5.7247, center=<3.7501 5.7179 0>}]
 dyn_loinClothFront_01_r [{shapeType=Capsule, length=11.8053, radius=5.7247, center=<3.7501 5.7179 0>, rotation=<rot 0 0 180>}]
 dyn_loinClothFront_02_c [{shapeType=Capsule, length=23.5342, radius=4.201, center=<1.6791 4.9575 -0.0001>}]
 dyn_loinClothFront_02_l [{shapeType=Capsule, length=16.3046, radius=4.8957, center=<3.7501 4.5644 0>}]
 dyn_loinClothFront_02_r [{shapeType=Capsule, length=16.0493, radius=4.8957, center=<3.7501 4.5644 0>, rotation=<rot 0 0 180>}]
 dyn_loinClothFront_03_c [{shapeType=Capsule, length=22.2235, radius=5.5117, center=<0 5.2779 0.0001>}]
 dyn_loinClothFront_03_l [{shapeType=Capsule, length=16.0407, radius=4.7984, center=<3.7501 4.5644 0>}]
 dyn_loinClothFront_03_r [{shapeType=Capsule, length=15.7854, radius=4.7984, center=<3.7501 4.5644 0>, rotation=<rot 0 0 180>}]
 dyn_loinClothFront_04_c [{shapeType=Capsule, length=16.8688, radius=5.4726, center=<0 4.4776 0>}]
 dyn_loinClothFront_04_l [{shapeType=Capsule, length=7.3948, radius=1.3977, center=<3.607 0.0721 -0.2441>, rotation=<rot 0 -88.8547 -86.1299>}]
 dyn_loinClothFront_04_r [{shapeType=Capsule, length=7.3948, radius=1.3977, center=<3.607 0.0721 -0.2441>, rotation=<rot 0 -88.8547 -86.1299>}]
 dyn_loinClothFront_05_c [{shapeType=Capsule, length=17.0393, radius=5.8253, center=<0 4.4032 0>}]
 dyn_loinClothFront_05_l [{shapeType=Capsule, length=7.3948, radius=1.3977, center=<3.607 0.0721 -0.2441>, rotation=<rot 0 -88.8547 -86.1299>}]
 dyn_loinClothFront_05_r [{shapeType=Capsule, length=7.3948, radius=1.3977, center=<3.607 0.0721 -0.2441>, rotation=<rot 0 -88.8547 -86.1299>}]
 dyn_loinClothFront_06_c [{shapeType=Capsule, length=9.8042, radius=0.6661}]
 pelvis [{shapeType=Capsule, length=10.1548, radius=11.1842, center=<-5.6844 -2.986 0>}]
 thigh_l [{shapeType=Capsule, length=28.8802, radius=9.4902, center=<-18.8949 -0.5644 0>, rotation=<rot 0 87.8329 89.9999>}]
 thigh_r [{shapeType=Capsule, length=28.8802, radius=9.3911, center=<18.8946 0.5635 0>, rotation=<rot 0 87.8328 89.9999>}]
}
constraints [
 {
  bone1Name=thigh_r
  bone2Name=pelvis
  swing1LimitDegrees=45
  swing1Motion=Free
  swing2LimitDegrees=45
  swing2Motion=Free
  twistLimitDegrees=45
  twistMotion=Free
 }
 {
  bone1Name=thigh_l
  bone2Name=pelvis
  swing1LimitDegrees=45
  swing1Motion=Free
  swing2LimitDegrees=45
  swing2Motion=Free
  twistLimitDegrees=45
  twistMotion=Free
 }
 {
  bone1Name=dyn_loinClothFront_01_c_Base
  bone2Name=pelvis
  swing1LimitDegrees=45
  swing1Motion=Free
  swing2LimitDegrees=45
  swing2Motion=Free
  twistLimitDegrees=45
  twistMotion=Free
 }
 {bone1Name=dyn_loinClothFront_02_c, bone2Name=dyn_loinClothFront_01_c_Base, swing1LimitDegrees=16, swing2LimitDegrees=16, twistLimitDegrees=4}
 {bone1Name=dyn_loinClothFront_03_c, bone2Name=dyn_loinClothFront_02_c, swing1LimitDegrees=16, swing2LimitDegrees=16, twistLimitDegrees=4}
 {bone1Name=dyn_loinClothFront_04_c, bone2Name=dyn_loinClothFront_03_c, swing1LimitDegrees=16, swing2LimitDegrees=16, twistLimitDegrees=4}
 {bone1Name=dyn_loinClothFront_05_c, bone2Name=dyn_loinClothFront_04_c, swing1LimitDegrees=16, swing2LimitDegrees=18, twistLimitDegrees=4}
 {bone1Name=dyn_loinClothFront_06_c, bone2Name=dyn_loinClothFront_05_c, swing1LimitDegrees=16, swing2LimitDegrees=18, twistLimitDegrees=4}
 {bone1Name=dyn_loinClothFront_02_l, bone2Name=dyn_loinClothFront_01_l, swing1LimitDegrees=8, swing2LimitDegrees=8, twistLimitDegrees=4}
 {bone1Name=dyn_loinClothFront_03_l, bone2Name=dyn_loinClothFront_02_l, swing1LimitDegrees=8, swing2LimitDegrees=8, twistLimitDegrees=4}
 {
  bone1Name=dyn_loinClothFront_01_Base_r
  bone2Name=pelvis
  swing1LimitDegrees=45
  swing1Motion=Free
  swing2LimitDegrees=45
  swing2Motion=Free
  twistLimitDegrees=45
  twistMotion=Free
 }
 {
  bone1Name=dyn_loinClothFront_01_Base_l
  bone2Name=pelvis
  swing1LimitDegrees=45
  swing1Motion=Free
  swing2LimitDegrees=45
  swing2Motion=Free
  twistLimitDegrees=45
  twistMotion=Free
 }
 {bone1Name=dyn_loinClothFront_01_l, bone2Name=dyn_loinClothFront_01_Base_l, swing1LimitDegrees=8, swing2LimitDegrees=8, twistLimitDegrees=4}
 {
  bone1Name=dyn_loinClothFront_03_l
  bone2Name=dyn_loinClothFront_04_c
  swing1LimitDegrees=45
  swing1Motion=Free
  swing2LimitDegrees=45
  swing2Motion=Free
  twistLimitDegrees=45
  twistMotion=Free
 }
 {
  bone1Name=dyn_frontStrap_01_Base_c
  bone2Name=pelvis
  swing1LimitDegrees=45
  swing1Motion=Free
  swing2LimitDegrees=45
  swing2Motion=Free
  twistLimitDegrees=45
  twistMotion=Free
 }
 {
  bone1Name=dyn_frontStrap_01_Base_r
  bone2Name=pelvis
  swing1LimitDegrees=45
  swing1Motion=Free
  swing2LimitDegrees=45
  swing2Motion=Free
  twistLimitDegrees=45
  twistMotion=Free
 }
 {
  bone1Name=dyn_frontStrap_01_Base_l
  bone2Name=pelvis
  swing1LimitDegrees=45
  swing1Motion=Free
  swing2LimitDegrees=45
  swing2Motion=Free
  twistLimitDegrees=45
  twistMotion=Free
 }
 {bone1Name=dyn_frontStrap_01_c, bone2Name=dyn_frontStrap_01_Base_c, swing1LimitDegrees=12, swing2LimitDegrees=12, twistLimitDegrees=4}
 {bone1Name=dyn_frontStrap_02_c, bone2Name=dyn_frontStrap_01_c, swing1LimitDegrees=12, swing2LimitDegrees=12, twistLimitDegrees=4}
 {bone1Name=dyn_frontStrap_03_c, bone2Name=dyn_frontStrap_02_c, swing1LimitDegrees=12, swing2LimitDegrees=12, twistLimitDegrees=4}
 {bone1Name=dyn_frontStrap_04_c, bone2Name=dyn_frontStrap_03_c, swing1LimitDegrees=12, swing2LimitDegrees=12, twistLimitDegrees=4}
 {bone1Name=dyn_frontStrap_01_r, bone2Name=dyn_frontStrap_01_Base_r, swing1LimitDegrees=12, swing2LimitDegrees=12, twistLimitDegrees=4}
 {bone1Name=dyn_frontStrap_02_r, bone2Name=dyn_frontStrap_01_r, swing1LimitDegrees=20, swing2LimitDegrees=14, twistLimitDegrees=4}
 {bone1Name=dyn_frontStrap_03_r, bone2Name=dyn_frontStrap_02_r, swing1LimitDegrees=20, swing2LimitDegrees=14, twistLimitDegrees=4}
 {bone1Name=dyn_frontStrap_04_r, bone2Name=dyn_frontStrap_03_r, swing1LimitDegrees=20, swing2LimitDegrees=14, twistLimitDegrees=4}
 {bone1Name=dyn_frontStrap_05_r, bone2Name=dyn_frontStrap_04_r, swing1LimitDegrees=20, swing2LimitDegrees=14, twistLimitDegrees=4}
 {bone1Name=dyn_frontStrap_01_l, bone2Name=dyn_frontStrap_01_Base_l, swing1LimitDegrees=10, swing2LimitDegrees=8, twistLimitDegrees=4}
 {bone1Name=dyn_frontStrap_02_l, bone2Name=dyn_frontStrap_01_l, swing1LimitDegrees=20, swing2LimitDegrees=14, twistLimitDegrees=4}
 {bone1Name=dyn_frontStrap_03_l, bone2Name=dyn_frontStrap_02_l, swing1LimitDegrees=20, swing2LimitDegrees=14, twistLimitDegrees=4}
 {bone1Name=dyn_frontStrap_04_l, bone2Name=dyn_frontStrap_03_l, swing1LimitDegrees=20, swing2LimitDegrees=14, twistLimitDegrees=4}
 {bone1Name=dyn_frontStrap_05_l, bone2Name=dyn_frontStrap_04_l, swing1LimitDegrees=20, swing2LimitDegrees=14, twistLimitDegrees=4}
 {bone1Name=dyn_loinClothFront_02_r, bone2Name=dyn_loinClothFront_01_r, swing1LimitDegrees=8, swing2LimitDegrees=8, twistLimitDegrees=4}
 {
  bone1Name=dyn_loinClothFront_03_r
  bone2Name=dyn_loinClothFront_02_r
  swing1LimitDegrees=45
  swing1Motion=Free
  swing2LimitDegrees=45
  swing2Motion=Free
  twistLimitDegrees=45
  twistMotion=Free
 }
 {bone1Name=dyn_loinClothFront_01_r, bone2Name=dyn_loinClothFront_01_Base_r, swing1LimitDegrees=8, swing2LimitDegrees=8, twistLimitDegrees=4}
 {bone1Name=dyn_loinClothFront_04_l, bone2Name=dyn_loinClothFront_03_l, swing1LimitDegrees=8, swing2LimitDegrees=8, twistLimitDegrees=4}
 {bone1Name=dyn_loinClothFront_05_l, bone2Name=dyn_loinClothFront_04_l, swing1LimitDegrees=8, swing2LimitDegrees=8, twistLimitDegrees=4}
 {bone1Name=dyn_loinClothFront_04_r, bone2Name=dyn_loinClothFront_03_r, swing1LimitDegrees=8, swing2LimitDegrees=8, twistLimitDegrees=4}
 {bone1Name=dyn_loinClothFront_05_r, bone2Name=dyn_loinClothFront_04_r, swing1LimitDegrees=8, swing2LimitDegrees=8, twistLimitDegrees=4}
]
