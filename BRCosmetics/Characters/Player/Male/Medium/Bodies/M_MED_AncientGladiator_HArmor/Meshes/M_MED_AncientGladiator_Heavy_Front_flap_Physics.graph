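# Physics asset for the front loincloth and cut flaps on the male medium Ancient Gladiator heavy armor, giving them secondary motion.
# Box bodies form center, left and right loincloth chains hung from pelvis-linked bases, plus one flap per side hung from thigh-linked bases.
# Chain links have tight swing and twist limits, while base links and cross-links from the side tips to the center chain are free.
# Capsules on pelvis, spine and thighs approximate the body.
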PhysicsAsset /BRCosmetics/Characters/Player/Male/Medium/Bodies/M_MED_AncientGladiator_HArmor/Meshes/M_MED_AncientGladiator_Heavy_Front_flap_Physics
bodies {
 dyn_cutFlap_Base_l [{shapeType=Box, extentX=1.01, extentY=1.01, extentZ=1.01, rotation=<rot 0 -90 0>}]
 dyn_cutFlap_Base_r [{shapeType=Box, extentX=1.01, extentY=1.01, extentZ=1.01, rotation=<rot 0 -90 0>}]
 dyn_cutFlap_l [{shapeType=Box, extentX=2.636, extentY=10.1257, extentZ=9.8212, center=<-3.9303 0 0>, rotation=<rot 0 90 0>}]
 dyn_cutFlap_r [{shapeType=Box, extentX=2.636, extentY=10.1257, extentZ=9.8212, center=<3.9303 0 0>, rotation=<rot 0 90 0>}]
 dyn_loinClothFront_01_Base_l [{shapeType=Box, extentX=1.01, extentY=1.01, extentZ=1.01, rotation=<rot 0 -90 180>}]
 dyn_loinClothFront_01_Base_r [{shapeType=Box, extentX=1.01, extentY=1.01, extentZ=1.01, rotation=<rot 0 -90 0>}]
 dyn_loinClothFront_01_c_Base [{shapeType=Box, extentX=1.01, extentY=1.01, extentZ=1.01, rotation=<rot 0 -90 0>}]
 dyn_loinClothFront_01_l [{shapeType=Box, extentX=2.636, extentY=7.3398, extentZ=9.8212, center=<2.66 0 0>, rotation=<rot 0 89.9999 -180>}]
 dyn_loinClothFront_01_r [{shapeType=Box, extentX=2.636, extentY=7.7796, extentZ=9.8212, center=<2.7383 -0.0001 0>, rotation=<rot 0 90 0>}]
 dyn_loinClothFront_02_c [{shapeType=Box, extentX=2.636, extentY=10.1257, extentZ=13.4194, center=<3.9303 0 0>, rotation=<rot 0 90 0>}]
 dyn_loinClothFront_02_l [{shapeType=Box, extentX=2.636, extentY=10.1257, extentZ=9.8212, center=<3.9303 0 0>, rotation=<rot 0 90 0>}]
 dyn_loinClothFront_02_r [{shapeType=Box, extentX=3.0564, extentY=10.1257, extentZ=9.8212, center=<3.9303 0 0>, rotation=<rot 0 90 0>}]
 dyn_loinClothFront_03_c [{shapeType=Box, extentX=0.9061, extentY=10.14, extentZ=13.3311, center=<4.4019 0 0>, rotation=<rot 0 90 0>}]
 dyn_loinClothFront_03_l [{shapeType=Box, extentX=0.9061, extentY=10.14, extentZ=5.6593, center=<4.4019 0 0>, rotation=<rot 0 90 -0.1246>}]
 dyn_loinClothFront_03_r [{shapeType=Box, extentX=1.4257, extentY=10.14, extentZ=5.6593, center=<4.4019 0 0>, rotation=<rot 0 90 -0.1246>}]
 dyn_loinClothFront_04_c [{shapeType=Box, extentX=1.9564, extentY=10.5962, extentZ=13.8782, center=<4.5513 0.0001 0>, rotation=<rot 0 90 0>}]
 dyn_loinClothFront_05_c [{shapeType=Box, extentX=2.0125, extentY=8.8308, extentZ=13.6581, center=<4.2912 0.0003 0>, rotation=<rot 0 90 0>}]
 dyn_loinClothFront_06_c [{shapeType=Box, extentX=0.949, extentY=6.8986, extentZ=10.8906, center=<2.8585 0.0001 0>, rotation=<rot 0 90 0>}]
 pelvis [
  {shapeType=Capsule, length=10.8314, radius=10.5076, center=<-5.1316 -2.7219 0>}
  {shapeType=Capsule, length=11.0763, radius=5.0861, center=<4.8024 -8.4961 0.0001>}
 ]
 spine_03 [{shapeType=Capsule, length=15.2435, radius=14.2512, center=<0.6364 1.0803 0.5679>, rotation=<rot 90 0 0>}]
 thigh_l [{shapeType=Capsule, length=29.211, radius=9.1594, center=<-18.8949 -0.5644 0>, rotation=<rot 0 87.8329 89.9999>}]
 thigh_r [{shapeType=Capsule, length=29.211, radius=9.0603, center=<18.8946 0.5635 0>, rotation=<rot 0 87.8328 89.9999>}]
}
constraints [
 {
  bone1Name=thigh_r
  bone2Name=pelvis
  swing1LimitDegrees=45
  swing1Motion=Free
  swing2LimitDegrees=45
  swing2Motion=Free
  twistLimitDegrees=45
  twistMotion=Free
 }
 {
  bone1Name=thigh_l
  bone2Name=pelvis
  swing1LimitDegrees=45
  swing1Motion=Free
  swing2LimitDegrees=45
  swing2Motion=Free
  twistLimitDegrees=45
  twistMotion=Free
 }
 {
  bone1Name=dyn_loinClothFront_01_c_Base
  bone2Name=pelvis
  swing1LimitDegrees=45
  swing1Motion=Free
  swing2LimitDegrees=45
  swing2Motion=Free
  twistLimitDegrees=45
  twistMotion=Free
 }
 {bone1Name=dyn_loinClothFront_01_r, bone2Name=dyn_loinClothFront_01_Base_r, swing1LimitDegrees=8, swing2LimitDegrees=8, twistLimitDegrees=4}
 {bone1Name=dyn_loinClothFront_02_r, bone2Name=dyn_loinClothFront_01_r, swing1LimitDegrees=8, swing2LimitDegrees=8, twistLimitDegrees=4}
 {bone1Name=dyn_loinClothFront_03_r, bone2Name=dyn_loinClothFront_02_r, swing1LimitDegrees=8, swing2LimitDegrees=8, twistLimitDegrees=4}
 {bone1Name=dyn_loinClothFront_02_c, bone2Name=dyn_loinClothFront_01_c_Base, swing1LimitDegrees=16, swing2LimitDegrees=16, twistLimitDegrees=4}
 {bone1Name=dyn_loinClothFront_03_c, bone2Name=dyn_loinClothFront_02_c, swing1LimitDegrees=16, swing2LimitDegrees=16, twistLimitDegrees=4}
 {bone1Name=dyn_loinClothFront_04_c, bone2Name=dyn_loinClothFront_03_c, swing1LimitDegrees=16, swing2LimitDegrees=16, twistLimitDegrees=4}
 {bone1Name=dyn_loinClothFront_05_c, bone2Name=dyn_loinClothFront_04_c, swing1LimitDegrees=16, swing2LimitDegrees=18, twistLimitDegrees=4}
 {bone1Name=dyn_loinClothFront_06_c, bone2Name=dyn_loinClothFront_05_c, swing1LimitDegrees=16, swing2LimitDegrees=18, twistLimitDegrees=4}
 {bone1Name=dyn_loinClothFront_02_l, bone2Name=dyn_loinClothFront_01_l, swing1LimitDegrees=16, swing2LimitDegrees=16, twistLimitDegrees=4}
 {bone1Name=dyn_loinClothFront_03_l, bone2Name=dyn_loinClothFront_02_l, swing1LimitDegrees=16, swing2LimitDegrees=16, twistLimitDegrees=4}
 {
  bone1Name=dyn_cutFlap_Base_l
  bone2Name=thigh_l
  swing1LimitDegrees=45
  swing1Motion=Free
  swing2LimitDegrees=45
  swing2Motion=Free
  twistLimitDegrees=45
  twistMotion=Free
 }
 {bone1Name=dyn_cutFlap_l, bone2Name=dyn_cutFlap_Base_l, swing1LimitDegrees=12, swing2LimitDegrees=14, twistLimitDegrees=4}
 {
  bone1Name=dyn_cutFlap_Base_r
  bone2Name=thigh_r
  swing1LimitDegrees=45
  swing1Motion=Free
  swing2LimitDegrees=45
  swing2Motion=Free
  twistLimitDegrees=45
  twistMotion=Free
 }
 {bone1Name=dyn_cutFlap_r, bone2Name=dyn_cutFlap_Base_r, swing1LimitDegrees=12, swing2LimitDegrees=14, twistLimitDegrees=4}
 {
  bone1Name=dyn_loinClothFront_01_Base_r
  bone2Name=pelvis
  swing1LimitDegrees=45
  swing1Motion=Free
  swing2LimitDegrees=45
  swing2Motion=Free
  twistLimitDegrees=45
  twistMotion=Free
 }
 {
  bone1Name=dyn_loinClothFront_01_Base_l
  bone2Name=pelvis
  swing1LimitDegrees=45
  swing1Motion=Free
  swing2LimitDegrees=45
  swing2Motion=Free
  twistLimitDegrees=45
  twistMotion=Free
 }
 {bone1Name=dyn_loinClothFront_01_l, bone2Name=dyn_loinClothFront_01_Base_l, swing1LimitDegrees=16, swing2LimitDegrees=16, twistLimitDegrees=4}
 {
  bone1Name=dyn_loinClothFront_03_r
  bone2Name=dyn_loinClothFront_04_c
  swing1LimitDegrees=45
  swing1Motion=Free
  swing2LimitDegrees=45
  swing2Motion=Free
  twistLimitDegrees=45
  twistMotion=Free
 }
 {
  bone1Name=dyn_loinClothFront_03_l
  bone2Name=dyn_loinClothFront_04_c
  swing1LimitDegrees=45
  swing1Motion=Free
  swing2LimitDegrees=45
  swing2Motion=Free
  twistLimitDegrees=45
  twistMotion=Free
 }
 {
  bone1Name=spine_03
  bone2Name=pelvis
  swing1LimitDegrees=45
  swing1Motion=Free
  swing2LimitDegrees=45
  swing2Motion=Free
  twistLimitDegrees=45
  twistMotion=Free
 }
]
