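# Physics asset for the male medium Ancient Gladiator heavy armor, covering its dangling parts: six back fringes, rear loincloth, two hip flaps and a dagger.
# Each part is a chain of dyn_ bodies with tight swing and twist limits, the dagger's first joint also locking one swing axis.
# Chains start at Base bodies with free angular constraints to the pelvis, or thighs for hip flaps.
# Capsules on pelvis, spine_03 and thighs approximate the body.
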PhysicsAsset /BRCosmetics/Characters/Player/Male/Medium/Bodies/M_MED_AncientGladiator_HArmor/Meshes/M_MED_AncientGladiator_Heavy_Physics
bodies {
 dyn_backFringeA_01 [{shapeType=Box, extentX=1.0133, extentY=4.3246, extentZ=4.5788, center=<2.4569 0 0.0005>, rotation=<rot 0 90 90>}]
 dyn_backFringeA_01_Base [{shapeType=Box, extentX=1.01, extentY=1.01, extentZ=1.01, rotation=<rot 0 -90 0>}]
 dyn_backFringeA_02 [{shapeType=Box, extentX=1.0133, extentY=4.3246, extentZ=8.9135, center=<4.2608 0.0001 0.0005>, rotation=<rot 0 90 90>}]
 dyn_backFringeB_01 [{shapeType=Box, extentX=1.1004, extentY=4.2772, extentZ=6.9737, center=<3.1594 0 0.0005>, rotation=<rot 0 90 90>}]
 dyn_backFringeB_01_Base [{shapeType=Box, extentX=1.01, extentY=1.01, extentZ=1.01, rotation=<rot 0 -90 0>}]
 dyn_backFringeB_02 [{shapeType=Box, extentX=1.1004, extentY=4.2772, extentZ=8.4107, center=<3.9687 0 0.0005>, rotation=<rot 0 90 90>}]
 dyn_backFringeC_01 [{shapeType=Box, extentX=1.9581, extentY=4.4391, extentZ=6.4948, center=<2.9558 0 0>, rotation=<rot 0 90 90>}]
 dyn_backFringeC_01_Base [{shapeType=Box, extentX=1.01, extentY=1.01, extentZ=1.01, rotation=<rot 0 -90 0>}]
 dyn_backFringeC_02 [{shapeType=Box, extentX=1.7239, extentY=4.4391, extentZ=6.1151, center=<2.7151 0 0>, rotation=<rot 0 90 90>}]
 dyn_backFringeC_03 [{shapeType=Box, extentX=1.7239, extentY=4.4391, extentZ=6.4948, center=<2.9558 0 0>, rotation=<rot 0 90 90>}]
 dyn_backFringeD_01 [{shapeType=Box, extentX=1.5261, extentY=4.2275, extentZ=6.7272, center=<2.9224 0 0>, rotation=<rot 0 90 90>}]
 dyn_backFringeD_01_Base [{shapeType=Box, extentX=1.01, extentY=1.01, extentZ=1.01, rotation=<rot 0 -90 0>}]
 dyn_backFringeD_02 [{shapeType=Box, extentX=1.5261, extentY=4.2275, extentZ=6.3662, center=<3.0466 0 0>, rotation=<rot 0 90 90>}]
 dyn_backFringeD_03 [{shapeType=Box, extentX=1.5261, extentY=4.2275, extentZ=7.3481, center=<3.3747 0 0>, rotation=<rot 0 90 90>}]
 dyn_backFringeE_01 [{shapeType=Box, extentX=1.1606, extentY=4.5243, extentZ=6.3544, center=<3.0321 0 0.0005>, rotation=<rot 0 90 90>}]
 dyn_backFringeE_01_Base [{shapeType=Box, extentX=1.01, extentY=1.01, extentZ=1.01, rotation=<rot 0 -90 0>}]
 dyn_backFringeE_02 [{shapeType=Box, extentX=1.1606, extentY=4.5243, extentZ=7.9961, center=<3.6772 0 0.0005>, rotation=<rot 0 90 90>}]
 dyn_backFringeF_01 [{shapeType=Box, extentX=1.0133, extentY=4.1751, extentZ=4.9628, center=<2.6187 0 0.0005>, rotation=<rot 0 90 90>}]
 dyn_backFringeF_01_Base [{shapeType=Box, extentX=1.01, extentY=1.01, extentZ=1.01, rotation=<rot 0 -90 0>}]
 dyn_backFringeF_02 [{shapeType=Box, extentX=1.0133, extentY=4.1751, extentZ=7.0956, center=<3.585 0.0002 0.0005>, rotation=<rot 0 90 90>}]
 dyn_dagger_01 [{shapeType=Capsule, length=6.5991, radius=0.505, center=<2.8871 0 0>, rotation=<rot 90 0 0>}]
 dyn_dagger_01_Base [{shapeType=Capsule, length=15.3416, radius=0.505, center=<-0.8237 2.5979 -0.4778>, rotation=<rot 0 -90 0>}]
 dyn_dagger_02 [{shapeType=Capsule, length=38.502, radius=4.9662, center=<4.879 -3.4941 -0.3724>, rotation=<rot 0 91.9365 -89.0175>}]
 dyn_hipFlap_01_Base_l [{shapeType=Box, extentX=0.5001, extentY=1.01, extentZ=1.01, rotation=<rot 0 -90 0>}]
 dyn_hipFlap_01_Base_r [{shapeType=Box, extentX=0.5001, extentY=1.01, extentZ=1.01, rotation=<rot 0 -90 0>}]
 dyn_hipFlap_01_l [{shapeType=Box, extentX=3.6615, extentY=9.6216, extentZ=20.7574, center=<4.4311 0 0.0001>, rotation=<rot 1.781 89.9999 -180>}]
 dyn_hipFlap_01_r [{shapeType=Box, extentX=3.6615, extentY=9.6216, extentZ=20.7574, center=<-4.4311 0 -0.0001>, rotation=<rot 1.781 89.9999 0>}]
 dyn_hipFlap_02_l [{shapeType=Box, extentX=2.3127, extentY=12.1622, extentZ=15.7931, center=<5.7458 -0.0002 0>, rotation=<rot 0 -90 180>}]
 dyn_hipFlap_02_r [{shapeType=Box, extentX=2.3127, extentY=12.1622, extentZ=15.7931, center=<-5.7458 0.0002 0>, rotation=<rot 0 -90 0>}]
 dyn_loinClothBack_01 [{shapeType=Box, extentX=2.7058, extentY=8.0176, extentZ=26.8089, center=<4.0654 0 0>, rotation=<rot 0 -91.4988 0>}]
 dyn_loinClothBack_01_Base [{shapeType=Box, extentX=1.01, extentY=1.01, extentZ=1.01, rotation=<rot 0 -90 0>}]
 dyn_loinClothBack_02 [{shapeType=Box, extentX=3.3936, extentY=8.7153, extentZ=28.2157, center=<2.7992 -1.3425 0>, rotation=<rot 5.5043 -90 -0.12>}]
 dyn_loinClothBack_03 [{shapeType=Box, extentX=3.3816, extentY=8.2636, extentZ=20.3034, center=<3.4395 -0.7894 0.0958>, rotation=<rot -6.9178 89.9997 0>}]
 dyn_loinClothBack_04 [{shapeType=Box, extentX=2.4842, extentY=10.9321, extentZ=13.6015, center=<4.711 -0.5652 0.0386>, rotation=<rot 5.5054 -89.9995 0.4092>}]
 pelvis [
  {shapeType=Capsule, length=10.8314, radius=10.5076, center=<-5.1316 -2.7219 0>}
  {shapeType=Capsule, length=11.0763, radius=5.0861, center=<4.8024 -8.4961 0.0001>}
 ]
 spine_03 [{shapeType=Capsule, length=15.2435, radius=14.2512, center=<0.6364 1.0803 0.5679>, rotation=<rot 90 0 0>}]
 thigh_l [{shapeType=Capsule, length=29.211, radius=9.1594, center=<-18.8949 -0.5644 0>, rotation=<rot 0 87.8329 89.9999>}]
 thigh_r [{shapeType=Capsule, length=29.211, radius=9.0603, center=<18.8946 0.5635 0>, rotation=<rot 0 87.8328 89.9999>}]
}
constraints [
 {
  bone1Name=dyn_loinClothBack_01_Base
  bone2Name=pelvis
  swing1LimitDegrees=45
  swing1Motion=Free
  swing2LimitDegrees=45
  swing2Motion=Free
  twistLimitDegrees=45
  twistMotion=Free
 }
 {bone1Name=dyn_loinClothBack_01, bone2Name=dyn_loinClothBack_01_Base, swing1LimitDegrees=20, swing2LimitDegrees=30, twistLimitDegrees=12}
 {bone1Name=dyn_loinClothBack_02, bone2Name=dyn_loinClothBack_01, swing1LimitDegrees=20, swing2LimitDegrees=30, twistLimitDegrees=12}
 {bone1Name=dyn_loinClothBack_03, bone2Name=dyn_loinClothBack_02, swing1LimitDegrees=20, swing2LimitDegrees=30, twistLimitDegrees=12}
 {bone1Name=dyn_loinClothBack_04, bone2Name=dyn_loinClothBack_03, swing1LimitDegrees=20, swing2LimitDegrees=30, twistLimitDegrees=12}
 {
  bone1Name=dyn_hipFlap_01_Base_l
  bone2Name=thigh_l
  swing1LimitDegrees=45
  swing1Motion=Free
  swing2LimitDegrees=45
  swing2Motion=Free
  twistLimitDegrees=45
  twistMotion=Free
 }
 {
  bone1Name=dyn_hipFlap_01_Base_r
  bone2Name=thigh_r
  swing1LimitDegrees=45
  swing1Motion=Free
  swing2LimitDegrees=45
  swing2Motion=Free
  twistLimitDegrees=45
  twistMotion=Free
 }
 {bone1Name=dyn_hipFlap_01_r, bone2Name=dyn_hipFlap_01_Base_r, swing1LimitDegrees=30, swing2LimitDegrees=12, twistLimitDegrees=8}
 {bone1Name=dyn_hipFlap_02_r, bone2Name=dyn_hipFlap_01_r, swing1LimitDegrees=30, swing2LimitDegrees=12, twistLimitDegrees=8}
 {bone1Name=dyn_hipFlap_01_l, bone2Name=dyn_hipFlap_01_Base_l, swing1LimitDegrees=30, swing2LimitDegrees=12, twistLimitDegrees=8}
 {bone1Name=dyn_hipFlap_02_l, bone2Name=dyn_hipFlap_01_l, swing1LimitDegrees=30, swing2LimitDegrees=12, twistLimitDegrees=8}
 {
  bone1Name=thigh_r
  bone2Name=pelvis
  swing1LimitDegrees=45
  swing1Motion=Free
  swing2LimitDegrees=45
  swing2Motion=Free
  twistLimitDegrees=45
  twistMotion=Free
 }
 {
  bone1Name=thigh_l
  bone2Name=pelvis
  swing1LimitDegrees=45
  swing1Motion=Free
  swing2LimitDegrees=45
  swing2Motion=Free
  twistLimitDegrees=45
  twistMotion=Free
 }
 {
  bone1Name=dyn_dagger_01_Base
  bone2Name=pelvis
  swing1LimitDegrees=45
  swing1Motion=Free
  swing2LimitDegrees=45
  swing2Motion=Free
  twistLimitDegrees=45
  twistMotion=Free
 }
 {bone1Name=dyn_dagger_01, bone2Name=dyn_dagger_01_Base, swing1LimitDegrees=12, swing1Motion=Locked, swing2LimitDegrees=8, twistLimitDegrees=8}
 {bone1Name=dyn_dagger_02, bone2Name=dyn_dagger_01, swing1LimitDegrees=12, swing2LimitDegrees=12, twistLimitDegrees=8}
 {
  bone1Name=dyn_backFringeA_01_Base
  bone2Name=pelvis
  swing1LimitDegrees=45
  swing1Motion=Free
  swing2LimitDegrees=45
  swing2Motion=Free
  twistLimitDegrees=45
  twistMotion=Free
 }
 {bone1Name=dyn_backFringeA_01, bone2Name=dyn_backFringeA_01_Base, swing1LimitDegrees=16, swing2LimitDegrees=12, twistLimitDegrees=4}
 {bone1Name=dyn_backFringeA_02, bone2Name=dyn_backFringeA_01, swing1LimitDegrees=12, swing2LimitDegrees=12, twistLimitDegrees=4}
 {
  bone1Name=dyn_backFringeB_01_Base
  bone2Name=pelvis
  swing1LimitDegrees=45
  swing1Motion=Free
  swing2LimitDegrees=45
  swing2Motion=Free
  twistLimitDegrees=45
  twistMotion=Free
 }
 {bone1Name=dyn_backFringeB_01, bone2Name=dyn_backFringeB_01_Base, swing1LimitDegrees=16, swing2LimitDegrees=12, twistLimitDegrees=4}
 {bone1Name=dyn_backFringeB_02, bone2Name=dyn_backFringeB_01, swing1LimitDegrees=12, swing2LimitDegrees=12, twistLimitDegrees=4}
 {
  bone1Name=dyn_backFringeC_01_Base
  bone2Name=pelvis
  swing1LimitDegrees=45
  swing1Motion=Free
  swing2LimitDegrees=45
  swing2Motion=Free
  twistLimitDegrees=45
  twistMotion=Free
 }
 {bone1Name=dyn_backFringeC_01, bone2Name=dyn_backFringeC_01_Base, swing1LimitDegrees=16, swing2LimitDegrees=12, twistLimitDegrees=4}
 {bone1Name=dyn_backFringeC_02, bone2Name=dyn_backFringeC_01, swing1LimitDegrees=12, swing2LimitDegrees=12, twistLimitDegrees=4}
 {bone1Name=dyn_backFringeC_03, bone2Name=dyn_backFringeC_02, swing1LimitDegrees=12, swing2LimitDegrees=12, twistLimitDegrees=4}
 {
  bone1Name=dyn_backFringeD_01_Base
  bone2Name=pelvis
  swing1LimitDegrees=45
  swing1Motion=Free
  swing2LimitDegrees=45
  swing2Motion=Free
  twistLimitDegrees=45
  twistMotion=Free
 }
 {bone1Name=dyn_backFringeD_01, bone2Name=dyn_backFringeD_01_Base, swing1LimitDegrees=16, swing2LimitDegrees=12, twistLimitDegrees=4}
 {bone1Name=dyn_backFringeD_02, bone2Name=dyn_backFringeD_01, swing1LimitDegrees=12, swing2LimitDegrees=12, twistLimitDegrees=4}
 {bone1Name=dyn_backFringeD_03, bone2Name=dyn_backFringeD_02, swing1LimitDegrees=12, swing2LimitDegrees=12, twistLimitDegrees=4}
 {
  bone1Name=dyn_backFringeE_01_Base
  bone2Name=pelvis
  swing1LimitDegrees=45
  swing1Motion=Free
  swing2LimitDegrees=45
  swing2Motion=Free
  twistLimitDegrees=45
  twistMotion=Free
 }
 {bone1Name=dyn_backFringeE_01, bone2Name=dyn_backFringeE_01_Base, swing1LimitDegrees=16, swing2LimitDegrees=12, twistLimitDegrees=4}
 {bone1Name=dyn_backFringeE_02, bone2Name=dyn_backFringeE_01, swing1LimitDegrees=12, swing2LimitDegrees=12, twistLimitDegrees=4}
 {
  bone1Name=dyn_backFringeF_01_Base
  bone2Name=pelvis
  swing1LimitDegrees=45
  swing1Motion=Free
  swing2LimitDegrees=45
  swing2Motion=Free
  twistLimitDegrees=45
  twistMotion=Free
 }
 {bone1Name=dyn_backFringeF_01, bone2Name=dyn_backFringeF_01_Base, swing1LimitDegrees=16, swing2LimitDegrees=12, twistLimitDegrees=4}
 {bone1Name=dyn_backFringeF_02, bone2Name=dyn_backFringeF_01, swing1LimitDegrees=12, swing2LimitDegrees=12, twistLimitDegrees=4}
 {
  bone1Name=spine_03
  bone2Name=pelvis
  swing1LimitDegrees=45
  swing1Motion=Free
  swing2LimitDegrees=45
  swing2Motion=Free
  twistLimitDegrees=45
  twistMotion=Free
 }
]
