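# Physics asset for the male medium Ancient Gladiator heavy armor body, setting up secondary motion for straps, loincloth and fur.
# Pelvis and thigh capsules form the body shapes, with base bones hung on free constraints from pelvis or clavicles.
# Three front straps and three loincloth strips are capsule chains with mostly tight swing limits and limited or locked twist.
# Fur tufts swing narrowly, twist-locked; a free constraint also links left loincloth to center.
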PhysicsAsset /BRCosmetics/Characters/Player/Male/Medium/Bodies/M_MED_AncientGladiator_HArmor/Meshes/M_MED_AncientGladiator_Heavy_StrapPhysics
bodies {
 clavicle_l [{shapeType=Capsule, length=1.1, radius=1.1}]
 clavicle_r [{shapeType=Capsule, length=1.1, radius=1.1}]
 dyn_backFurA [{shapeType=Capsule, length=3.1927, radius=2.4189, center=<2.2981 -0.0001 0>, rotation=<rot 90 0 0>}]
 dyn_backFurB [{shapeType=Capsule, length=1.5758, radius=1.3468, center=<2.2981 -0.0001 0>, rotation=<rot 90 0 0>}]
 dyn_backFurC [{shapeType=Capsule, length=2.5915, radius=1.5741, center=<1.4909 -0.0001 0>, rotation=<rot 90 0 0>}]
 dyn_backFurD [{shapeType=Capsule, length=3.1927, radius=2.4189, center=<2.2981 -0.0001 0>, rotation=<rot 90 0 0>}]
 dyn_backFurE [{shapeType=Capsule, length=1.8067, radius=1.2271, center=<2.2981 -0.0001 0>, rotation=<rot 90 0 0>}]
 dyn_backFurF [{shapeType=Capsule, length=1.3393, radius=1.3129, center=<0.9229 -0.0001 0>, rotation=<rot 90 0 0>}]
 dyn_backFurG [{shapeType=Capsule, length=1.6688, radius=1.5703, center=<2.2981 -0.0001 0>, rotation=<rot 90 0 0>}]
 dyn_backFurH [{shapeType=Capsule, length=1.6688, radius=1.5703, center=<2.2981 -0.0001 0>, rotation=<rot 90 0 0>}]
 dyn_backFurI [{shapeType=Capsule, length=3.1927, radius=2.4189, center=<2.2981 -0.0001 0>, rotation=<rot 90 0 0>}]
 dyn_backFurJ [{shapeType=Capsule, length=3.5989, radius=1.4445, center=<1.2369 -0.0001 0>, rotation=<rot 90 0 0>}]
 dyn_backFurK [{shapeType=Capsule, length=2.7979, radius=1.7033, center=<1.4631 0.0001 0>, rotation=<rot 90 0 0>}]
 dyn_backFurMain_01 [{shapeType=Box, extentX=1.1, extentY=1.1, extentZ=1.1}]
 dyn_backFurMain_02 [{shapeType=Box, extentX=1.1, extentY=1.1, extentZ=1.1}]
 dyn_frontStrap_01_Base_c [{shapeType=Box, extentX=1.01, extentY=1.01, extentZ=1.01, rotation=<rot 0 -90 0>}]
 dyn_frontStrap_01_Base_l [{shapeType=Box, extentX=1.01, extentY=1.01, extentZ=1.01, rotation=<rot 0 -90 0>}]
 dyn_frontStrap_01_Base_r [{shapeType=Box, extentX=1.01, extentY=1.01, extentZ=1.01, rotation=<rot 0 -90 0>}]
 dyn_frontStrap_01_c [{shapeType=Capsule, length=8.6582, radius=1.5042, center=<3.2996 -0.785 0>, rotation=<rot 90 0 0>}]
 dyn_frontStrap_01_l [{shapeType=Capsule, length=6.4747, radius=1.5042, center=<3.2996 -0.785 0>, rotation=<rot 90 0 0>}]
 dyn_frontStrap_01_r [{shapeType=Capsule, length=6.4747, radius=1.5042, center=<3.2996 -0.785 0>, rotation=<rot 90 0 0>}]
 dyn_frontStrap_02_c [{shapeType=Capsule, length=8.6582, radius=1.5042, center=<3.2996 -0.785 0>, rotation=<rot 90 0 0>}]
 dyn_frontStrap_02_l [{shapeType=Capsule, length=6.4747, radius=1.5042, center=<3.2996 -0.785 0>, rotation=<rot 90 0 0>}]
 dyn_frontStrap_02_r [{shapeType=Capsule, length=6.4747, radius=1.5042, center=<3.2996 -0.785 0>, rotation=<rot 90 0 0>}]
 dyn_frontStrap_03_c [{shapeType=Capsule, length=8.6582, radius=1.5042, center=<3.2996 -0.785 0>, rotation=<rot 90 0 0>}]
 dyn_frontStrap_03_l [{shapeType=Capsule, length=6.4747, radius=1.5042, center=<3.2996 -0.785 0>, rotation=<rot 90 0 0>}]
 dyn_frontStrap_03_r [{shapeType=Capsule, length=6.4747, radius=1.5042, center=<3.2996 -0.785 0>, rotation=<rot 90 0 0>}]
 dyn_frontStrap_04_c [{shapeType=Capsule, length=8.6582, radius=1.5042, center=<3.2996 -0.785 0>, rotation=<rot 90 0 0>}]
 dyn_frontStrap_04_l [{shapeType=Capsule, length=6.4747, radius=1.5042, center=<3.2996 -0.785 0>, rotation=<rot 90 0 0>}]
 dyn_frontStrap_04_r [{shapeType=Capsule, length=6.4747, radius=1.5042, center=<3.2996 -0.785 0>, rotation=<rot 90 0 0>}]
 dyn_frontStrap_05_l [{shapeType=Capsule, length=6.4747, radius=1.5042, center=<3.2996 -0.785 0>, rotation=<rot 90 0 0>}]
 dyn_frontStrap_05_r [{shapeType=Capsule, length=6.4747, radius=1.5042, center=<3.2996 -0.785 0>, rotation=<rot 90 0 0>}]
 dyn_loinClothFront_01_Base_l [{shapeType=Box, extentX=6.3713, extentY=1.01, extentZ=0.8834, center=<0 2.7577 0>, rotation=<rot 0 -89.9995 -180>}]
 dyn_loinClothFront_01_Base_r [{shapeType=Box, extentX=6.3713, extentY=1.01, extentZ=0.8834, center=<0 2.7577 0>, rotation=<rot 0 -90 0>}]
 dyn_loinClothFront_01_c_Base [{shapeType=Box, extentX=1.01, extentY=1.01, extentZ=1.01, rotation=<rot 0 -90 0>}]
 dyn_loinClothFront_01_l [{shapeType=Capsule, length=12.0605, radius=5.7247, center=<3.7501 5.7179 0>}]
 dyn_loinClothFront_01_r [{shapeType=Capsule, length=11.8053, radius=5.7247, center=<3.7501 5.7179 0>, rotation=<rot 0 0 180>}]
 dyn_loinClothFront_02_c [{shapeType=Capsule, length=23.5342, radius=4.201, center=<1.6791 4.9575 -0.0001>}]
 dyn_loinClothFront_02_l [{shapeType=Capsule, length=16.3046, radius=4.8957, center=<3.7501 4.5644 0>}]
 dyn_loinClothFront_02_r [{shapeType=Capsule, length=16.0493, radius=4.8957, center=<3.7501 4.5644 0>, rotation=<rot 0 0 180>}]
 dyn_loinClothFront_03_c [{shapeType=Capsule, length=22.2235, radius=5.5117, center=<0 5.2779 0.0001>}]
 dyn_loinClothFront_03_l [{shapeType=Capsule, length=16.0407, radius=4.7984, center=<3.7501 4.5644 0>}]
 dyn_loinClothFront_03_r [{shapeType=Capsule, length=15.7854, radius=4.7984, center=<3.7501 4.5644 0>, rotation=<rot 0 0 180>}]
 dyn_loinClothFront_04_c [{shapeType=Capsule, length=16.8688, radius=5.4726, center=<0 4.4776 0>}]
 dyn_loinClothFront_04_l [{shapeType=Capsule, length=7.3948, radius=1.3977, center=<3.607 0.0721 -0.2441>, rotation=<rot 0 -88.8547 -86.1299>}]
 dyn_loinClothFront_04_r [{shapeType=Capsule, length=7.3948, radius=1.3977, center=<3.607 0.0721 -0.2441>, rotation=<rot 0 -88.8547 -86.1299>}]
 dyn_loinClothFront_05_c [{shapeType=Capsule, length=17.0393, radius=5.8253, center=<0 4.4032 0>}]
 dyn_loinClothFront_05_l [{shapeType=Capsule, length=7.3948, radius=1.3977, center=<3.607 0.0721 -0.2441>, rotation=<rot 0 -88.8547 -86.1299>}]
 dyn_loinClothFront_05_r [{shapeType=Capsule, length=7.3948, radius=1.3977, center=<3.607 0.0721 -0.2441>, rotation=<rot 0 -88.8547 -86.1299>}]
 dyn_loinClothFront_06_c [{shapeType=Capsule, length=9.8042, radius=0.6661}]
 dyn_mainShoulderpadArmBase_l [{shapeType=Capsule, length=1.1, radius=1.1}]
 dyn_mainShoulderpadArmBase_r [{shapeType=Capsule, length=1.1, radius=1.1}]
 dyn_shoulderFurBack_l [{shapeType=Capsule, length=1.7839, radius=1.7764, center=<2.2981 -0.0001 0>, rotation=<rot 90 0 0>}]
 dyn_shoulderFurBack_r [{shapeType=Capsule, length=1.7839, radius=1.7764, center=<2.2981 -0.0001 0>, rotation=<rot 90 0 0>}]
 dyn_shoulderFurFront_l [{shapeType=Capsule, length=5.0044, radius=1.2709, center=<1.5859 -0.0001 0>, rotation=<rot 90 0 0>}]
 dyn_shoulderFurFront_r [{shapeType=Capsule, length=5.0044, radius=1.2709, center=<1.5859 -0.0001 0>, rotation=<rot 90 0 0>}]
 dyn_shoulderFurMid_l [{shapeType=Capsule, length=4.1165, radius=1.5578, center=<1.625 -0.0001 0>, rotation=<rot 90 0 0>}]
 dyn_shoulderFurMid_r [{shapeType=Capsule, length=4.1165, radius=1.5578, center=<1.625 -0.0001 0>, rotation=<rot 90 0 0>}]
 pelvis [{shapeType=Capsule, length=10.1548, radius=11.1842, center=<-5.6844 -2.986 0>}]
 thigh_l [{shapeType=Capsule, length=28.8802, radius=9.4902, center=<-18.8949 -0.5644 0>, rotation=<rot 0 87.8329 89.9999>}]
 thigh_r [{shapeType=Capsule, length=28.8802, radius=9.3911, center=<18.8946 0.5635 0>, rotation=<rot 0 87.8328 89.9999>}]
}
constraints [
 {
  bone1Name=thigh_r
  bone2Name=pelvis
  swing1LimitDegrees=45
  swing1Motion=Free
  swing2LimitDegrees=45
  swing2Motion=Free
  twistLimitDegrees=45
  twistMotion=Free
 }
 {
  bone1Name=thigh_l
  bone2Name=pelvis
  swing1LimitDegrees=45
  swing1Motion=Free
  swing2LimitDegrees=45
  swing2Motion=Free
  twistLimitDegrees=45
  twistMotion=Free
 }
 {
  bone1Name=dyn_loinClothFront_01_c_Base
  bone2Name=pelvis
  swing1LimitDegrees=45
  swing1Motion=Free
  swing2LimitDegrees=45
  swing2Motion=Free
  twistLimitDegrees=45
  twistMotion=Free
 }
 {bone1Name=dyn_loinClothFront_02_c, bone2Name=dyn_loinClothFront_01_c_Base, swing1LimitDegrees=16, swing2LimitDegrees=16, twistLimitDegrees=4}
 {bone1Name=dyn_loinClothFront_03_c, bone2Name=dyn_loinClothFront_02_c, swing1LimitDegrees=16, swing2LimitDegrees=16, twistLimitDegrees=4}
 {bone1Name=dyn_loinClothFront_04_c, bone2Name=dyn_loinClothFront_03_c, swing1LimitDegrees=16, swing2LimitDegrees=16, twistLimitDegrees=4}
 {bone1Name=dyn_loinClothFront_05_c, bone2Name=dyn_loinClothFront_04_c, swing1LimitDegrees=16, swing2LimitDegrees=18, twistLimitDegrees=4}
 {bone1Name=dyn_loinClothFront_06_c, bone2Name=dyn_loinClothFront_05_c, swing1LimitDegrees=16, swing2LimitDegrees=18, twistLimitDegrees=4}
 {bone1Name=dyn_loinClothFront_02_l, bone2Name=dyn_loinClothFront_01_l, swing1LimitDegrees=8, swing2LimitDegrees=8, twistLimitDegrees=4}
 {bone1Name=dyn_loinClothFront_03_l, bone2Name=dyn_loinClothFront_02_l, swing1LimitDegrees=8, swing2LimitDegrees=8, twistLimitDegrees=4}
 {
  bone1Name=dyn_loinClothFront_01_Base_r
  bone2Name=pelvis
  swing1LimitDegrees=45
  swing1Motion=Free
  swing2LimitDegrees=45
  swing2Motion=Free
  twistLimitDegrees=45
  twistMotion=Free
 }
 {
  bone1Name=dyn_loinClothFront_01_Base_l
  bone2Name=pelvis
  swing1LimitDegrees=45
  swing1Motion=Free
  swing2LimitDegrees=45
  swing2Motion=Free
  twistLimitDegrees=45
  twistMotion=Free
 }
 {bone1Name=dyn_loinClothFront_01_l, bone2Name=dyn_loinClothFront_01_Base_l, swing1LimitDegrees=8, swing2LimitDegrees=8, twistLimitDegrees=4}
 {
  bone1Name=dyn_loinClothFront_03_l
  bone2Name=dyn_loinClothFront_04_c
  swing1LimitDegrees=45
  swing1Motion=Free
  swing2LimitDegrees=45
  swing2Motion=Free
  twistLimitDegrees=45
  twistMotion=Free
 }
 {
  bone1Name=dyn_frontStrap_01_Base_c
  bone2Name=pelvis
  swing1LimitDegrees=45
  swing1Motion=Free
  swing2LimitDegrees=45
  swing2Motion=Free
  twistLimitDegrees=45
  twistMotion=Free
 }
 {
  bone1Name=dyn_frontStrap_01_Base_r
  bone2Name=pelvis
  swing1LimitDegrees=45
  swing1Motion=Free
  swing2LimitDegrees=45
  swing2Motion=Free
  twistLimitDegrees=45
  twistMotion=Free
 }
 {
  bone1Name=dyn_frontStrap_01_Base_l
  bone2Name=pelvis
  swing1LimitDegrees=45
  swing1Motion=Free
  swing2LimitDegrees=45
  swing2Motion=Free
  twistLimitDegrees=45
  twistMotion=Free
 }
 {
  bone1Name=dyn_frontStrap_01_c
  bone2Name=dyn_frontStrap_01_Base_c
  swing1LimitDegrees=12
  swing2LimitDegrees=12
  twistLimitDegrees=4
  twistMotion=Locked
 }
 {
  bone1Name=dyn_frontStrap_02_c
  bone2Name=dyn_frontStrap_01_c
  swing1LimitDegrees=12
  swing2LimitDegrees=12
  twistLimitDegrees=4
  twistMotion=Locked
 }
 {
  bone1Name=dyn_frontStrap_03_c
  bone2Name=dyn_frontStrap_02_c
  swing1LimitDegrees=12
  swing2LimitDegrees=12
  twistLimitDegrees=4
  twistMotion=Locked
 }
 {
  bone1Name=dyn_frontStrap_04_c
  bone2Name=dyn_frontStrap_03_c
  swing1LimitDegrees=12
  swing2LimitDegrees=12
  twistLimitDegrees=4
  twistMotion=Locked
 }
 {bone1Name=dyn_frontStrap_01_r, bone2Name=dyn_frontStrap_01_Base_r, swing1LimitDegrees=12, swing2LimitDegrees=12, twistLimitDegrees=4}
 {
  bone1Name=dyn_frontStrap_02_r
  bone2Name=dyn_frontStrap_01_r
  swing1LimitDegrees=20
  swing2LimitDegrees=14
  twistLimitDegrees=4
  twistMotion=Locked
 }
 {
  bone1Name=dyn_frontStrap_03_r
  bone2Name=dyn_frontStrap_02_r
  swing1LimitDegrees=20
  swing2LimitDegrees=14
  twistLimitDegrees=4
  twistMotion=Locked
 }
 {
  bone1Name=dyn_frontStrap_04_r
  bone2Name=dyn_frontStrap_03_r
  swing1LimitDegrees=20
  swing2LimitDegrees=14
  twistLimitDegrees=4
  twistMotion=Locked
 }
 {
  bone1Name=dyn_frontStrap_05_r
  bone2Name=dyn_frontStrap_04_r
  swing1LimitDegrees=20
  swing2LimitDegrees=14
  twistLimitDegrees=4
  twistMotion=Locked
 }
 {bone1Name=dyn_frontStrap_01_l, bone2Name=dyn_frontStrap_01_Base_l, swing1LimitDegrees=10, swing2LimitDegrees=8, twistLimitDegrees=4}
 {
  bone1Name=dyn_frontStrap_02_l
  bone2Name=dyn_frontStrap_01_l
  swing1LimitDegrees=20
  swing2LimitDegrees=14
  twistLimitDegrees=4
  twistMotion=Locked
 }
 {
  bone1Name=dyn_frontStrap_03_l
  bone2Name=dyn_frontStrap_02_l
  swing1LimitDegrees=20
  swing2LimitDegrees=14
  twistLimitDegrees=4
  twistMotion=Locked
 }
 {
  bone1Name=dyn_frontStrap_04_l
  bone2Name=dyn_frontStrap_03_l
  swing1LimitDegrees=20
  swing2LimitDegrees=14
  twistLimitDegrees=4
  twistMotion=Locked
 }
 {
  bone1Name=dyn_frontStrap_05_l
  bone2Name=dyn_frontStrap_04_l
  swing1LimitDegrees=20
  swing2LimitDegrees=14
  twistLimitDegrees=4
  twistMotion=Locked
 }
 {bone1Name=dyn_loinClothFront_02_r, bone2Name=dyn_loinClothFront_01_r, swing1LimitDegrees=8, swing2LimitDegrees=8, twistLimitDegrees=4}
 {
  bone1Name=dyn_loinClothFront_03_r
  bone2Name=dyn_loinClothFront_02_r
  swing1LimitDegrees=45
  swing1Motion=Free
  swing2LimitDegrees=45
  swing2Motion=Free
  twistLimitDegrees=45
  twistMotion=Free
 }
 {bone1Name=dyn_loinClothFront_01_r, bone2Name=dyn_loinClothFront_01_Base_r, swing1LimitDegrees=8, swing2LimitDegrees=8, twistLimitDegrees=4}
 {
  bone1Name=dyn_backFurMain_02
  bone2Name=dyn_backFurMain_01
  swing1LimitDegrees=45
  swing1Motion=Free
  swing2LimitDegrees=45
  swing2Motion=Free
  twistLimitDegrees=45
  twistMotion=Free
 }
 {
  bone1Name=dyn_backFurMain_01
  bone2Name=pelvis
  swing1LimitDegrees=45
  swing1Motion=Free
  swing2LimitDegrees=45
  swing2Motion=Free
  twistLimitDegrees=45
  twistMotion=Free
 }
 {bone1Name=dyn_backFurE, bone2Name=dyn_backFurMain_01, swing1LimitDegrees=12, swing2LimitDegrees=12, twistLimitDegrees=4, twistMotion=Locked}
 {bone1Name=dyn_backFurG, bone2Name=dyn_backFurMain_02, swing1LimitDegrees=12, swing2LimitDegrees=12, twistLimitDegrees=4, twistMotion=Locked}
 {bone1Name=dyn_backFurI, bone2Name=dyn_backFurMain_02, swing1LimitDegrees=12, swing2LimitDegrees=12, twistLimitDegrees=4, twistMotion=Locked}
 {bone1Name=dyn_backFurJ, bone2Name=dyn_backFurMain_02, swing1LimitDegrees=12, swing2LimitDegrees=12, twistLimitDegrees=4, twistMotion=Locked}
 {bone1Name=dyn_backFurK, bone2Name=dyn_backFurMain_02, swing1LimitDegrees=12, swing2LimitDegrees=12, twistLimitDegrees=4, twistMotion=Locked}
 {bone1Name=dyn_backFurB, bone2Name=dyn_backFurMain_01, swing1LimitDegrees=12, swing2LimitDegrees=12, twistLimitDegrees=4, twistMotion=Locked}
 {bone1Name=dyn_backFurC, bone2Name=dyn_backFurMain_01, swing1LimitDegrees=12, swing2LimitDegrees=12, twistLimitDegrees=4, twistMotion=Locked}
 {bone1Name=dyn_backFurF, bone2Name=dyn_backFurMain_01, swing1LimitDegrees=12, swing2LimitDegrees=12, twistLimitDegrees=4, twistMotion=Locked}
 {
  bone1Name=clavicle_l
  bone2Name=pelvis
  swing1LimitDegrees=45
  swing1Motion=Free
  swing2LimitDegrees=45
  swing2Motion=Free
  twistLimitDegrees=45
  twistMotion=Free
 }
 {
  bone1Name=dyn_mainShoulderpadArmBase_l
  bone2Name=clavicle_l
  swing1LimitDegrees=45
  swing1Motion=Free
  swing2LimitDegrees=45
  swing2Motion=Free
  twistLimitDegrees=45
  twistMotion=Free
 }
 {
  bone1Name=clavicle_r
  bone2Name=pelvis
  swing1LimitDegrees=45
  swing1Motion=Free
  swing2LimitDegrees=45
  swing2Motion=Free
  twistLimitDegrees=45
  twistMotion=Free
 }
 {
  bone1Name=dyn_mainShoulderpadArmBase_r
  bone2Name=clavicle_r
  swing1LimitDegrees=45
  swing1Motion=Free
  swing2LimitDegrees=45
  swing2Motion=Free
  twistLimitDegrees=45
  twistMotion=Free
 }
 {
  bone1Name=dyn_shoulderFurBack_r
  bone2Name=dyn_mainShoulderpadArmBase_r
  swing1LimitDegrees=12
  swing2LimitDegrees=12
  twistLimitDegrees=4
  twistMotion=Locked
 }
 {
  bone1Name=dyn_shoulderFurMid_r
  bone2Name=dyn_mainShoulderpadArmBase_r
  swing1LimitDegrees=12
  swing2LimitDegrees=12
  twistLimitDegrees=4
  twistMotion=Locked
 }
 {
  bone1Name=dyn_shoulderFurFront_r
  bone2Name=dyn_mainShoulderpadArmBase_r
  swing1LimitDegrees=12
  swing2LimitDegrees=12
  twistLimitDegrees=4
  twistMotion=Locked
 }
 {bone1Name=dyn_backFurD, bone2Name=clavicle_r, swing1LimitDegrees=12, swing2LimitDegrees=12, twistLimitDegrees=4, twistMotion=Locked}
 {bone1Name=dyn_backFurA, bone2Name=clavicle_l, swing1LimitDegrees=12, swing2LimitDegrees=12, twistLimitDegrees=4, twistMotion=Locked}
 {
  bone1Name=dyn_shoulderFurBack_l
  bone2Name=dyn_mainShoulderpadArmBase_l
  swing1LimitDegrees=12
  swing2LimitDegrees=12
  twistLimitDegrees=4
  twistMotion=Locked
 }
 {
  bone1Name=dyn_shoulderFurMid_l
  bone2Name=dyn_mainShoulderpadArmBase_l
  swing1LimitDegrees=12
  swing2LimitDegrees=12
  twistLimitDegrees=4
  twistMotion=Locked
 }
 {
  bone1Name=dyn_shoulderFurFront_l
  bone2Name=dyn_mainShoulderpadArmBase_l
  swing1LimitDegrees=12
  swing2LimitDegrees=12
  twistLimitDegrees=4
  twistMotion=Locked
 }
 {bone1Name=dyn_backFurH, bone2Name=dyn_backFurMain_02, swing1LimitDegrees=12, swing2LimitDegrees=12, twistLimitDegrees=4, twistMotion=Locked}
 {
  bone1Name=dyn_loinClothFront_04_l
  bone2Name=dyn_loinClothFront_03_l
  swing1LimitDegrees=8
  swing1Motion=Locked
  swing2LimitDegrees=8
  twistLimitDegrees=4
  twistMotion=Locked
 }
 {
  bone1Name=dyn_loinClothFront_05_l
  bone2Name=dyn_loinClothFront_04_l
  swing1LimitDegrees=8
  swing2LimitDegrees=8
  twistLimitDegrees=4
  twistMotion=Locked
 }
 {
  bone1Name=dyn_loinClothFront_04_r
  bone2Name=dyn_loinClothFront_03_r
  swing1LimitDegrees=8
  swing1Motion=Locked
  swing2LimitDegrees=8
  twistLimitDegrees=4
  twistMotion=Locked
 }
 {
  bone1Name=dyn_loinClothFront_05_r
  bone2Name=dyn_loinClothFront_04_r
  swing1LimitDegrees=8
  swing2LimitDegrees=8
  twistLimitDegrees=4
  twistMotion=Locked
 }
]
