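# Physics asset for the upper-body mesh of the male medium Ancient Gladiator heavy-armor cosmetic, defining physics for its back fur piece.
# A capsule on spine_05 anchors a chain of two box bodies on the dyn_backFurMain bones.
# Each link is constrained to its parent with narrow angular limits, widest on swing 1, so the fur sways mainly along one axis with little twist.
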PhysicsAsset /BRCosmetics/Characters/Player/Male/Medium/Bodies/M_MED_AncientGladiator_HArmor/Meshes/M_MED_AncientGladiator_Upper_Dyn
bodies {
 dyn_backFurMain_01 [{shapeType=Box, extentX=6.4178, extentY=11.7512, extentZ=22.8605, center=<4.5448 -3.0231 -2.0305>, rotation=<rot 0 89.9999 -6.4992>}]
 dyn_backFurMain_02 [{shapeType=Box, extentX=5.6479, extentY=6.8191, extentZ=20.2428, center=<6.6996 -2.2776 -0.4893>, rotation=<rot -3.0237 89.9999 8.3207>}]
 spine_05 [{shapeType=Capsule, length=15.0246, radius=13.8597, center=<0 -0.8319 0>, rotation=<rot 0 90 0>}]
}
constraints [
 {bone1Name=dyn_backFurMain_01, bone2Name=spine_05, swing1LimitDegrees=20, swing2LimitDegrees=4, twistLimitDegrees=4}
 {bone1Name=dyn_backFurMain_02, bone2Name=dyn_backFurMain_01, swing1LimitDegrees=20, swing2LimitDegrees=4, twistLimitDegrees=4}
]
