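# Secondary-motion AnimBP for the male medium Ancient Gladiator heavy armor cosmetic, animating straps, fur and fringe.
# The parent mesh pose passes through a Control Rig, pelvis RigidBody simulation and chained per-bone AnimDynamics springs, with spherical limits keeping fringe off upper arms.
# Crouching constrains front straps mostly to loincloth bones and switches a pose blend involving strap physics.
# Sprinting lowers the alpha of pelvis and spine RigidBody simulations.
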AnimBlueprint /BRCosmetics/Characters/Player/Male/Medium/Bodies/M_MED_AncientGladiator_HArmor/Meshes/M_MED_Ancient_Gladiator_Heavy_Armor_AnimBP
target_skeleton=&/M_MED_Ancient_Gladiator_Heavy_Armor_Skeleton
graphs [
 AnimGraph {nodes [
   n1 = RigidBody(ComponentPose=Local To Component(LocalPose=Use cached pose 'preSecondaryRigid'())) {
    bClampLinearTranslationLimitToRefPose=true
    bOverrideWorldGravity=true
    overridePhysicsAsset=&/M_MED_AncientGladiator_Heavy_StrapPhysics
    simulationSpace=BaseBoneSpace
    alphaBoolBlend {blendOption=Linear}
    baseBoneRef {boneName=root}
    componentAppliedLinearAccClamp=<2000 2000 1500>
    componentLinearAccScale=<0.35 0.35 0.25>
    componentLinearVelScale=<0.1 0.1 0.1>
    overrideWorldGravity {z=-450}
   }
   Blend Poses by bool(BlendPose_1=Component To Local(ComponentPose=@n1), bActiveValue=Get bIsCrouching()) {blendType=Linear}
   n2 = Get isCrouching() -> (isCrouching:double)
   n3 = "Constraint - dyn_frontStrap_01_c (P)"(Alpha=@n2, ConstraintWeights_0=0) {
    alphaBoolBlend {blendOption=Linear}
    boneToModify {boneName=dyn_frontStrap_01_c}
    constraintSetup [{targetBone {boneName=dyn_loinClothFront_02_c}}]
    constraintWeights [1]
   }
   n4 = "Constraint - dyn_frontStrap_02_c (P)"(ComponentPose=@n3, Alpha=@n2, ConstraintWeights_0=1) {
    alphaBoolBlend {blendOption=Linear}
    boneToModify {boneName=dyn_frontStrap_02_c}
    constraintSetup [{targetBone {boneName=dyn_loinClothFront_03_c}}]
    constraintWeights [1]
   }
   n5 = "Constraint - dyn_frontStrap_03_c (P)"(ComponentPose=@n4, Alpha=@n2, ConstraintWeights_0=1) {
    alphaBoolBlend {blendOption=Linear}
    boneToModify {boneName=dyn_frontStrap_03_c}
    constraintSetup [{targetBone {boneName=dyn_loinClothFront_04_c}}]
    constraintWeights [1]
   }
   n6 = "Constraint - dyn_frontStrap_01_r (P)"(ComponentPose=@n5, Alpha=@n2, ConstraintWeights_0=1) {
    alphaBoolBlend {blendOption=Linear}
    boneToModify {boneName=dyn_frontStrap_01_r}
    constraintSetup [{targetBone {boneName=dyn_loinClothFront_01_r}}]
    constraintWeights [1]
   }
   n7 = "Constraint - dyn_frontStrap_02_r (P)"(ComponentPose=@n6, Alpha=@n2, ConstraintWeights_0=1) {
    alphaBoolBlend {blendOption=Linear}
    boneToModify {boneName=dyn_frontStrap_02_r}
    constraintSetup [{targetBone {boneName=dyn_loinClothFront_02_r}}]
    constraintWeights [1]
   }
   n8 = "Constraint - dyn_frontStrap_03_r (P)"(ComponentPose=@n7, Alpha=@n2, ConstraintWeights_0=1) {
    alphaBoolBlend {blendOption=Linear}
    boneToModify {boneName=dyn_frontStrap_03_r}
    constraintSetup [{targetBone {boneName=dyn_loinClothFront_03_r}}]
    constraintWeights [1]
   }
   n9 = "Constraint - dyn_frontStrap_04_r (P)"(ComponentPose=@n8, Alpha=@n2, ConstraintWeights_0=1) {
    alphaBoolBlend {blendOption=Linear}
    boneToModify {boneName=dyn_frontStrap_04_r}
    constraintSetup [{targetBone {boneName=dyn_frontStrap_04_r}}]
    constraintWeights [1]
   }
   n10 = "Constraint - dyn_frontStrap_05_r (P)"(ComponentPose=@n9, Alpha=@n2, ConstraintWeights_0=1) {
    alphaBoolBlend {blendOption=Linear}
    boneToModify {boneName=dyn_frontStrap_05_r}
    constraintSetup [{targetBone {boneName=dyn_frontStrap_05_r}}]
    constraintWeights [1]
   }
   n11 = "Constraint - dyn_frontStrap_01_l (P)"(ComponentPose=@n10, Alpha=@n2, ConstraintWeights_0=1) {
    alphaBoolBlend {blendOption=Linear}
    boneToModify {boneName=dyn_frontStrap_01_l}
    constraintSetup [{targetBone {boneName=dyn_loinClothFront_02_c}}]
    constraintWeights [1]
   }
   n12 = "Constraint - dyn_frontStrap_02_l (P)"(ComponentPose=@n11, Alpha=@n2, ConstraintWeights_0=1) {
    alphaBoolBlend {blendOption=Linear}
    boneToModify {boneName=dyn_frontStrap_02_l}
    constraintSetup [{targetBone {boneName=dyn_loinClothFront_02_c}}]
    constraintWeights [1]
   }
   n13 = "Constraint - dyn_frontStrap_03_l (P)"(ComponentPose=@n12, Alpha=@n2, ConstraintWeights_0=1) {
    alphaBoolBlend {blendOption=Linear}
    boneToModify {boneName=dyn_frontStrap_03_l}
    constraintSetup [{targetBone {boneName=dyn_loinClothFront_03_c}}]
    constraintWeights [1]
   }
   n14 = "Constraint - dyn_frontStrap_04_l (P)"(ComponentPose=@n13, Alpha=@n2, ConstraintWeights_0=1) {
    alphaBoolBlend {blendOption=Linear}
    boneToModify {boneName=dyn_frontStrap_04_l}
    constraintSetup [{targetBone {boneName=dyn_loinClothFront_04_c}}]
    constraintWeights [1]
   }
   "Constraint - dyn_frontStrap_05_l (P)"(ComponentPose=@n14, Alpha=@n2, ConstraintWeights_0=1) {
    alphaBoolBlend {blendOption=Linear}
    boneToModify {boneName=dyn_frontStrap_05_l}
    constraintSetup [{targetBone {boneName=dyn_loinClothFront_05_c}}]
    constraintWeights [1]
   }
   n15 = "Transform (Modify) Bone - Bone: dyn_frontStrap_01_Base_r"(Alpha=1, Translation="0,0,0", Rotation="0,70,0", Scale="1,1,1") {
    rotationMode=BMM_Additive
    rotationSpace=BCS_BoneSpace
    alphaBoolBlend {blendOption=Linear}
    boneToModify {boneName=dyn_frontStrap_01_Base_r}
    rotation {yaw=70}
   }
   "Transform (Modify) Bone - Bone: dyn_frontStrap_01_r"(ComponentPose=@n15, Alpha=1, Translation="0,0,0", Rotation="0,-70,0", Scale="1,1,1") {
    rotationMode=BMM_Additive
    rotationSpace=BCS_BoneSpace
    alphaBoolBlend {blendOption=Linear}
    boneToModify {boneName=dyn_frontStrap_01_r}
    rotation {yaw=-70}
   }
   n16 = Get isSprinting() -> (isSprinting:double)
   RigidBody(Alpha=@n16) {
    bClampLinearTranslationLimitToRefPose=true
    bOverrideWorldGravity=true
    overridePhysicsAsset=&/M_MED_AncientGladiator_Upper_Dyn
    simulationSpace=BaseBoneSpace
    alphaBoolBlend {blendOption=Linear}
    alphaScaleBiasClamp {bMapRange=true, outRange {max=0.65, min=1}}
    baseBoneRef {boneName=spine_05}
    componentAppliedLinearAccClamp=<2000 2000 3000>
    componentLinearAccScale=<0.2 0.2 0.1>
    componentLinearVelScale=<0.1 0.1 0.1>
    overrideWorldGravity {z=-980}
    simSpaceSettings {maxAngularAcceleration=0, maxAngularVelocity=0, maxLinearAcceleration=500, maxLinearVelocity=50, velocityScaleZ=0, worldAlpha=1}
   }
   n17 = Control Rig!(
    Source=Copy Pose From Mesh() {bUseAttachedParent=true}
    Alpha
    crouching:double=Get isCrouching()
    skydivingINV:double=Get isSkydivingINV() -> (isSkydivingINV:double)
   ) -> (Pose) {
    alphaBoolBlend {blendOption=Linear}
    controlRigAssetReference {blueprintRigClass=&/M_MED_AncientGladiator_H_CtrlRig.M_MED_AncientGladiator_H_CtrlRig_C}
   }
   n18 = RigidBody(ComponentPose=Local To Component(LocalPose=@n17), Alpha=@n16) {
    bClampLinearTranslationLimitToRefPose=true
    bOverrideWorldGravity=true
    overridePhysicsAsset=&/M_MED_AncientGladiator_Heavy_Physics
    simulationSpace=BaseBoneSpace
    alphaBoolBlend {blendOption=Linear}
    alphaScaleBiasClamp {bMapRange=true, outRange {max=0.65, min=1}}
    baseBoneRef {boneName=pelvis}
    componentAppliedLinearAccClamp=<2000 2000 3000>
    componentLinearAccScale=<0.2 0.2 0.1>
    componentLinearVelScale=<0.1 0.1 0.1>
    overrideWorldGravity {z=-980}
    simSpaceSettings {maxAngularAcceleration=0, maxAngularVelocity=0, maxLinearAcceleration=500, maxLinearVelocity=50, velocityScaleZ=0, worldAlpha=1}
   }
   n19 = Anim Dynamics - Bone: dyn_backFurMain_01(ComponentPose=@n18, Alpha=@n16, ExternalForce="0,0,0") {
    angularDampingOverride=0.9999
    angularSpringConstant=50
    bAngularSpring=true
    bGravityOverrideInSimSpace=true
    bOverrideAngularDamping=true
    simulationSpace=RootRelative
    alphaBoolBlend {blendOption=Linear}
    boundBone {boneName=dyn_backFurMain_01}
    componentAppliedLinearAccClamp=<3500 3500 3500>
    componentLinearAccScale=<0.2 0.2 0.2>
    componentLinearVelScale=<0.1 0.1 0.1>
    physicsBodyDefinitions [{
      sphereCollisionRadius=0
      boundBone {boneName=dyn_backFurMain_01}
      boxExtents=<20 10 20>
      constraintSetup {angularLimitsMax=<1 5 0>, angularLimitsMin=<-1 -5 -30>, angularTarget=<0 0 0>}
      localJointOffset=<1 0 0>
     }]
    relativeSpaceBone {boneName=root}
   }
   n20 = Anim Dynamics - Bone: dyn_backFurMain_02(ComponentPose=@n19, Alpha=@n16, ExternalForce="0,0,0") {
    angularDampingOverride=0.9999
    angularSpringConstant=50
    bAngularSpring=true
    bGravityOverrideInSimSpace=true
    bOverrideAngularDamping=true
    simulationSpace=RootRelative
    alphaBoolBlend {blendOption=Linear}
    boundBone {boneName=dyn_backFurMain_02}
    componentAppliedLinearAccClamp=<3500 3500 3500>
    componentLinearAccScale=<0.2 0.2 0.2>
    componentLinearVelScale=<0.1 0.1 0.1>
    physicsBodyDefinitions [{
      sphereCollisionRadius=0
      boundBone {boneName=dyn_backFurMain_02}
      boxExtents=<20 10 20>
      constraintSetup {angularLimitsMax=<1 5 0>, angularLimitsMin=<-1 -5 -30>, angularTarget=<0 0 0>}
      localJointOffset=<1 0 0>
     }]
    relativeSpaceBone {boneName=root}
   }
   n21 = Anim Dynamics - Bone: dyn_shoulderFurBack_l(ComponentPose=@n20, Alpha=1, ExternalForce="0,0,0") {
    angularDampingOverride=0.9999
    angularSpringConstant=100
    bAngularSpring=true
    bGravityOverrideInSimSpace=true
    bOverrideAngularDamping=true
    simulationSpace=RootRelative
    alphaBoolBlend {blendOption=Linear}
    boundBone {boneName=dyn_shoulderFurBack_l}
    componentAppliedLinearAccClamp=<3500 3500 3500>
    componentLinearAccScale=<1.2 1.2 1.2>
    componentLinearVelScale=<1.2 1.2 1.2>
    physicsBodyDefinitions [{
      sphereCollisionRadius=0
      boundBone {boneName=dyn_shoulderFurBack_l}
      boxExtents=<20 10 10>
      constraintSetup {angularLimitsMax=<5 10 5>, angularLimitsMin=<-5 -10 -30>, angularTarget=<0 0 0>}
      localJointOffset=<2 0 0>
     }]
    relativeSpaceBone {boneName=root}
   }
   n22 = Anim Dynamics - Bone: dyn_shoulderFurMid_l(ComponentPose=@n21, Alpha=1, ExternalForce="0,0,0") {
    angularDampingOverride=0.9999
    angularSpringConstant=100
    bAngularSpring=true
    bGravityOverrideInSimSpace=true
    bOverrideAngularDamping=true
    simulationSpace=RootRelative
    alphaBoolBlend {blendOption=Linear}
    boundBone {boneName=dyn_shoulderFurMid_l}
    componentAppliedLinearAccClamp=<3500 3500 3500>
    componentLinearAccScale=<1.2 1.2 1.2>
    componentLinearVelScale=<1.2 1.2 1.2>
    physicsBodyDefinitions [{
      sphereCollisionRadius=0
      boundBone {boneName=dyn_shoulderFurMid_l}
      boxExtents=<20 10 10>
      constraintSetup {angularLimitsMax=<5 10 5>, angularLimitsMin=<-5 -10 -30>, angularTarget=<0 0 0>}
      localJointOffset=<2 0 0>
     }]
    relativeSpaceBone {boneName=root}
   }
   n23 = Anim Dynamics - Bone: dyn_shoulderFurFront_l(ComponentPose=@n22, Alpha=1, ExternalForce="0,0,0") {
    angularDampingOverride=0.9999
    angularSpringConstant=100
    bAngularSpring=true
    bGravityOverrideInSimSpace=true
    bOverrideAngularDamping=true
    simulationSpace=RootRelative
    alphaBoolBlend {blendOption=Linear}
    boundBone {boneName=dyn_shoulderFurFront_l}
    componentAppliedLinearAccClamp=<3500 3500 3500>
    componentLinearAccScale=<1.2 1.2 1.2>
    componentLinearVelScale=<1.2 1.2 1.2>
    physicsBodyDefinitions [{
      sphereCollisionRadius=0
      boundBone {boneName=dyn_shoulderFurFront_l}
      boxExtents=<20 10 10>
      constraintSetup {angularLimitsMax=<5 10 5>, angularLimitsMin=<-5 -10 -30>, angularTarget=<0 0 0>}
      localJointOffset=<2 0 0>
     }]
    relativeSpaceBone {boneName=root}
   }
   n24 = Anim Dynamics - Bone: dyn_backFurA(ComponentPose=@n23, Alpha=1, ExternalForce="0,0,0") {
    angularDampingOverride=0.9999
    angularSpringConstant=100
    bAngularSpring=true
    bGravityOverrideInSimSpace=true
    bOverrideAngularDamping=true
    simulationSpace=RootRelative
    alphaBoolBlend {blendOption=Linear}
    boundBone {boneName=dyn_backFurA}
    componentAppliedLinearAccClamp=<3500 3500 3500>
    componentLinearAccScale=<1.2 1.2 1.2>
    componentLinearVelScale=<1.2 1.2 1.2>
    physicsBodyDefinitions [{
      sphereCollisionRadius=0
      boundBone {boneName=dyn_backFurA}
      boxExtents=<20 10 10>
      constraintSetup {angularLimitsMax=<5 10 5>, angularLimitsMin=<-5 -10 -30>, angularTarget=<0 0 0>}
      localJointOffset=<2 0 0>
     }]
    relativeSpaceBone {boneName=root}
   }
   n25 = Anim Dynamics - Bone: dyn_backFurD(ComponentPose=@n24, Alpha=1, ExternalForce="0,0,0") {
    angularDampingOverride=0.9999
    angularSpringConstant=100
    bAngularSpring=true
    bGravityOverrideInSimSpace=true
    bOverrideAngularDamping=true
    simulationSpace=RootRelative
    alphaBoolBlend {blendOption=Linear}
    boundBone {boneName=dyn_backFurD}
    componentAppliedLinearAccClamp=<3500 3500 3500>
    componentLinearAccScale=<1.2 1.2 1.2>
    componentLinearVelScale=<1.2 1.2 1.2>
    physicsBodyDefinitions [{
      sphereCollisionRadius=0
      boundBone {boneName=dyn_backFurD}
      boxExtents=<20 10 10>
      constraintSetup {angularLimitsMax=<5 10 5>, angularLimitsMin=<-5 -10 -30>, angularTarget=<0 0 0>}
      localJointOffset=<2 0 0>
     }]
    relativeSpaceBone {boneName=root}
   }
   n26 = Anim Dynamics - Bone: dyn_backFurG(ComponentPose=@n25, Alpha=1, ExternalForce="0,0,0") {
    angularDampingOverride=0.9999
    angularSpringConstant=100
    bAngularSpring=true
    bGravityOverrideInSimSpace=true
    bOverrideAngularDamping=true
    simulationSpace=RootRelative
    alphaBoolBlend {blendOption=Linear}
    boundBone {boneName=dyn_backFurG}
    componentAppliedLinearAccClamp=<3500 3500 3500>
    componentLinearAccScale=<1.2 1.2 1.2>
    componentLinearVelScale=<1.2 1.2 1.2>
    physicsBodyDefinitions [{
      sphereCollisionRadius=0
      boundBone {boneName=dyn_backFurG}
      boxExtents=<20 10 10>
      constraintSetup {angularLimitsMax=<5 10 5>, angularLimitsMin=<-5 -10 -30>, angularTarget=<0 0 0>}
      localJointOffset=<2 0 0>
     }]
    relativeSpaceBone {boneName=root}
   }
   n27 = Anim Dynamics - Bone: dyn_backFurH(ComponentPose=@n26, Alpha=1, ExternalForce="0,0,0") {
    angularDampingOverride=0.9999
    angularSpringConstant=100
    bAngularSpring=true
    bGravityOverrideInSimSpace=true
    bOverrideAngularDamping=true
    simulationSpace=RootRelative
    alphaBoolBlend {blendOption=Linear}
    boundBone {boneName=dyn_backFurH}
    componentAppliedLinearAccClamp=<3500 3500 3500>
    componentLinearAccScale=<1.2 1.2 1.2>
    componentLinearVelScale=<1.2 1.2 1.2>
    physicsBodyDefinitions [{
      sphereCollisionRadius=0
      boundBone {boneName=dyn_backFurH}
      boxExtents=<20 10 10>
      constraintSetup {angularLimitsMax=<5 10 5>, angularLimitsMin=<-5 -10 -30>, angularTarget=<0 0 0>}
      localJointOffset=<2 0 0>
     }]
    relativeSpaceBone {boneName=root}
   }
   n28 = Anim Dynamics - Bone: dyn_backFurI(ComponentPose=@n27, Alpha=1, ExternalForce="0,0,0") {
    angularDampingOverride=0.9999
    angularSpringConstant=100
    bAngularSpring=true
    bGravityOverrideInSimSpace=true
    bOverrideAngularDamping=true
    simulationSpace=RootRelative
    alphaBoolBlend {blendOption=Linear}
    boundBone {boneName=dyn_backFurI}
    componentAppliedLinearAccClamp=<3500 3500 3500>
    componentLinearAccScale=<1.2 1.2 1.2>
    componentLinearVelScale=<1.2 1.2 1.2>
    physicsBodyDefinitions [{
      sphereCollisionRadius=0
      boundBone {boneName=dyn_backFurI}
      boxExtents=<20 10 10>
      constraintSetup {angularLimitsMax=<5 10 5>, angularLimitsMin=<-5 -10 -30>, angularTarget=<0 0 0>}
      localJointOffset=<2 0 0>
     }]
    relativeSpaceBone {boneName=root}
   }
   n29 = Anim Dynamics - Bone: dyn_backFurJ(ComponentPose=@n28, Alpha=1, ExternalForce="0,0,0") {
    angularDampingOverride=0.9999
    angularSpringConstant=100
    bAngularSpring=true
    bGravityOverrideInSimSpace=true
    bOverrideAngularDamping=true
    simulationSpace=RootRelative
    alphaBoolBlend {blendOption=Linear}
    boundBone {boneName=dyn_backFurJ}
    componentAppliedLinearAccClamp=<3500 3500 3500>
    componentLinearAccScale=<1.2 1.2 1.2>
    componentLinearVelScale=<1.2 1.2 1.2>
    physicsBodyDefinitions [{
      sphereCollisionRadius=0
      boundBone {boneName=dyn_backFurJ}
      boxExtents=<20 10 10>
      constraintSetup {angularLimitsMax=<5 10 5>, angularLimitsMin=<-5 -10 -30>, angularTarget=<0 0 0>}
      localJointOffset=<2 0 0>
     }]
    relativeSpaceBone {boneName=root}
   }
   n30 = Anim Dynamics - Bone: dyn_backFurK(ComponentPose=@n29, Alpha=1, ExternalForce="0,0,0") {
    angularDampingOverride=0.9999
    angularSpringConstant=100
    bAngularSpring=true
    bGravityOverrideInSimSpace=true
    bOverrideAngularDamping=true
    simulationSpace=RootRelative
    alphaBoolBlend {blendOption=Linear}
    boundBone {boneName=dyn_backFurK}
    componentAppliedLinearAccClamp=<3500 3500 3500>
    componentLinearAccScale=<1.2 1.2 1.2>
    componentLinearVelScale=<1.2 1.2 1.2>
    physicsBodyDefinitions [{
      sphereCollisionRadius=0
      boundBone {boneName=dyn_backFurK}
      boxExtents=<20 10 10>
      constraintSetup {angularLimitsMax=<5 10 5>, angularLimitsMin=<-5 -10 -30>, angularTarget=<0 0 0>}
      localJointOffset=<2 0 0>
     }]
    relativeSpaceBone {boneName=root}
   }
   n31 = Anim Dynamics - Bone: dyn_backFurB(ComponentPose=@n30, Alpha=1, ExternalForce="0,0,0") {
    angularDampingOverride=0.9999
    angularSpringConstant=100
    bAngularSpring=true
    bGravityOverrideInSimSpace=true
    bOverrideAngularDamping=true
    simulationSpace=RootRelative
    alphaBoolBlend {blendOption=Linear}
    boundBone {boneName=dyn_backFurB}
    componentAppliedLinearAccClamp=<3500 3500 3500>
    componentLinearAccScale=<1.2 1.2 1.2>
    componentLinearVelScale=<1.2 1.2 1.2>
    physicsBodyDefinitions [{
      sphereCollisionRadius=0
      boundBone {boneName=dyn_backFurB}
      boxExtents=<20 10 10>
      constraintSetup {angularLimitsMax=<5 10 5>, angularLimitsMin=<-5 -10 -30>, angularTarget=<0 0 0>}
      localJointOffset=<2 0 0>
     }]
    relativeSpaceBone {boneName=root}
   }
   n32 = Anim Dynamics - Bone: dyn_backFurE(ComponentPose=@n31, Alpha=1, ExternalForce="0,0,0") {
    angularDampingOverride=0.9999
    angularSpringConstant=100
    bAngularSpring=true
    bGravityOverrideInSimSpace=true
    bOverrideAngularDamping=true
    simulationSpace=RootRelative
    alphaBoolBlend {blendOption=Linear}
    boundBone {boneName=dyn_backFurE}
    componentAppliedLinearAccClamp=<3500 3500 3500>
    componentLinearAccScale=<1.2 1.2 1.2>
    componentLinearVelScale=<1.2 1.2 1.2>
    physicsBodyDefinitions [{
      sphereCollisionRadius=0
      boundBone {boneName=dyn_backFurE}
      boxExtents=<20 10 10>
      constraintSetup {angularLimitsMax=<5 10 5>, angularLimitsMin=<-5 -10 -30>, angularTarget=<0 0 0>}
      localJointOffset=<2 0 0>
     }]
    relativeSpaceBone {boneName=root}
   }
   n33 = Anim Dynamics - Bone: dyn_backFurF(ComponentPose=@n32, Alpha=1, ExternalForce="0,0,0") {
    angularDampingOverride=0.9999
    angularSpringConstant=100
    bAngularSpring=true
    bGravityOverrideInSimSpace=true
    bOverrideAngularDamping=true
    simulationSpace=RootRelative
    alphaBoolBlend {blendOption=Linear}
    boundBone {boneName=dyn_backFurF}
    componentAppliedLinearAccClamp=<3500 3500 3500>
    componentLinearAccScale=<1.2 1.2 1.2>
    componentLinearVelScale=<1.2 1.2 1.2>
    physicsBodyDefinitions [{
      sphereCollisionRadius=0
      boundBone {boneName=dyn_backFurF}
      boxExtents=<20 10 10>
      constraintSetup {angularLimitsMax=<5 10 5>, angularLimitsMin=<-5 -10 -30>, angularTarget=<0 0 0>}
      localJointOffset=<2 0 0>
     }]
    relativeSpaceBone {boneName=root}
   }
   n34 = Anim Dynamics - Bone: dyn_backFurC(ComponentPose=@n33, Alpha=1, ExternalForce="0,0,0") {
    angularDampingOverride=0.9999
    angularSpringConstant=100
    bAngularSpring=true
    bGravityOverrideInSimSpace=true
    bOverrideAngularDamping=true
    simulationSpace=RootRelative
    alphaBoolBlend {blendOption=Linear}
    boundBone {boneName=dyn_backFurC}
    componentAppliedLinearAccClamp=<3500 3500 3500>
    componentLinearAccScale=<1.2 1.2 1.2>
    componentLinearVelScale=<1.2 1.2 1.2>
    physicsBodyDefinitions [{
      sphereCollisionRadius=0
      boundBone {boneName=dyn_backFurC}
      boxExtents=<20 10 10>
      constraintSetup {angularLimitsMax=<5 10 5>, angularLimitsMin=<-5 -10 -30>, angularTarget=<0 0 0>}
      localJointOffset=<2 0 0>
     }]
    relativeSpaceBone {boneName=root}
   }
   n35 = Anim Dynamics - Bone: dyn_shoulderFurBack_r(ComponentPose=@n34, Alpha=1, ExternalForce="0,0,0") {
    angularDampingOverride=0.9999
    angularSpringConstant=100
    bAngularSpring=true
    bGravityOverrideInSimSpace=true
    bOverrideAngularDamping=true
    simulationSpace=RootRelative
    alphaBoolBlend {blendOption=Linear}
    boundBone {boneName=dyn_shoulderFurBack_r}
    componentAppliedLinearAccClamp=<3500 3500 3500>
    componentLinearAccScale=<1.2 1.2 1.2>
    componentLinearVelScale=<1.2 1.2 1.2>
    physicsBodyDefinitions [{
      sphereCollisionRadius=0
      boundBone {boneName=dyn_shoulderFurBack_r}
      boxExtents=<20 10 10>
      constraintSetup {angularLimitsMax=<5 10 5>, angularLimitsMin=<-5 -10 -30>, angularTarget=<0 0 0>}
      localJointOffset=<2 0 0>
     }]
    relativeSpaceBone {boneName=root}
   }
   n36 = Anim Dynamics - Bone: dyn_shoulderFurMid_r(ComponentPose=@n35, Alpha=1, ExternalForce="0,0,0") {
    angularDampingOverride=0.9999
    angularSpringConstant=100
    bAngularSpring=true
    bGravityOverrideInSimSpace=true
    bOverrideAngularDamping=true
    simulationSpace=RootRelative
    alphaBoolBlend {blendOption=Linear}
    boundBone {boneName=dyn_shoulderFurMid_r}
    componentAppliedLinearAccClamp=<3500 3500 3500>
    componentLinearAccScale=<1.2 1.2 1.2>
    componentLinearVelScale=<1.2 1.2 1.2>
    physicsBodyDefinitions [{
      sphereCollisionRadius=0
      boundBone {boneName=dyn_shoulderFurMid_r}
      boxExtents=<20 10 10>
      constraintSetup {angularLimitsMax=<5 10 5>, angularLimitsMin=<-5 -10 -30>, angularTarget=<0 0 0>}
      localJointOffset=<2 0 0>
     }]
    relativeSpaceBone {boneName=root}
   }
   n37 = Anim Dynamics - Bone: dyn_shoulderFurFront_r(ComponentPose=@n36, Alpha=1, ExternalForce="0,0,0") {
    angularDampingOverride=0.9999
    angularSpringConstant=100
    bAngularSpring=true
    bGravityOverrideInSimSpace=true
    bOverrideAngularDamping=true
    simulationSpace=RootRelative
    alphaBoolBlend {blendOption=Linear}
    boundBone {boneName=dyn_shoulderFurFront_r}
    componentAppliedLinearAccClamp=<3500 3500 3500>
    componentLinearAccScale=<1.2 1.2 1.2>
    componentLinearVelScale=<1.2 1.2 1.2>
    physicsBodyDefinitions [{
      sphereCollisionRadius=0
      boundBone {boneName=dyn_shoulderFurFront_r}
      boxExtents=<20 10 10>
      constraintSetup {angularLimitsMax=<5 10 5>, angularLimitsMin=<-5 -10 -30>, angularTarget=<0 0 0>}
      localJointOffset=<2 0 0>
     }]
    relativeSpaceBone {boneName=root}
   }
   n38 = Anim Dynamics - Bone: dyn_fringeA_01_l(ComponentPose=@n37, Alpha=1, ExternalForce="0,0,0") {
    angularDampingOverride=0.9999
    angularSpringConstant=75
    bAngularSpring=true
    bGravityOverrideInSimSpace=true
    bOverrideAngularDamping=true
    simulationSpace=RootRelative
    alphaBoolBlend {blendOption=Linear}
    boundBone {boneName=dyn_fringeA_01_l}
    componentAppliedLinearAccClamp=<3500 3500 3500>
    componentLinearAccScale=<0.9 0.9 0.9>
    componentLinearVelScale=<0.9 0.9 0.9>
    physicsBodyDefinitions [{
      sphereCollisionRadius=0
      boundBone {boneName=dyn_fringeA_01_l}
      boxExtents=<25 15 15>
      constraintSetup {angularLimitsMax=<1 35 5>, angularLimitsMin=<-1 -35 -25>, angularTarget=<0 0 0>}
      localJointOffset=<-2 0 0>
     }]
    relativeSpaceBone {boneName=pelvis}
   }
   n39 = Anim Dynamics - Bone: dyn_fringeA_02_l(ComponentPose=@n38, Alpha=1, ExternalForce="0,0,0") {
    angularDampingOverride=0.9999
    angularSpringConstant=75
    bAngularSpring=true
    bGravityOverrideInSimSpace=true
    bOverrideAngularDamping=true
    bUseSphericalLimits=true
    simulationSpace=RootRelative
    alphaBoolBlend {blendOption=Linear}
    boundBone {boneName=dyn_fringeA_02_l}
    componentAppliedLinearAccClamp=<3500 3500 3500>
    componentLinearAccScale=<0.9 0.9 0.9>
    componentLinearVelScale=<0.9 0.9 0.9>
    physicsBodyDefinitions [{
      collisionType=CustomSphere
      sphereCollisionRadius=4
      boundBone {boneName=dyn_fringeA_02_l}
      boxExtents=<25 15 15>
      constraintSetup {angularLimitsMax=<1 35 5>, angularLimitsMin=<-1 -35 -25>, angularTarget=<0 0 0>}
      localJointOffset=<-2 0 0>
     }]
    relativeSpaceBone {boneName=pelvis}
    sphericalLimits [{limitRadius=6, drivingBone {boneName=upperarm_twist_02_l}, sphereLocalOffset=<0 0 -3>}]
   }
   n40 = Anim Dynamics - Bone: dyn_fringeB_01_l(ComponentPose=@n39, Alpha=1, ExternalForce="0,0,0") {
    angularDampingOverride=0.9999
    angularSpringConstant=50
    bAngularSpring=true
    bGravityOverrideInSimSpace=true
    bOverrideAngularDamping=true
    bUseSphericalLimits=true
    simulationSpace=RootRelative
    alphaBoolBlend {blendOption=Linear}
    boundBone {boneName=dyn_fringeB_01_l}
    componentAppliedLinearAccClamp=<3500 3500 3500>
    componentLinearAccScale=<0.9 0.9 0.9>
    componentLinearVelScale=<0.9 0.9 0.9>
    physicsBodyDefinitions [{
      collisionType=CustomSphere
      sphereCollisionRadius=3
      boundBone {boneName=dyn_fringeB_01_l}
      boxExtents=<23 10 10>
      constraintSetup {angularLimitsMax=<10 35 5>, angularLimitsMin=<-10 -35 -25>, angularTarget=<0 0 0>}
      localJointOffset=<-3 0 0>
     }]
    relativeSpaceBone {boneName=pelvis}
    sphericalLimits [{limitRadius=6, drivingBone {boneName=upperarm_twist_02_l}, sphereLocalOffset=<-1 -1 -3>}]
   }
   n41 = Anim Dynamics - Bone: dyn_fringeB_02_l(ComponentPose=@n40, Alpha=1, ExternalForce="0,0,0") {
    angularDampingOverride=0.9999
    angularSpringConstant=50
    bAngularSpring=true
    bGravityOverrideInSimSpace=true
    bOverrideAngularDamping=true
    bUseSphericalLimits=true
    simulationSpace=RootRelative
    alphaBoolBlend {blendOption=Linear}
    boundBone {boneName=dyn_fringeB_02_l}
    componentAppliedLinearAccClamp=<3500 3500 3500>
    componentLinearAccScale=<0.9 0.9 0.9>
    componentLinearVelScale=<0.9 0.9 0.9>
    physicsBodyDefinitions [{
      collisionType=CustomSphere
      sphereCollisionRadius=4
      boundBone {boneName=dyn_fringeB_02_l}
      boxExtents=<23 10 10>
      constraintSetup {angularLimitsMax=<10 35 25>, angularLimitsMin=<-10 -35 -25>, angularTarget=<0 0 0>}
      localJointOffset=<-3 0 0>
     }]
    relativeSpaceBone {boneName=pelvis}
    sphericalLimits [{limitRadius=6, drivingBone {boneName=upperarm_twist_02_l}, sphereLocalOffset=<0 0 -3>}]
   }
   n42 = Anim Dynamics - Bone: dyn_fringeC_01_l(ComponentPose=@n41, Alpha=1, ExternalForce="0,0,0") {
    angularDampingOverride=0.9999
    angularSpringConstant=50
    bAngularSpring=true
    bGravityOverrideInSimSpace=true
    bOverrideAngularDamping=true
    bUseSphericalLimits=true
    simulationSpace=RootRelative
    alphaBoolBlend {blendOption=Linear}
    boundBone {boneName=dyn_fringeC_01_l}
    componentAppliedLinearAccClamp=<3500 3500 3500>
    componentLinearAccScale=<0.9 0.9 0.9>
    componentLinearVelScale=<0.9 0.9 0.9>
    physicsBodyDefinitions [{
      collisionType=CustomSphere
      sphereCollisionRadius=4
      boundBone {boneName=dyn_fringeC_01_l}
      boxExtents=<23 10 10>
      constraintSetup {angularLimitsMax=<10 35 5>, angularLimitsMin=<-10 -35 -25>, angularTarget=<0 0 0>}
      localJointOffset=<-3 0 0>
     }]
    relativeSpaceBone {boneName=pelvis}
    sphericalLimits [{limitRadius=6, drivingBone {boneName=upperarm_twist_02_l}, sphereLocalOffset=<0 0 -3>}]
   }
   n43 = Anim Dynamics - Bone: dyn_fringeC_02_l(ComponentPose=@n42, Alpha=1, ExternalForce="0,0,0") {
    angularDampingOverride=0.9999
    angularSpringConstant=50
    bAngularSpring=true
    bGravityOverrideInSimSpace=true
    bOverrideAngularDamping=true
    bUseSphericalLimits=true
    simulationSpace=RootRelative
    alphaBoolBlend {blendOption=Linear}
    boundBone {boneName=dyn_fringeC_02_l}
    componentAppliedLinearAccClamp=<3500 3500 3500>
    componentLinearAccScale=<0.9 0.9 0.9>
    componentLinearVelScale=<0.9 0.9 0.9>
    physicsBodyDefinitions [{
      collisionType=CustomSphere
      sphereCollisionRadius=3
      boundBone {boneName=dyn_fringeC_02_l}
      boxExtents=<23 10 10>
      constraintSetup {angularLimitsMax=<10 35 25>, angularLimitsMin=<-10 -35 -25>, angularTarget=<0 0 0>}
      localJointOffset=<-3 0 0>
     }]
    relativeSpaceBone {boneName=pelvis}
    sphericalLimits [{limitRadius=6, drivingBone {boneName=upperarm_twist_02_l}, sphereLocalOffset=<0 -1 -3>}]
   }
   n44 = Anim Dynamics - Bone: dyn_fringeD_01_l(ComponentPose=@n43, Alpha=1, ExternalForce="0,0,0") {
    angularDampingOverride=0.9999
    angularSpringConstant=50
    bAngularSpring=true
    bGravityOverrideInSimSpace=true
    bOverrideAngularDamping=true
    bUseSphericalLimits=true
    simulationSpace=RootRelative
    alphaBoolBlend {blendOption=Linear}
    boundBone {boneName=dyn_fringeD_01_l}
    componentAppliedLinearAccClamp=<3500 3500 3500>
    componentLinearAccScale=<0.9 0.9 0.9>
    componentLinearVelScale=<0.9 0.9 0.9>
    physicsBodyDefinitions [{
      collisionType=CustomSphere
      sphereCollisionRadius=3
      boundBone {boneName=dyn_fringeD_01_l}
      boxExtents=<23 10 10>
      constraintSetup {angularLimitsMax=<20 25 10>, angularLimitsMin=<-20 -25 -25>, angularTarget=<0 0 0>}
      localJointOffset=<-3 0 0>
     }]
    relativeSpaceBone {boneName=pelvis}
    sphericalLimits [{limitRadius=6, drivingBone {boneName=upperarm_twist_02_l}, sphereLocalOffset=<0 0 -3>}]
   }
   n45 = Anim Dynamics - Bone: dyn_fringeD_02_l(ComponentPose=@n44, Alpha=1, ExternalForce="0,0,0") {
    angularDampingOverride=0.9999
    angularSpringConstant=50
    bAngularSpring=true
    bGravityOverrideInSimSpace=true
    bOverrideAngularDamping=true
    bUseSphericalLimits=true
    simulationSpace=RootRelative
    alphaBoolBlend {blendOption=Linear}
    boundBone {boneName=dyn_fringeD_02_l}
    componentAppliedLinearAccClamp=<3500 3500 3500>
    componentLinearAccScale=<0.9 0.9 0.9>
    componentLinearVelScale=<0.9 0.9 0.9>
    physicsBodyDefinitions [{
      collisionType=CustomSphere
      sphereCollisionRadius=3
      boundBone {boneName=dyn_fringeD_02_l}
      boxExtents=<23 10 10>
      constraintSetup {angularLimitsMax=<20 25 10>, angularLimitsMin=<-20 -25 -25>, angularTarget=<0 0 0>}
      localJointOffset=<-3 0 0>
     }]
    relativeSpaceBone {boneName=pelvis}
    sphericalLimits [{limitRadius=6, drivingBone {boneName=upperarm_twist_02_l}, sphereLocalOffset=<0 0 -2>}]
   }
   n46 = Anim Dynamics - Bone: dyn_fringeE_01_l(ComponentPose=@n45, Alpha=1, ExternalForce="0,0,0") {
    angularDampingOverride=0.9999
    angularSpringConstant=50
    bAngularSpring=true
    bGravityOverrideInSimSpace=true
    bOverrideAngularDamping=true
    bUseSphericalLimits=true
    simulationSpace=RootRelative
    alphaBoolBlend {blendOption=Linear}
    boundBone {boneName=dyn_fringeE_01_l}
    componentAppliedLinearAccClamp=<3500 3500 3500>
    componentLinearAccScale=<0.9 0.9 0.9>
    componentLinearVelScale=<0.9 0.9 0.9>
    physicsBodyDefinitions [{
      collisionType=CustomSphere
      sphereCollisionRadius=4
      boundBone {boneName=dyn_fringeE_01_l}
      boxExtents=<23 10 10>
      constraintSetup {angularLimitsMax=<10 20 5>, angularLimitsMin=<-10 -20 -25>, angularTarget=<0 0 0>}
      localJointOffset=<-3 0 0>
     }]
    relativeSpaceBone {boneName=pelvis}
    sphericalLimits [{limitRadius=6, drivingBone {boneName=upperarm_twist_02_l}, sphereLocalOffset=<-2 -2 -2>}]
   }
   n47 = Anim Dynamics - Bone: dyn_fringeE_02_l(ComponentPose=@n46, Alpha=1, ExternalForce="0,0,0") {
    angularDampingOverride=0.9999
    angularSpringConstant=50
    bAngularSpring=true
    bGravityOverrideInSimSpace=true
    bOverrideAngularDamping=true
    bUseSphericalLimits=true
    simulationSpace=RootRelative
    alphaBoolBlend {blendOption=Linear}
    boundBone {boneName=dyn_fringeE_02_l}
    componentAppliedLinearAccClamp=<3500 3500 3500>
    componentLinearAccScale=<0.9 0.9 0.9>
    componentLinearVelScale=<0.9 0.9 0.9>
    physicsBodyDefinitions [{
      collisionType=CustomSphere
      sphereCollisionRadius=4
      boundBone {boneName=dyn_fringeE_02_l}
      boxExtents=<23 10 10>
      constraintSetup {angularLimitsMax=<10 35 25>, angularLimitsMin=<-10 -35 -25>, angularTarget=<0 0 0>}
      localJointOffset=<-3 0 0>
     }]
    relativeSpaceBone {boneName=pelvis}
    sphericalLimits [{limitRadius=6.5, drivingBone {boneName=upperarm_twist_02_l}, sphereLocalOffset=<0 -2 -3>}]
   }
   n48 = Anim Dynamics - Bone: dyn_fringeF_01_l(ComponentPose=@n47, Alpha=1, ExternalForce="0,0,0") {
    angularDampingOverride=0.9999
    angularSpringConstant=50
    bAngularSpring=true
    bGravityOverrideInSimSpace=true
    bOverrideAngularDamping=true
    bUseSphericalLimits=true
    simulationSpace=RootRelative
    alphaBoolBlend {blendOption=Linear}
    boundBone {boneName=dyn_fringeF_01_l}
    componentAppliedLinearAccClamp=<3500 3500 3500>
    componentLinearAccScale=<0.9 0.9 0.9>
    componentLinearVelScale=<0.9 0.9 0.9>
    physicsBodyDefinitions [{
      collisionType=CustomSphere
      sphereCollisionRadius=4
      boundBone {boneName=dyn_fringeF_01_l}
      boxExtents=<23 10 10>
      constraintSetup {angularLimitsMax=<10 20 5>, angularLimitsMin=<-10 -20 -25>, angularTarget=<0 0 0>}
      localJointOffset=<-3 0 0>
     }]
    relativeSpaceBone {boneName=pelvis}
    sphericalLimits [{limitRadius=5, drivingBone {boneName=upperarm_twist_02_l}, sphereLocalOffset=<-2 -2 0>}]
   }
   n49 = Anim Dynamics - Bone: dyn_fringeF_02_l(ComponentPose=@n48, Alpha=1, ExternalForce="0,0,0") {
    angularDampingOverride=0.9999
    angularSpringConstant=50
    bAngularSpring=true
    bGravityOverrideInSimSpace=true
    bOverrideAngularDamping=true
    bUseSphericalLimits=true
    simulationSpace=RootRelative
    alphaBoolBlend {blendOption=Linear}
    boundBone {boneName=dyn_fringeF_02_l}
    componentAppliedLinearAccClamp=<3500 3500 3500>
    componentLinearAccScale=<0.9 0.9 0.9>
    componentLinearVelScale=<0.9 0.9 0.9>
    physicsBodyDefinitions [{
      collisionType=CustomSphere
      sphereCollisionRadius=4
      boundBone {boneName=dyn_fringeF_02_l}
      boxExtents=<23 10 10>
      constraintSetup {angularLimitsMax=<10 35 25>, angularLimitsMin=<-10 -35 -25>, angularTarget=<0 0 0>}
      localJointOffset=<-3 0 0>
     }]
    relativeSpaceBone {boneName=pelvis}
    sphericalLimits [{limitRadius=5, drivingBone {boneName=upperarm_twist_02_l}, sphereLocalOffset=<0 -2 0>}]
   }
   n50 = Anim Dynamics - Bone: dyn_fringeG_01_l(ComponentPose=@n49, Alpha=1, ExternalForce="0,0,0") {
    angularDampingOverride=0.9999
    angularSpringConstant=50
    bAngularSpring=true
    bGravityOverrideInSimSpace=true
    bOverrideAngularDamping=true
    bUseSphericalLimits=true
    simulationSpace=RootRelative
    alphaBoolBlend {blendOption=Linear}
    boundBone {boneName=dyn_fringeG_01_l}
    componentAppliedLinearAccClamp=<3500 3500 3500>
    componentLinearAccScale=<0.9 0.9 0.9>
    componentLinearVelScale=<0.9 0.9 0.9>
    physicsBodyDefinitions [{
      collisionType=CustomSphere
      sphereCollisionRadius=4
      boundBone {boneName=dyn_fringeG_01_l}
      boxExtents=<23 10 10>
      constraintSetup {angularLimitsMax=<10 20 5>, angularLimitsMin=<-10 -20 -25>, angularTarget=<0 0 0>}
      localJointOffset=<-3 0 0>
     }]
    relativeSpaceBone {boneName=pelvis}
    sphericalLimits [{limitRadius=5, drivingBone {boneName=upperarm_twist_02_l}, sphereLocalOffset=<-2 -2 0>}]
   }
   n51 = Anim Dynamics - Bone: dyn_fringeG_02_l(ComponentPose=@n50, Alpha=1, ExternalForce="0,0,0") {
    angularDampingOverride=0.9999
    angularSpringConstant=50
    bAngularSpring=true
    bGravityOverrideInSimSpace=true
    bOverrideAngularDamping=true
    bUseSphericalLimits=true
    simulationSpace=RootRelative
    alphaBoolBlend {blendOption=Linear}
    boundBone {boneName=dyn_fringeG_02_l}
    componentAppliedLinearAccClamp=<3500 3500 3500>
    componentLinearAccScale=<0.9 0.9 0.9>
    componentLinearVelScale=<0.9 0.9 0.9>
    physicsBodyDefinitions [{
      collisionType=CustomSphere
      sphereCollisionRadius=4
      boundBone {boneName=dyn_fringeG_02_l}
      boxExtents=<23 10 10>
      constraintSetup {angularLimitsMax=<10 35 25>, angularLimitsMin=<-10 -35 -25>, angularTarget=<0 0 0>}
      localJointOffset=<-3 0 0>
     }]
    relativeSpaceBone {boneName=pelvis}
    sphericalLimits [{limitRadius=5, drivingBone {boneName=upperarm_twist_02_l}, sphereLocalOffset=<0 -2 0>}]
   }
   n52 = Anim Dynamics - Bone: dyn_fringeH_01_l(ComponentPose=@n51, Alpha=1, ExternalForce="0,0,0") {
    angularDampingOverride=0.9999
    angularSpringConstant=50
    bAngularSpring=true
    bGravityOverrideInSimSpace=true
    bOverrideAngularDamping=true
    bUseSphericalLimits=true
    simulationSpace=RootRelative
    alphaBoolBlend {blendOption=Linear}
    boundBone {boneName=dyn_fringeH_01_l}
    componentAppliedLinearAccClamp=<3500 3500 3500>
    componentLinearAccScale=<0.9 0.9 0.9>
    componentLinearVelScale=<0.9 0.9 0.9>
    physicsBodyDefinitions [{
      collisionType=CustomSphere
      sphereCollisionRadius=4
      boundBone {boneName=dyn_fringeH_01_l}
      boxExtents=<23 10 10>
      constraintSetup {angularLimitsMax=<10 20 5>, angularLimitsMin=<-10 -20 -25>, angularTarget=<0 0 0>}
      localJointOffset=<-3 0 0>
     }]
    relativeSpaceBone {boneName=pelvis}
    sphericalLimits [{limitRadius=5, drivingBone {boneName=upperarm_twist_02_l}, sphereLocalOffset=<-2 -2 0>}]
   }
   n53 = Anim Dynamics - Bone: dyn_fringeH_02_l(ComponentPose=@n52, Alpha=1, ExternalForce="0,0,0") {
    angularDampingOverride=0.9999
    angularSpringConstant=50
    bAngularSpring=true
    bGravityOverrideInSimSpace=true
    bOverrideAngularDamping=true
    bUseSphericalLimits=true
    simulationSpace=RootRelative
    alphaBoolBlend {blendOption=Linear}
    boundBone {boneName=dyn_fringeH_02_l}
    componentAppliedLinearAccClamp=<3500 3500 3500>
    componentLinearAccScale=<0.9 0.9 0.9>
    componentLinearVelScale=<0.9 0.9 0.9>
    physicsBodyDefinitions [{
      collisionType=CustomSphere
      sphereCollisionRadius=4
      boundBone {boneName=dyn_fringeH_02_l}
      boxExtents=<23 10 10>
      constraintSetup {angularLimitsMax=<10 35 25>, angularLimitsMin=<-10 -35 -25>, angularTarget=<0 0 0>}
      localJointOffset=<-3 0 0>
     }]
    relativeSpaceBone {boneName=pelvis}
    sphericalLimits [{limitRadius=5, drivingBone {boneName=upperarm_twist_02_l}, sphereLocalOffset=<0 -2 0>}]
   }
   n54 = Anim Dynamics - Bone: dyn_fringeA_01_r(ComponentPose=@n53, Alpha=1, ExternalForce="0,0,0") {
    angularDampingOverride=0.9999
    angularSpringConstant=75
    bAngularSpring=true
    bGravityOverrideInSimSpace=true
    bOverrideAngularDamping=true
    simulationSpace=RootRelative
    alphaBoolBlend {blendOption=Linear}
    boundBone {boneName=dyn_fringeA_01_r}
    componentAppliedLinearAccClamp=<3500 3500 3500>
    componentLinearAccScale=<0.9 0.9 0.9>
    componentLinearVelScale=<0.9 0.9 0.9>
    physicsBodyDefinitions [{
      sphereCollisionRadius=0
      boundBone {boneName=dyn_fringeA_01_r}
      boxExtents=<25 15 15>
      constraintSetup {angularLimitsMax=<1 35 5>, angularLimitsMin=<-1 -35 -25>, angularTarget=<0 0 0>}
      localJointOffset=<2 0 0>
     }]
    relativeSpaceBone {boneName=pelvis}
   }
   n55 = Anim Dynamics - Bone: dyn_fringeA_02_r(ComponentPose=@n54, Alpha=1, ExternalForce="0,0,0") {
    angularDampingOverride=0.9999
    angularSpringConstant=75
    bAngularSpring=true
    bGravityOverrideInSimSpace=true
    bOverrideAngularDamping=true
    bUseSphericalLimits=true
    simulationSpace=RootRelative
    alphaBoolBlend {blendOption=Linear}
    boundBone {boneName=dyn_fringeA_02_r}
    componentAppliedLinearAccClamp=<3500 3500 3500>
    componentLinearAccScale=<0.9 0.9 0.9>
    componentLinearVelScale=<0.9 0.9 0.9>
    physicsBodyDefinitions [{
      collisionType=CustomSphere
      sphereCollisionRadius=4
      boundBone {boneName=dyn_fringeA_02_r}
      boxExtents=<25 15 15>
      constraintSetup {angularLimitsMax=<1 35 5>, angularLimitsMin=<-1 -35 -25>, angularTarget=<0 0 0>}
      localJointOffset=<2 0 0>
     }]
    relativeSpaceBone {boneName=pelvis}
    sphericalLimits [{limitRadius=6, drivingBone {boneName=upperarm_twist_02_r}, sphereLocalOffset=<0 0 3>}]
   }
   n56 = Anim Dynamics - Bone: dyn_fringeB_01_r(ComponentPose=@n55, Alpha=1, ExternalForce="0,0,0") {
    angularDampingOverride=0.9999
    angularSpringConstant=50
    bAngularSpring=true
    bGravityOverrideInSimSpace=true
    bOverrideAngularDamping=true
    bUseSphericalLimits=true
    simulationSpace=RootRelative
    alphaBoolBlend {blendOption=Linear}
    boundBone {boneName=dyn_fringeB_01_r}
    componentAppliedLinearAccClamp=<3500 3500 3500>
    componentLinearAccScale=<0.9 0.9 0.9>
    componentLinearVelScale=<0.9 0.9 0.9>
    physicsBodyDefinitions [{
      collisionType=CustomSphere
      sphereCollisionRadius=3
      boundBone {boneName=dyn_fringeB_01_r}
      boxExtents=<23 10 10>
      constraintSetup {angularLimitsMax=<10 35 5>, angularLimitsMin=<-10 -35 -25>, angularTarget=<0 0 0>}
      localJointOffset=<3 0 0>
     }]
    relativeSpaceBone {boneName=pelvis}
    sphericalLimits [{limitRadius=6, drivingBone {boneName=upperarm_twist_02_r}, sphereLocalOffset=<1 1 3>}]
   }
   n57 = Anim Dynamics - Bone: dyn_fringeB_02_r(ComponentPose=@n56, Alpha=1, ExternalForce="0,0,0") {
    angularDampingOverride=0.9999
    angularSpringConstant=50
    bAngularSpring=true
    bGravityOverrideInSimSpace=true
    bOverrideAngularDamping=true
    bUseSphericalLimits=true
    simulationSpace=RootRelative
    alphaBoolBlend {blendOption=Linear}
    boundBone {boneName=dyn_fringeB_02_r}
    componentAppliedLinearAccClamp=<3500 3500 3500>
    componentLinearAccScale=<0.9 0.9 0.9>
    componentLinearVelScale=<0.9 0.9 0.9>
    physicsBodyDefinitions [{
      collisionType=CustomSphere
      sphereCollisionRadius=4
      boundBone {boneName=dyn_fringeB_02_r}
      boxExtents=<23 10 10>
      constraintSetup {angularLimitsMax=<10 35 25>, angularLimitsMin=<-10 -35 -25>, angularTarget=<0 0 0>}
      localJointOffset=<3 0 0>
     }]
    relativeSpaceBone {boneName=pelvis}
    sphericalLimits [{limitRadius=6, drivingBone {boneName=upperarm_twist_02_r}, sphereLocalOffset=<0 0 3>}]
   }
   n58 = Anim Dynamics - Bone: dyn_fringeC_01_r(ComponentPose=@n57, Alpha=1, ExternalForce="0,0,0") {
    angularDampingOverride=0.9999
    angularSpringConstant=50
    bAngularSpring=true
    bGravityOverrideInSimSpace=true
    bOverrideAngularDamping=true
    bUseSphericalLimits=true
    simulationSpace=RootRelative
    alphaBoolBlend {blendOption=Linear}
    boundBone {boneName=dyn_fringeC_01_r}
    componentAppliedLinearAccClamp=<3500 3500 3500>
    componentLinearAccScale=<0.9 0.9 0.9>
    componentLinearVelScale=<0.9 0.9 0.9>
    physicsBodyDefinitions [{
      collisionType=CustomSphere
      sphereCollisionRadius=4
      boundBone {boneName=dyn_fringeC_01_r}
      boxExtents=<23 10 10>
      constraintSetup {angularLimitsMax=<10 35 5>, angularLimitsMin=<-10 -35 -25>, angularTarget=<0 0 0>}
      localJointOffset=<3 0 0>
     }]
    relativeSpaceBone {boneName=pelvis}
    sphericalLimits [{limitRadius=6, drivingBone {boneName=upperarm_twist_02_r}, sphereLocalOffset=<0 0 3>}]
   }
   n59 = Anim Dynamics - Bone: dyn_fringeC_02_r(ComponentPose=@n58, Alpha=1, ExternalForce="0,0,0") {
    angularDampingOverride=0.9999
    angularSpringConstant=50
    bAngularSpring=true
    bGravityOverrideInSimSpace=true
    bOverrideAngularDamping=true
    bUseSphericalLimits=true
    simulationSpace=RootRelative
    alphaBoolBlend {blendOption=Linear}
    boundBone {boneName=dyn_fringeC_02_r}
    componentAppliedLinearAccClamp=<3500 3500 3500>
    componentLinearAccScale=<0.9 0.9 0.9>
    componentLinearVelScale=<0.9 0.9 0.9>
    physicsBodyDefinitions [{
      collisionType=CustomSphere
      sphereCollisionRadius=3
      boundBone {boneName=dyn_fringeC_02_r}
      boxExtents=<23 10 10>
      constraintSetup {angularLimitsMax=<10 35 25>, angularLimitsMin=<-10 -35 -25>, angularTarget=<0 0 0>}
      localJointOffset=<3 0 0>
     }]
    relativeSpaceBone {boneName=pelvis}
    sphericalLimits [{limitRadius=6, drivingBone {boneName=upperarm_twist_02_r}, sphereLocalOffset=<0 1 3>}]
   }
   n60 = Anim Dynamics - Bone: dyn_fringeD_01_r(ComponentPose=@n59, Alpha=1, ExternalForce="0,0,0") {
    angularDampingOverride=0.9999
    angularSpringConstant=50
    bAngularSpring=true
    bGravityOverrideInSimSpace=true
    bOverrideAngularDamping=true
    bUseSphericalLimits=true
    simulationSpace=RootRelative
    alphaBoolBlend {blendOption=Linear}
    boundBone {boneName=dyn_fringeD_01_r}
    componentAppliedLinearAccClamp=<3500 3500 3500>
    componentLinearAccScale=<0.9 0.9 0.9>
    componentLinearVelScale=<0.9 0.9 0.9>
    physicsBodyDefinitions [{
      collisionType=CustomSphere
      sphereCollisionRadius=3
      boundBone {boneName=dyn_fringeD_01_r}
      boxExtents=<23 10 10>
      constraintSetup {angularLimitsMax=<20 25 10>, angularLimitsMin=<-20 -25 -25>, angularTarget=<0 0 0>}
      localJointOffset=<3 0 0>
     }]
    relativeSpaceBone {boneName=pelvis}
    sphericalLimits [{limitRadius=6, drivingBone {boneName=upperarm_twist_02_r}, sphereLocalOffset=<0 0 3>}]
   }
   n61 = Anim Dynamics - Bone: dyn_fringeD_02_r(ComponentPose=@n60, Alpha=1, ExternalForce="0,0,0") {
    angularDampingOverride=0.9999
    angularSpringConstant=50
    bAngularSpring=true
    bGravityOverrideInSimSpace=true
    bOverrideAngularDamping=true
    bUseSphericalLimits=true
    simulationSpace=RootRelative
    alphaBoolBlend {blendOption=Linear}
    boundBone {boneName=dyn_fringeD_02_r}
    componentAppliedLinearAccClamp=<3500 3500 3500>
    componentLinearAccScale=<0.9 0.9 0.9>
    componentLinearVelScale=<0.9 0.9 0.9>
    physicsBodyDefinitions [{
      collisionType=CustomSphere
      sphereCollisionRadius=3
      boundBone {boneName=dyn_fringeD_02_r}
      boxExtents=<23 10 10>
      constraintSetup {angularLimitsMax=<20 25 10>, angularLimitsMin=<-20 -25 -25>, angularTarget=<0 0 0>}
      localJointOffset=<3 0 0>
     }]
    relativeSpaceBone {boneName=pelvis}
    sphericalLimits [{limitRadius=6, drivingBone {boneName=upperarm_twist_02_r}, sphereLocalOffset=<0 0 2>}]
   }
   n62 = Anim Dynamics - Bone: dyn_fringeE_01_r(ComponentPose=@n61, Alpha=1, ExternalForce="0,0,0") {
    angularDampingOverride=0.9999
    angularSpringConstant=50
    bAngularSpring=true
    bGravityOverrideInSimSpace=true
    bOverrideAngularDamping=true
    bUseSphericalLimits=true
    simulationSpace=RootRelative
    alphaBoolBlend {blendOption=Linear}
    boundBone {boneName=dyn_fringeE_01_r}
    componentAppliedLinearAccClamp=<3500 3500 3500>
    componentLinearAccScale=<0.9 0.9 0.9>
    componentLinearVelScale=<0.9 0.9 0.9>
    physicsBodyDefinitions [{
      collisionType=CustomSphere
      sphereCollisionRadius=4
      boundBone {boneName=dyn_fringeE_01_r}
      boxExtents=<23 10 10>
      constraintSetup {angularLimitsMax=<10 20 5>, angularLimitsMin=<-10 -20 -25>, angularTarget=<0 0 0>}
      localJointOffset=<3 0 0>
     }]
    relativeSpaceBone {boneName=pelvis}
    sphericalLimits [{limitRadius=6, drivingBone {boneName=upperarm_twist_02_r}, sphereLocalOffset=<2 2 3>}]
   }
   n63 = Anim Dynamics - Bone: dyn_fringeE_02_r(ComponentPose=@n62, Alpha=1, ExternalForce="0,0,0") {
    angularDampingOverride=0.9999
    angularSpringConstant=50
    bAngularSpring=true
    bGravityOverrideInSimSpace=true
    bOverrideAngularDamping=true
    bUseSphericalLimits=true
    simulationSpace=RootRelative
    alphaBoolBlend {blendOption=Linear}
    boundBone {boneName=dyn_fringeE_02_r}
    componentAppliedLinearAccClamp=<3500 3500 3500>
    componentLinearAccScale=<0.9 0.9 0.9>
    componentLinearVelScale=<0.9 0.9 0.9>
    physicsBodyDefinitions [{
      collisionType=CustomSphere
      sphereCollisionRadius=4
      boundBone {boneName=dyn_fringeE_02_r}
      boxExtents=<23 10 10>
      constraintSetup {angularLimitsMax=<10 35 25>, angularLimitsMin=<-10 -35 -25>, angularTarget=<0 0 0>}
      localJointOffset=<3 0 0>
     }]
    relativeSpaceBone {boneName=pelvis}
    sphericalLimits [{limitRadius=6.5, drivingBone {boneName=upperarm_twist_02_r}, sphereLocalOffset=<0 2 3>}]
   }
   n64 = Anim Dynamics - Bone: dyn_fringeF_01_r(ComponentPose=@n63, Alpha=1, ExternalForce="0,0,0") {
    angularDampingOverride=0.9999
    angularSpringConstant=50
    bAngularSpring=true
    bGravityOverrideInSimSpace=true
    bOverrideAngularDamping=true
    bUseSphericalLimits=true
    simulationSpace=RootRelative
    alphaBoolBlend {blendOption=Linear}
    boundBone {boneName=dyn_fringeF_01_r}
    componentAppliedLinearAccClamp=<3500 3500 3500>
    componentLinearAccScale=<0.9 0.9 0.9>
    componentLinearVelScale=<0.9 0.9 0.9>
    physicsBodyDefinitions [{
      collisionType=CustomSphere
      sphereCollisionRadius=4
      boundBone {boneName=dyn_fringeF_01_r}
      boxExtents=<23 10 10>
      constraintSetup {angularLimitsMax=<10 20 5>, angularLimitsMin=<-10 -20 -25>, angularTarget=<0 0 0>}
      localJointOffset=<3 0 0>
     }]
    relativeSpaceBone {boneName=pelvis}
    sphericalLimits [{limitRadius=5, drivingBone {boneName=upperarm_twist_02_r}, sphereLocalOffset=<2 2 0>}]
   }
   n65 = Anim Dynamics - Bone: dyn_fringeF_02_r(ComponentPose=@n64, Alpha=1, ExternalForce="0,0,0") {
    angularDampingOverride=0.9999
    angularSpringConstant=50
    bAngularSpring=true
    bGravityOverrideInSimSpace=true
    bOverrideAngularDamping=true
    bUseSphericalLimits=true
    simulationSpace=RootRelative
    alphaBoolBlend {blendOption=Linear}
    boundBone {boneName=dyn_fringeF_02_r}
    componentAppliedLinearAccClamp=<3500 3500 3500>
    componentLinearAccScale=<0.9 0.9 0.9>
    componentLinearVelScale=<0.9 0.9 0.9>
    physicsBodyDefinitions [{
      collisionType=CustomSphere
      sphereCollisionRadius=4
      boundBone {boneName=dyn_fringeF_02_r}
      boxExtents=<23 10 10>
      constraintSetup {angularLimitsMax=<10 35 25>, angularLimitsMin=<-10 -35 -25>, angularTarget=<0 0 0>}
      localJointOffset=<3 0 0>
     }]
    relativeSpaceBone {boneName=pelvis}
    sphericalLimits [{limitRadius=5, drivingBone {boneName=upperarm_twist_02_r}, sphereLocalOffset=<0 2 0>}]
   }
   n66 = Anim Dynamics - Bone: dyn_fringeG_01_r(ComponentPose=@n65, Alpha=1, ExternalForce="0,0,0") {
    angularDampingOverride=0.9999
    angularSpringConstant=50
    bAngularSpring=true
    bGravityOverrideInSimSpace=true
    bOverrideAngularDamping=true
    bUseSphericalLimits=true
    simulationSpace=RootRelative
    alphaBoolBlend {blendOption=Linear}
    boundBone {boneName=dyn_fringeG_01_r}
    componentAppliedLinearAccClamp=<3500 3500 3500>
    componentLinearAccScale=<0.9 0.9 0.9>
    componentLinearVelScale=<0.9 0.9 0.9>
    physicsBodyDefinitions [{
      collisionType=CustomSphere
      sphereCollisionRadius=4
      boundBone {boneName=dyn_fringeG_01_r}
      boxExtents=<23 10 10>
      constraintSetup {angularLimitsMax=<10 20 5>, angularLimitsMin=<-10 -20 -25>, angularTarget=<0 0 0>}
      localJointOffset=<3 0 0>
     }]
    relativeSpaceBone {boneName=pelvis}
    sphericalLimits [{limitRadius=5, drivingBone {boneName=upperarm_twist_02_r}, sphereLocalOffset=<2 2 0>}]
   }
   n67 = Anim Dynamics - Bone: dyn_fringeG_02_r(ComponentPose=@n66, Alpha=1, ExternalForce="0,0,0") {
    angularDampingOverride=0.9999
    angularSpringConstant=50
    bAngularSpring=true
    bGravityOverrideInSimSpace=true
    bOverrideAngularDamping=true
    bUseSphericalLimits=true
    simulationSpace=RootRelative
    alphaBoolBlend {blendOption=Linear}
    boundBone {boneName=dyn_fringeG_02_r}
    componentAppliedLinearAccClamp=<3500 3500 3500>
    componentLinearAccScale=<0.9 0.9 0.9>
    componentLinearVelScale=<0.9 0.9 0.9>
    physicsBodyDefinitions [{
      collisionType=CustomSphere
      sphereCollisionRadius=4
      boundBone {boneName=dyn_fringeG_02_r}
      boxExtents=<23 10 10>
      constraintSetup {angularLimitsMax=<10 35 25>, angularLimitsMin=<-10 -35 -25>, angularTarget=<0 0 0>}
      localJointOffset=<3 0 0>
     }]
    relativeSpaceBone {boneName=pelvis}
    sphericalLimits [{limitRadius=5, drivingBone {boneName=upperarm_twist_02_r}, sphereLocalOffset=<0 2 0>}]
   }
   n68 = Anim Dynamics - Bone: dyn_fringeH_01_r(ComponentPose=@n67, Alpha=1, ExternalForce="0,0,0") {
    angularDampingOverride=0.9999
    angularSpringConstant=50
    bAngularSpring=true
    bGravityOverrideInSimSpace=true
    bOverrideAngularDamping=true
    bUseSphericalLimits=true
    simulationSpace=RootRelative
    alphaBoolBlend {blendOption=Linear}
    boundBone {boneName=dyn_fringeH_01_r}
    componentAppliedLinearAccClamp=<3500 3500 3500>
    componentLinearAccScale=<0.9 0.9 0.9>
    componentLinearVelScale=<0.9 0.9 0.9>
    physicsBodyDefinitions [{
      collisionType=CustomSphere
      sphereCollisionRadius=4
      boundBone {boneName=dyn_fringeH_01_r}
      boxExtents=<23 10 10>
      constraintSetup {angularLimitsMax=<10 20 5>, angularLimitsMin=<-10 -20 -25>, angularTarget=<0 0 0>}
      localJointOffset=<3 0 0>
     }]
    relativeSpaceBone {boneName=pelvis}
    sphericalLimits [{limitRadius=5, drivingBone {boneName=upperarm_twist_02_l}, sphereLocalOffset=<2 2 0>}]
   }
   n69 = Anim Dynamics - Bone: dyn_fringeH_02_r(ComponentPose=@n68, Alpha=1, ExternalForce="0,0,0") {
    angularDampingOverride=0.9999
    angularSpringConstant=50
    bAngularSpring=true
    bGravityOverrideInSimSpace=true
    bOverrideAngularDamping=true
    bUseSphericalLimits=true
    simulationSpace=RootRelative
    alphaBoolBlend {blendOption=Linear}
    boundBone {boneName=dyn_fringeH_02_r}
    componentAppliedLinearAccClamp=<3500 3500 3500>
    componentLinearAccScale=<0.9 0.9 0.9>
    componentLinearVelScale=<0.9 0.9 0.9>
    physicsBodyDefinitions [{
      collisionType=CustomSphere
      sphereCollisionRadius=4
      boundBone {boneName=dyn_fringeH_02_r}
      boxExtents=<23 10 10>
      constraintSetup {angularLimitsMax=<10 35 25>, angularLimitsMin=<-10 -35 -25>, angularTarget=<0 0 0>}
      localJointOffset=<3 0 0>
     }]
    relativeSpaceBone {boneName=pelvis}
    sphericalLimits [{limitRadius=5, drivingBone {boneName=upperarm_twist_02_l}, sphereLocalOffset=<0 2 0>}]
   }
   n70 = Anim Dynamics - Bone: dyn_backFringeA_01(ComponentPose=@n69, Alpha=1, ExternalForce="0,0,0") {
    angularDampingOverride=0.9999
    angularSpringConstant=100
    bAngularSpring=true
    bGravityOverrideInSimSpace=true
    bOverrideAngularDamping=true
    simulationSpace=RootRelative
    alphaBoolBlend {blendOption=Linear}
    boundBone {boneName=dyn_backFringeA_01}
    componentLinearAccScale=<1.2 1.2 1.2>
    componentLinearVelScale=<1.2 1.2 1.2>
    physicsBodyDefinitions [{
      sphereCollisionRadius=0
      boundBone {boneName=dyn_backFringeA_01}
      boxExtents=<20 10 10>
      constraintSetup {angularLimitsMax=<1 5 0>, angularLimitsMin=<-1 -5 -30>, angularTarget=<0 0 0>}
      localJointOffset=<2 0 0>
     }]
    relativeSpaceBone {boneName=root}
   }
   n71 = Anim Dynamics - Bone: dyn_backFringeA_02(ComponentPose=@n70, Alpha=1, ExternalForce="0,0,0") {
    angularDampingOverride=0.9999
    angularSpringConstant=100
    bAngularSpring=true
    bGravityOverrideInSimSpace=true
    bOverrideAngularDamping=true
    simulationSpace=RootRelative
    alphaBoolBlend {blendOption=Linear}
    boundBone {boneName=dyn_backFringeA_02}
    componentLinearAccScale=<1.2 1.2 1.2>
    componentLinearVelScale=<1.2 1.2 1.2>
    physicsBodyDefinitions [{
      sphereCollisionRadius=0
      boundBone {boneName=dyn_backFringeA_02}
      boxExtents=<20 10 10>
      constraintSetup {angularLimitsMax=<1 5 0>, angularLimitsMin=<-1 -5 -30>, angularTarget=<0 0 0>}
      localJointOffset=<2 0 0>
     }]
    relativeSpaceBone {boneName=root}
   }
   n72 = Anim Dynamics - Bone: dyn_backFringeB_01(ComponentPose=@n71, Alpha=1, ExternalForce="0,0,0") {
    angularDampingOverride=0.9999
    angularSpringConstant=100
    bAngularSpring=true
    bGravityOverrideInSimSpace=true
    bOverrideAngularDamping=true
    simulationSpace=RootRelative
    alphaBoolBlend {blendOption=Linear}
    boundBone {boneName=dyn_backFringeB_01}
    componentLinearAccScale=<1.2 1.2 1.2>
    componentLinearVelScale=<1.2 1.2 1.2>
    physicsBodyDefinitions [{
      sphereCollisionRadius=0
      boundBone {boneName=dyn_backFringeB_01}
      boxExtents=<20 10 10>
      constraintSetup {angularLimitsMax=<1 5 0>, angularLimitsMin=<-1 -5 -30>, angularTarget=<0 0 0>}
      localJointOffset=<2 0 0>
     }]
    relativeSpaceBone {boneName=root}
   }
   n73 = Anim Dynamics - Bone: dyn_backFringeB_02(ComponentPose=@n72, Alpha=1, ExternalForce="0,0,0") {
    angularDampingOverride=0.9999
    angularSpringConstant=100
    bAngularSpring=true
    bGravityOverrideInSimSpace=true
    bOverrideAngularDamping=true
    simulationSpace=RootRelative
    alphaBoolBlend {blendOption=Linear}
    boundBone {boneName=dyn_backFringeB_02}
    componentLinearAccScale=<1.2 1.2 1.2>
    componentLinearVelScale=<1.2 1.2 1.2>
    physicsBodyDefinitions [{
      sphereCollisionRadius=0
      boundBone {boneName=dyn_backFringeB_02}
      boxExtents=<20 10 10>
      constraintSetup {angularLimitsMax=<1 5 0>, angularLimitsMin=<-1 -5 -30>, angularTarget=<0 0 0>}
      localJointOffset=<2 0 0>
     }]
    relativeSpaceBone {boneName=root}
   }
   n74 = Anim Dynamics - Bone: dyn_backFringeC_01(ComponentPose=@n73, Alpha=1, ExternalForce="0,0,0") {
    angularDampingOverride=0.9999
    angularSpringConstant=50
    bAngularSpring=true
    bGravityOverrideInSimSpace=true
    bOverrideAngularDamping=true
    simulationSpace=RootRelative
    alphaBoolBlend {blendOption=Linear}
    boundBone {boneName=dyn_backFringeC_01}
    componentAppliedLinearAccClamp=<3500 3500 3500>
    componentLinearAccScale=<0.3 0.3 0.4>
    componentLinearVelScale=<0.3 0.3 0.4>
    physicsBodyDefinitions [{
      sphereCollisionRadius=0
      boundBone {boneName=dyn_backFringeC_01}
      boxExtents=<20 10 20>
      constraintSetup {angularLimitsMax=<1 5 0>, angularLimitsMin=<-1 -5 -30>, angularTarget=<0 0 0>}
      localJointOffset=<1 0 0>
     }]
    relativeSpaceBone {boneName=root}
   }
   n75 = Anim Dynamics - Bone: dyn_backFringeC_02(ComponentPose=@n74, Alpha=1, ExternalForce="0,0,0") {
    angularDampingOverride=0.9999
    angularSpringConstant=100
    bAngularSpring=true
    bGravityOverrideInSimSpace=true
    bOverrideAngularDamping=true
    simulationSpace=RootRelative
    alphaBoolBlend {blendOption=Linear}
    boundBone {boneName=dyn_backFringeC_02}
    componentLinearAccScale=<1.2 1.2 1.2>
    componentLinearVelScale=<1.2 1.2 1.2>
    physicsBodyDefinitions [{
      sphereCollisionRadius=0
      boundBone {boneName=dyn_backFringeC_02}
      boxExtents=<20 10 10>
      constraintSetup {angularLimitsMax=<1 5 0>, angularLimitsMin=<-1 -5 -30>, angularTarget=<0 0 0>}
      localJointOffset=<2 0 0>
     }]
    relativeSpaceBone {boneName=root}
   }
   n76 = Anim Dynamics - Bone: dyn_backFringeC_03(ComponentPose=@n75, Alpha=1, ExternalForce="0,0,0") {
    angularDampingOverride=0.9999
    angularSpringConstant=100
    bAngularSpring=true
    bGravityOverrideInSimSpace=true
    bOverrideAngularDamping=true
    simulationSpace=RootRelative
    alphaBoolBlend {blendOption=Linear}
    boundBone {boneName=dyn_backFringeC_03}
    componentLinearAccScale=<1.2 1.2 1.2>
    componentLinearVelScale=<1.2 1.2 1.2>
    physicsBodyDefinitions [{
      sphereCollisionRadius=0
      boundBone {boneName=dyn_backFringeC_03}
      boxExtents=<20 10 10>
      constraintSetup {angularLimitsMax=<1 5 0>, angularLimitsMin=<-1 -5 -30>, angularTarget=<0 0 0>}
      localJointOffset=<2 0 0>
     }]
    relativeSpaceBone {boneName=root}
   }
   n77 = Anim Dynamics - Bone: dyn_backFringeD_01(ComponentPose=@n76, Alpha=1, ExternalForce="0,0,0") {
    angularDampingOverride=0.9999
    angularSpringConstant=50
    bAngularSpring=true
    bGravityOverrideInSimSpace=true
    bOverrideAngularDamping=true
    simulationSpace=RootRelative
    alphaBoolBlend {blendOption=Linear}
    boundBone {boneName=dyn_backFringeD_01}
    componentAppliedLinearAccClamp=<3500 3500 3500>
    componentLinearAccScale=<0.3 0.3 0.4>
    componentLinearVelScale=<0.3 0.3 0.4>
    physicsBodyDefinitions [{
      sphereCollisionRadius=0
      boundBone {boneName=dyn_backFringeD_01}
      boxExtents=<20 10 20>
      constraintSetup {angularLimitsMax=<1 5 0>, angularLimitsMin=<-1 -5 -30>, angularTarget=<0 0 0>}
      localJointOffset=<1 0 0>
     }]
    relativeSpaceBone {boneName=root}
   }
   n78 = Anim Dynamics - Bone: dyn_backFringeD_02(ComponentPose=@n77, Alpha=1, ExternalForce="0,0,0") {
    angularDampingOverride=0.9999
    angularSpringConstant=100
    bAngularSpring=true
    bGravityOverrideInSimSpace=true
    bOverrideAngularDamping=true
    simulationSpace=RootRelative
    alphaBoolBlend {blendOption=Linear}
    boundBone {boneName=dyn_backFringeD_02}
    componentLinearAccScale=<1.2 1.2 1.2>
    componentLinearVelScale=<1.2 1.2 1.2>
    physicsBodyDefinitions [{
      sphereCollisionRadius=0
      boundBone {boneName=dyn_backFringeD_02}
      boxExtents=<20 10 10>
      constraintSetup {angularLimitsMax=<1 5 0>, angularLimitsMin=<-1 -5 -30>, angularTarget=<0 0 0>}
      localJointOffset=<2 0 0>
     }]
    relativeSpaceBone {boneName=root}
   }
   n79 = Anim Dynamics - Bone: dyn_backFringeD_03(ComponentPose=@n78, Alpha=1, ExternalForce="0,0,0") {
    angularDampingOverride=0.9999
    angularSpringConstant=100
    bAngularSpring=true
    bGravityOverrideInSimSpace=true
    bOverrideAngularDamping=true
    simulationSpace=RootRelative
    alphaBoolBlend {blendOption=Linear}
    boundBone {boneName=dyn_backFringeD_03}
    componentLinearAccScale=<1.2 1.2 1.2>
    componentLinearVelScale=<1.2 1.2 1.2>
    physicsBodyDefinitions [{
      sphereCollisionRadius=0
      boundBone {boneName=dyn_backFringeD_03}
      boxExtents=<20 10 10>
      constraintSetup {angularLimitsMax=<1 5 0>, angularLimitsMin=<-1 -5 -30>, angularTarget=<0 0 0>}
      localJointOffset=<2 0 0>
     }]
    relativeSpaceBone {boneName=root}
   }
   n80 = Anim Dynamics - Bone: dyn_backFringeE_01(ComponentPose=@n79, Alpha=1, ExternalForce="0,0,0") {
    angularDampingOverride=0.9999
    angularSpringConstant=100
    bAngularSpring=true
    bGravityOverrideInSimSpace=true
    bOverrideAngularDamping=true
    simulationSpace=RootRelative
    alphaBoolBlend {blendOption=Linear}
    boundBone {boneName=dyn_backFringeE_01}
    componentLinearAccScale=<1.2 1.2 1.2>
    componentLinearVelScale=<1.2 1.2 1.2>
    physicsBodyDefinitions [{
      sphereCollisionRadius=0
      boundBone {boneName=dyn_backFringeE_01}
      boxExtents=<20 10 10>
      constraintSetup {angularLimitsMax=<1 5 0>, angularLimitsMin=<-1 -5 -30>, angularTarget=<0 0 0>}
      localJointOffset=<2 0 0>
     }]
    relativeSpaceBone {boneName=root}
   }
   n81 = Anim Dynamics - Bone: dyn_backFringeE_02(ComponentPose=@n80, Alpha=1, ExternalForce="0,0,0") {
    angularDampingOverride=0.9999
    angularSpringConstant=100
    bAngularSpring=true
    bGravityOverrideInSimSpace=true
    bOverrideAngularDamping=true
    simulationSpace=RootRelative
    alphaBoolBlend {blendOption=Linear}
    boundBone {boneName=dyn_backFringeE_02}
    componentLinearAccScale=<1.2 1.2 1.2>
    componentLinearVelScale=<1.2 1.2 1.2>
    physicsBodyDefinitions [{
      sphereCollisionRadius=0
      boundBone {boneName=dyn_backFringeE_02}
      boxExtents=<20 10 10>
      constraintSetup {angularLimitsMax=<1 5 0>, angularLimitsMin=<-1 -5 -30>, angularTarget=<0 0 0>}
      localJointOffset=<2 0 0>
     }]
    relativeSpaceBone {boneName=root}
   }
   n82 = Anim Dynamics - Bone: dyn_backFringeF_01(ComponentPose=@n81, Alpha=1, ExternalForce="0,0,0") {
    angularDampingOverride=0.9999
    angularSpringConstant=100
    bAngularSpring=true
    bGravityOverrideInSimSpace=true
    bOverrideAngularDamping=true
    simulationSpace=RootRelative
    alphaBoolBlend {blendOption=Linear}
    boundBone {boneName=dyn_backFringeF_01}
    componentLinearAccScale=<1.2 1.2 1.2>
    componentLinearVelScale=<1.2 1.2 1.2>
    physicsBodyDefinitions [{
      sphereCollisionRadius=0
      boundBone {boneName=dyn_backFringeF_01}
      boxExtents=<20 10 10>
      constraintSetup {angularLimitsMax=<1 5 0>, angularLimitsMin=<-1 -5 -30>, angularTarget=<0 0 0>}
      localJointOffset=<2 0 0>
     }]
    relativeSpaceBone {boneName=root}
   }
   n83 = Anim Dynamics - Bone: dyn_backFringeF_02(ComponentPose=@n82, Alpha=1, ExternalForce="0,0,0") {
    angularDampingOverride=0.9999
    angularSpringConstant=100
    bAngularSpring=true
    bGravityOverrideInSimSpace=true
    bOverrideAngularDamping=true
    simulationSpace=RootRelative
    alphaBoolBlend {blendOption=Linear}
    boundBone {boneName=dyn_backFringeF_02}
    componentLinearAccScale=<1.2 1.2 1.2>
    componentLinearVelScale=<1.2 1.2 1.2>
    physicsBodyDefinitions [{
      sphereCollisionRadius=0
      boundBone {boneName=dyn_backFringeF_02}
      boxExtents=<20 10 10>
      constraintSetup {angularLimitsMax=<1 5 0>, angularLimitsMin=<-1 -5 -30>, angularTarget=<0 0 0>}
      localJointOffset=<2 0 0>
     }]
    relativeSpaceBone {boneName=root}
   }
   n84 = "Constraint - dyn_frontStrap_01_Base_r (PPPP)"!(
    ComponentPose=@n83
    Alpha=1
    ConstraintWeights_0=0.5799
    ConstraintWeights_1:float=0.0818
    ConstraintWeights_2:float=0.2507
    ConstraintWeights_3:float=0.0876
   ) -> (Pose) {
    alphaBoolBlend {blendOption=Linear}
    boneToModify {boneName=dyn_frontStrap_01_Base_r}
    constraintSetup [
     {targetBone {boneName=pelvis}}
     {targetBone {boneName=thigh_twist_01_l}}
     {targetBone {boneName=thigh_twist_01_r}}
     {targetBone {boneName=spine_01}}
    ]
    constraintWeights [1,1,1,1]
   }
   n85 = "Constraint - dyn_frontStrap_01_Base_c (PPPPPPPPP)"!(
    ComponentPose=@n84
    Alpha=1
    ConstraintWeights_0=0.0606
    ConstraintWeights_1:float=0.0055
    ConstraintWeights_2:float=0.0742
    ConstraintWeights_3:float=0.1216
    ConstraintWeights_4:float=0.058
    ConstraintWeights_5:float=0.1248
    ConstraintWeights_6:float=0.1471
    ConstraintWeights_7:float=0.3993
    ConstraintWeights_8:float=0.0089
   ) -> (Pose) {
    alphaBoolBlend {blendOption=Linear}
    boneToModify {boneName=dyn_frontStrap_01_Base_c}
    constraintSetup [
     {targetBone {boneName=dyn_loinClothFront_01_r}}
     {targetBone {boneName=dyn_loinClothFront_04_c}}
     {targetBone {boneName=dyn_loinClothFront_03_c}}
     {targetBone {boneName=thigh_twist_01_r}}
     {targetBone {boneName=spine_01}}
     {targetBone {boneName=thigh_twist_01_l}}
     {targetBone {boneName=dyn_loinClothFront_02_c}}
     {targetBone {boneName=pelvis}}
     {targetBone {boneName=dyn_loinClothFront_01_l}}
    ]
    constraintWeights [1,1,1,1,1,1,1,1,1]
   }
   n86 = "Constraint - dyn_frontStrap_01_Base_l (PPPP)"!(
    ComponentPose=@n85
    Alpha=1
    ConstraintWeights_0=0.5799
    ConstraintWeights_1:float=0.2507
    ConstraintWeights_2:float=0.0818
    ConstraintWeights_3:float=0.0876
   ) -> (Pose) {
    alphaBoolBlend {blendOption=Linear}
    boneToModify {boneName=dyn_frontStrap_01_Base_l}
    constraintSetup [
     {targetBone {boneName=pelvis}}
     {targetBone {boneName=thigh_twist_01_l}}
     {targetBone {boneName=thigh_twist_01_r}}
     {targetBone {boneName=spine_01}}
    ]
    constraintWeights [1,1,1,1]
   }
   n87 = "Constraint - dyn_frontStrap_04_r (T)"(ComponentPose=@n86, Alpha=1, ConstraintWeights_0=1) {
    alphaBoolBlend {blendOption=Linear}
    boneToModify {boneName=dyn_frontStrap_04_r}
    constraintSetup [{transformType=Translation, targetBone {boneName=dyn_loinClothFront_05_c}}]
    constraintWeights [1]
   }
   n88 = "Constraint - dyn_frontStrap_04_l (T)"(ComponentPose=@n87, Alpha=1, ConstraintWeights_0=1) {
    alphaBoolBlend {blendOption=Linear}
    boneToModify {boneName=dyn_frontStrap_04_l}
    constraintSetup [{transformType=Translation, targetBone {boneName=dyn_loinClothFront_05_c}}]
    constraintWeights [1]
   }
   n89 = RigidBody(ComponentPose=@n88) {
    bClampLinearTranslationLimitToRefPose=true
    bOverrideWorldGravity=true
    overridePhysicsAsset=&/M_MED_AncientGladiator_Heavy_Front_flap_Physics
    simulationSpace=BaseBoneSpace
    alphaBoolBlend {blendOption=Linear}
    baseBoneRef {boneName=root}
    componentAppliedLinearAccClamp=<2000 2000 1500>
    componentLinearAccScale=<0.2 0.2 0.1>
    componentLinearVelScale=<0.1 0.1 0.1>
    overrideWorldGravity {z=-100}
    simSpaceSettings {maxAngularAcceleration=0, maxAngularVelocity=0, maxLinearAcceleration=500, maxLinearVelocity=50, velocityScaleZ=0}
   }
   n90 = RigidBody(ComponentPose=@n89) {
    bClampLinearTranslationLimitToRefPose=true
    bOverrideWorldGravity=true
    overridePhysicsAsset=&/M_MED_AncientGladiator_Heavy_CRStrapPhysics
    simulationSpace=BaseBoneSpace
    alphaBoolBlend {blendOption=Linear}
    baseBoneRef {boneName=root}
    componentAppliedLinearAccClamp=<2000 2000 1500>
    componentLinearAccScale=<0.2 0.2 0.1>
    componentLinearVelScale=<0.1 0.1 0.1>
    overrideWorldGravity {z=-100}
    simSpaceSettings {maxAngularAcceleration=0, maxAngularVelocity=0, maxLinearAcceleration=500, maxLinearVelocity=50, velocityScaleZ=0}
   }
   n91 = "Constraint - dyn_frontStrap_01_r (PPPPPPP)"!(
    ComponentPose=@n90
    Alpha=1
    ConstraintWeights_0=0.1462
    ConstraintWeights_1:float=0.0484
    ConstraintWeights_2:float=0.2051
    ConstraintWeights_3:float=0.0539
    ConstraintWeights_4:float=0.0658
    ConstraintWeights_5:float=0.1045
    ConstraintWeights_6:float=0.3761
   ) -> (Pose) {
    alphaBoolBlend {blendOption=Linear}
    boneToModify {boneName=dyn_frontStrap_01_r}
    constraintSetup [
     {targetBone {boneName=dyn_loinClothFront_01_r}}
     {targetBone {boneName=dyn_loinClothFront_03_c}}
     {targetBone {boneName=thigh_twist_01_r}}
     {targetBone {boneName=spine_01}}
     {targetBone {boneName=thigh_twist_01_l}}
     {targetBone {boneName=dyn_loinClothFront_02_c}}
     {targetBone {boneName=pelvis}}
    ]
    constraintWeights [1,1,1,1,1,1,1]
   }
   n92 = "Constraint - dyn_frontStrap_02_r (PPPPPPPPPPPPP)"!(
    ComponentPose=@n91
    Alpha=1
    ConstraintWeights_0=0.5896
    ConstraintWeights_1:float=0.0233
    ConstraintWeights_2:float=0.021
    ConstraintWeights_3:float=0.0797
    ConstraintWeights_4:float=0.059
    ConstraintWeights_5:float=0.0105
    ConstraintWeights_6:float=0.0103
    ConstraintWeights_7:float=0.0256
    ConstraintWeights_8:float=0.0904
    ConstraintWeights_9:float=0.0772
    ConstraintWeights_10:float=0.003
    ConstraintWeights_11:float=0.0034
    ConstraintWeights_12:float=0.0069
   ) -> (Pose) {
    alphaBoolBlend {blendOption=Linear}
    boneToModify {boneName=dyn_frontStrap_02_r}
    constraintSetup [
     {targetBone {boneName=dyn_loinClothFront_01_r}}
     {targetBone {boneName=dyn_loinClothFront_02_r}}
     {targetBone {boneName=dyn_loinClothFront_04_c}}
     {targetBone {boneName=dyn_loinClothFront_03_c}}
     {targetBone {boneName=thigh_twist_01_r}}
     {targetBone {boneName=dyn_loinClothFront_01_l}}
     {targetBone {boneName=spine_01}}
     {targetBone {boneName=thigh_twist_01_l}}
     {targetBone {boneName=dyn_loinClothFront_02_c}}
     {targetBone {boneName=pelvis}}
     {targetBone {boneName=dyn_loinClothFront_02_l}}
     {targetBone {boneName=dyn_cutFlap_r}}
     {targetBone {boneName=dyn_loinClothFront_03_r}}
    ]
    constraintWeights [1,1,1,1,1,1,1,1,1,1,1,1,1]
   }
   n93 = "Constraint - dyn_frontStrap_03_r (PPPPPPPPPPPPPP)"!(
    ComponentPose=@n92
    Alpha=1
    ConstraintWeights_0=0.0063
    ConstraintWeights_1:float=0.2385
    ConstraintWeights_2:float=0.1909
    ConstraintWeights_3:float=0.1174
    ConstraintWeights_4:float=0.1348
    ConstraintWeights_5:float=0.0399
    ConstraintWeights_6:float=0.0124
    ConstraintWeights_7:float=0.0082
    ConstraintWeights_8:float=0.0614
    ConstraintWeights_9:float=0.0177
    ConstraintWeights_10:float=0.0112
    ConstraintWeights_11:float=0.0151
    ConstraintWeights_12:float=0.1422
    ConstraintWeights_13:float=0.0041
   ) -> (Pose) {
    alphaBoolBlend {blendOption=Linear}
    boneToModify {boneName=dyn_frontStrap_03_r}
    constraintSetup [
     {targetBone {boneName=dyn_loinClothFront_03_l}}
     {targetBone {boneName=dyn_loinClothFront_01_r}}
     {targetBone {boneName=dyn_loinClothFront_02_r}}
     {targetBone {boneName=dyn_loinClothFront_04_c}}
     {targetBone {boneName=dyn_loinClothFront_03_c}}
     {targetBone {boneName=thigh_twist_01_r}}
     {targetBone {boneName=dyn_loinClothFront_01_l}}
     {targetBone {boneName=thigh_twist_01_l}}
     {targetBone {boneName=dyn_loinClothFront_02_c}}
     {targetBone {boneName=pelvis}}
     {targetBone {boneName=dyn_loinClothFront_02_l}}
     {targetBone {boneName=dyn_cutFlap_r}}
     {targetBone {boneName=dyn_loinClothFront_03_r}}
     {targetBone {boneName=dyn_loinClothFront_05_c}}
    ]
    constraintWeights [1,1,1,1,1,1,1,1,1,1,1,1,1,1]
   }
   n94 = "Constraint - dyn_frontStrap_04_r (PPPPPPPPP)"!(
    ComponentPose=@n93
    Alpha=1
    ConstraintWeights_0=0.0192
    ConstraintWeights_1:float=0.0379
    ConstraintWeights_2:float=0.3889
    ConstraintWeights_3:float=0.0894
    ConstraintWeights_4:float=0.0062
    ConstraintWeights_5:float=0.0094
    ConstraintWeights_6:float=0.0061
    ConstraintWeights_7:float=0.3452
    ConstraintWeights_8:float=0.0978
   ) -> (Pose) {
    alphaBoolBlend {blendOption=Linear}
    boneToModify {boneName=dyn_frontStrap_04_r}
    constraintSetup [
     {targetBone {boneName=dyn_loinClothFront_03_l}}
     {targetBone {boneName=dyn_loinClothFront_02_r}}
     {targetBone {boneName=dyn_loinClothFront_04_c}}
     {targetBone {boneName=dyn_loinClothFront_03_c}}
     {targetBone {boneName=dyn_loinClothFront_06_c}}
     {targetBone {boneName=dyn_loinClothFront_02_c}}
     {targetBone {boneName=dyn_loinClothFront_04_r}}
     {targetBone {boneName=dyn_loinClothFront_03_r}}
     {targetBone {boneName=dyn_loinClothFront_05_c}}
    ]
    constraintWeights [1,1,1,1,1,1,1,1,1]
   }
   n95 = "Constraint - dyn_frontStrap_01_c (PPPPPPPPPPP)"!(
    ComponentPose=@n94
    Alpha=1
    ConstraintWeights_0=0.2034
    ConstraintWeights_1:float=0.0101
    ConstraintWeights_2:float=0.0448
    ConstraintWeights_3:float=0.1777
    ConstraintWeights_4:float=0.0505
    ConstraintWeights_5:float=0.0171
    ConstraintWeights_6:float=0.0821
    ConstraintWeights_7:float=0.1928
    ConstraintWeights_8:float=0.1257
    ConstraintWeights_9:float=0.0193
    ConstraintWeights_10:float=0.0766
   ) -> (Pose) {
    alphaBoolBlend {blendOption=Linear}
    boneToModify {boneName=dyn_frontStrap_01_c}
    constraintSetup [
     {targetBone {boneName=dyn_loinClothFront_01_r}}
     {targetBone {boneName=dyn_loinClothFront_02_r}}
     {targetBone {boneName=dyn_loinClothFront_04_c}}
     {targetBone {boneName=dyn_loinClothFront_03_c}}
     {targetBone {boneName=thigh_twist_01_r}}
     {targetBone {boneName=spine_01}}
     {targetBone {boneName=thigh_twist_01_l}}
     {targetBone {boneName=dyn_loinClothFront_02_c}}
     {targetBone {boneName=pelvis}}
     {targetBone {boneName=dyn_loinClothFront_02_l}}
     {targetBone {boneName=dyn_loinClothFront_01_l}}
    ]
    constraintWeights [1,1,1,1,1,1,1,1,1,1,1]
   }
   n96 = "Constraint - dyn_frontStrap_02_c (PPPPPPPPPPPPP)"!(
    ComponentPose=@n95
    Alpha=1
    ConstraintWeights_0=0.0458
    ConstraintWeights_1:float=0.1078
    ConstraintWeights_2:float=0.0599
    ConstraintWeights_3:float=0.1952
    ConstraintWeights_4:float=0.2142
    ConstraintWeights_5:float=0.0203
    ConstraintWeights_6:float=0.0797
    ConstraintWeights_7:float=0.0318
    ConstraintWeights_8:float=0.0956
    ConstraintWeights_9:float=0.0273
    ConstraintWeights_10:float=0.0651
    ConstraintWeights_11:float=0.047
    ConstraintWeights_12:float=0.0103
   ) -> (Pose) {
    alphaBoolBlend {blendOption=Linear}
    boneToModify {boneName=dyn_frontStrap_02_c}
    constraintSetup [
     {targetBone {boneName=dyn_loinClothFront_03_l}}
     {targetBone {boneName=dyn_loinClothFront_01_r}}
     {targetBone {boneName=dyn_loinClothFront_02_r}}
     {targetBone {boneName=dyn_loinClothFront_04_c}}
     {targetBone {boneName=dyn_loinClothFront_03_c}}
     {targetBone {boneName=thigh_twist_01_r}}
     {targetBone {boneName=dyn_loinClothFront_01_l}}
     {targetBone {boneName=thigh_twist_01_l}}
     {targetBone {boneName=dyn_loinClothFront_02_c}}
     {targetBone {boneName=pelvis}}
     {targetBone {boneName=dyn_loinClothFront_02_l}}
     {targetBone {boneName=dyn_loinClothFront_03_r}}
     {targetBone {boneName=dyn_loinClothFront_05_c}}
    ]
    constraintWeights [1,1,1,1,1,1,1,1,1,1,1,1,1]
   }
   n97 = "Constraint - dyn_frontStrap_03_c (PPPPPPPPPP)"!(
    ComponentPose=@n96
    Alpha=1
    ConstraintWeights_0=0.1033
    ConstraintWeights_1:float=0.0302
    ConstraintWeights_2:float=0.4523
    ConstraintWeights_3:float=0.1757
    ConstraintWeights_4:float=0.006
    ConstraintWeights_5:float=0.0068
    ConstraintWeights_6:float=0.029
    ConstraintWeights_7:float=0.0305
    ConstraintWeights_8:float=0.1033
    ConstraintWeights_9:float=0.0629
   ) -> (Pose) {
    alphaBoolBlend {blendOption=Linear}
    boneToModify {boneName=dyn_frontStrap_03_c}
    constraintSetup [
     {targetBone {boneName=dyn_loinClothFront_03_l}}
     {targetBone {boneName=dyn_loinClothFront_02_r}}
     {targetBone {boneName=dyn_loinClothFront_04_c}}
     {targetBone {boneName=dyn_loinClothFront_03_c}}
     {targetBone {boneName=dyn_loinClothFront_01_l}}
     {targetBone {boneName=dyn_loinClothFront_01_r}}
     {targetBone {boneName=dyn_loinClothFront_02_c}}
     {targetBone {boneName=dyn_loinClothFront_02_l}}
     {targetBone {boneName=dyn_loinClothFront_03_r}}
     {targetBone {boneName=dyn_loinClothFront_05_c}}
    ]
    constraintWeights [1,1,1,1,1,1,1,1,1,1]
   }
   n98 = "Constraint - dyn_hipFlap_01_l (P)"(ComponentPose=@n97, Alpha=1, ConstraintWeights_0=1) {
    alphaBoolBlend {blendOption=Linear}
    boneToModify {boneName=dyn_hipFlap_01_l}
    constraintSetup [{targetBone {boneName=thigh_l}}]
    constraintWeights [1]
   }
   n99 = "Constraint - dyn_frontStrap_01_l (PPPPPPP)"!(
    ComponentPose=@n98
    Alpha=1
    ConstraintWeights_0=0.0484
    ConstraintWeights_1:float=0.0658
    ConstraintWeights_2:float=0.0539
    ConstraintWeights_3:float=0.2051
    ConstraintWeights_4:float=0.1045
    ConstraintWeights_5:float=0.3761
    ConstraintWeights_6:float=0.1462
   ) -> (Pose) {
    alphaBoolBlend {blendOption=Linear}
    boneToModify {boneName=dyn_frontStrap_01_l}
    constraintSetup [
     {targetBone {boneName=dyn_loinClothFront_03_c}}
     {targetBone {boneName=thigh_twist_01_r}}
     {targetBone {boneName=spine_01}}
     {targetBone {boneName=thigh_twist_01_l}}
     {targetBone {boneName=dyn_loinClothFront_02_c}}
     {targetBone {boneName=pelvis}}
     {targetBone {boneName=dyn_loinClothFront_01_l}}
    ]
    constraintWeights [1,1,1,1,1,1,1]
   }
   n100 = "Constraint - dyn_frontStrap_02_l (PPPPPPPPPPPPP)"!(
    ComponentPose=@n99
    Alpha=1
    ConstraintWeights_0=0.0069
    ConstraintWeights_1:float=0.0105
    ConstraintWeights_2:float=0.003
    ConstraintWeights_3:float=0.021
    ConstraintWeights_4:float=0.0797
    ConstraintWeights_5:float=0.0256
    ConstraintWeights_6:float=0.0103
    ConstraintWeights_7:float=0.059
    ConstraintWeights_8:float=0.0904
    ConstraintWeights_9:float=0.0772
    ConstraintWeights_10:float=0.0233
    ConstraintWeights_11:float=0.5896
    ConstraintWeights_12:float=0.0034
   ) -> (Pose) {
    alphaBoolBlend {blendOption=Linear}
    boneToModify {boneName=dyn_frontStrap_02_l}
    constraintSetup [
     {targetBone {boneName=dyn_loinClothFront_03_l}}
     {targetBone {boneName=dyn_loinClothFront_01_r}}
     {targetBone {boneName=dyn_loinClothFront_02_r}}
     {targetBone {boneName=dyn_loinClothFront_04_c}}
     {targetBone {boneName=dyn_loinClothFront_03_c}}
     {targetBone {boneName=thigh_twist_01_r}}
     {targetBone {boneName=spine_01}}
     {targetBone {boneName=thigh_twist_01_l}}
     {targetBone {boneName=dyn_loinClothFront_02_c}}
     {targetBone {boneName=pelvis}}
     {targetBone {boneName=dyn_loinClothFront_02_l}}
     {targetBone {boneName=dyn_loinClothFront_01_l}}
     {targetBone {boneName=dyn_cutFlap_l}}
    ]
    constraintWeights [1,1,1,1,1,1,1,1,1,1,1,1,1]
   }
   n101 = "Constraint - dyn_frontStrap_03_l (PPPPPPPPPPPPPP)"!(
    ComponentPose=@n100
    Alpha=1
    ConstraintWeights_0=0.1419
    ConstraintWeights_1:float=0.0162
    ConstraintWeights_2:float=0.0103
    ConstraintWeights_3:float=0.1175
    ConstraintWeights_4:float=0.1348
    ConstraintWeights_5:float=0.0065
    ConstraintWeights_6:float=0.2347
    ConstraintWeights_7:float=0.0149
    ConstraintWeights_8:float=0.0416
    ConstraintWeights_9:float=0.0614
    ConstraintWeights_10:float=0.0177
    ConstraintWeights_11:float=0.1918
    ConstraintWeights_12:float=0.0066
    ConstraintWeights_13:float=0.0041
   ) -> (Pose) {
    alphaBoolBlend {blendOption=Linear}
    boneToModify {boneName=dyn_frontStrap_03_l}
    constraintSetup [
     {targetBone {boneName=dyn_loinClothFront_03_l}}
     {targetBone {boneName=dyn_loinClothFront_01_r}}
     {targetBone {boneName=dyn_loinClothFront_02_r}}
     {targetBone {boneName=dyn_loinClothFront_04_c}}
     {targetBone {boneName=dyn_loinClothFront_03_c}}
     {targetBone {boneName=thigh_twist_01_r}}
     {targetBone {boneName=dyn_loinClothFront_01_l}}
     {targetBone {boneName=dyn_cutFlap_l}}
     {targetBone {boneName=thigh_twist_01_l}}
     {targetBone {boneName=dyn_loinClothFront_02_c}}
     {targetBone {boneName=pelvis}}
     {targetBone {boneName=dyn_loinClothFront_02_l}}
     {targetBone {boneName=dyn_loinClothFront_03_r}}
     {targetBone {boneName=dyn_loinClothFront_05_c}}
    ]
    constraintWeights [1,1,1,1,1,1,1,1,1,1,1,1,1,1]
   }
   n102 = "Constraint - dyn_frontStrap_04_l (PPPPPPPPP)"!(
    ComponentPose=@n101
    Alpha=1
    ConstraintWeights_0=0.4862
    ConstraintWeights_1:float=0.0032
    ConstraintWeights_2:float=0.2276
    ConstraintWeights_3:float=0.09
    ConstraintWeights_4:float=0.0111
    ConstraintWeights_5:float=0.016
    ConstraintWeights_6:float=0.1232
    ConstraintWeights_7:float=0.0099
    ConstraintWeights_8:float=0.0329
   ) -> (Pose) {
    alphaBoolBlend {blendOption=Linear}
    boneToModify {boneName=dyn_frontStrap_04_l}
    constraintSetup [
     {targetBone {boneName=dyn_loinClothFront_03_l}}
     {targetBone {boneName=dyn_loinClothFront_02_r}}
     {targetBone {boneName=dyn_loinClothFront_04_c}}
     {targetBone {boneName=dyn_loinClothFront_03_c}}
     {targetBone {boneName=dyn_loinClothFront_01_l}}
     {targetBone {boneName=dyn_loinClothFront_02_c}}
     {targetBone {boneName=dyn_loinClothFront_02_l}}
     {targetBone {boneName=dyn_loinClothFront_03_r}}
     {targetBone {boneName=dyn_loinClothFront_05_c}}
    ]
    constraintWeights [1,1,1,1,1,1,1,1,1]
   }
   n103 = "Transform (Modify) Bone - Bone: dyn_frontStrap_01_r"(ComponentPose=@n102, Alpha=1, Translation="0,-0.275,0", Rotation="0,0,0", Scale="1,1,1") {
    translationMode=BMM_Additive
    translationSpace=BCS_BoneSpace
    alphaBoolBlend {blendOption=Linear}
    boneToModify {boneName=dyn_frontStrap_01_r}
    translation {y=-0.275}
   }
   n104 = "Transform (Modify) Bone - Bone: dyn_frontStrap_02_r"(ComponentPose=@n103, Alpha=1, Translation="0,-0.275,0", Rotation="0,0,0", Scale="1,1,1") {
    translationMode=BMM_Additive
    translationSpace=BCS_BoneSpace
    alphaBoolBlend {blendOption=Linear}
    boneToModify {boneName=dyn_frontStrap_02_r}
    translation {y=-0.275}
   }
   n105 = "Transform (Modify) Bone - Bone: dyn_frontStrap_03_r"(ComponentPose=@n104, Alpha=1, Translation="0,-0.275,0", Rotation="0,0,0", Scale="1,1,1") {
    translationMode=BMM_Additive
    translationSpace=BCS_BoneSpace
    alphaBoolBlend {blendOption=Linear}
    boneToModify {boneName=dyn_frontStrap_03_r}
    translation {y=-0.275}
   }
   n106 = "Transform (Modify) Bone - Bone: dyn_frontStrap_04_r"(ComponentPose=@n105, Alpha=1, Translation="0,-0.5,0", Rotation="0,0,0", Scale="1,1,1") {
    translationMode=BMM_Additive
    translationSpace=BCS_BoneSpace
    alphaBoolBlend {blendOption=Linear}
    boneToModify {boneName=dyn_frontStrap_04_r}
    translation {y=-0.5}
   }
   n107 = "Transform (Modify) Bone - Bone: dyn_frontStrap_01_c"(ComponentPose=@n106, Alpha=1, Translation="0,-0.275,0", Rotation="0,0,0", Scale="1,1,1") {
    translationMode=BMM_Additive
    translationSpace=BCS_BoneSpace
    alphaBoolBlend {blendOption=Linear}
    boneToModify {boneName=dyn_frontStrap_01_c}
    translation {y=-0.275}
   }
   n108 = "Transform (Modify) Bone - Bone: dyn_frontStrap_02_c"(ComponentPose=@n107, Alpha=1, Translation="0,-0.275,0", Rotation="0,0,0", Scale="1,1,1") {
    translationMode=BMM_Additive
    translationSpace=BCS_BoneSpace
    alphaBoolBlend {blendOption=Linear}
    boneToModify {boneName=dyn_frontStrap_02_c}
    translation {y=-0.275}
   }
   n109 = "Transform (Modify) Bone - Bone: dyn_frontStrap_03_c"(ComponentPose=@n108, Alpha=1, Translation="0,-0.5,0", Rotation="0,0,0", Scale="1,1,1") {
    translationMode=BMM_Additive
    translationSpace=BCS_BoneSpace
    alphaBoolBlend {blendOption=Linear}
    boneToModify {boneName=dyn_frontStrap_03_c}
    translation {y=-0.5}
   }
   n110 = "Transform (Modify) Bone - Bone: dyn_frontStrap_01_l"(ComponentPose=@n109, Alpha=1, Translation="0,-0.275,0", Rotation="0,0,0", Scale="1,1,1") {
    translationMode=BMM_Additive
    translationSpace=BCS_BoneSpace
    alphaBoolBlend {blendOption=Linear}
    boneToModify {boneName=dyn_frontStrap_01_l}
    translation {y=-0.275}
   }
   n111 = "Transform (Modify) Bone - Bone: dyn_frontStrap_02_l"(ComponentPose=@n110, Alpha=1, Translation="0,-0.275,0", Rotation="0,0,0", Scale="1,1,1") {
    translationMode=BMM_Additive
    translationSpace=BCS_BoneSpace
    alphaBoolBlend {blendOption=Linear}
    boneToModify {boneName=dyn_frontStrap_02_l}
    translation {y=-0.275}
   }
   n112 = "Transform (Modify) Bone - Bone: dyn_frontStrap_03_l"(ComponentPose=@n111, Alpha=1, Translation="0,0,0", Rotation="0,0,0", Scale="1,1,1") {
    translationMode=BMM_Additive
    translationSpace=BCS_BoneSpace
    alphaBoolBlend {blendOption=Linear}
    boneToModify {boneName=dyn_frontStrap_03_l}
   }
   n113 = "Transform (Modify) Bone - Bone: dyn_frontStrap_04_l"(ComponentPose=@n112, Alpha=1, Translation="0,-0.5,0", Rotation="0,0,0", Scale="1,1,1") {
    translationMode=BMM_Additive
    translationSpace=BCS_BoneSpace
    alphaBoolBlend {blendOption=Linear}
    boneToModify {boneName=dyn_frontStrap_04_l}
    translation {y=-0.5}
   }
   Output Pose(Result=Component To Local(ComponentPose=@n113)) {layerGroup=None}
   Save cached pose 'preSecondaryRigid'() {cacheName=preSecondaryRigid}
   Get isSkydiving() -> (isSkydiving:double)
  ]}
 EventGraph {nodes [
   Event BlueprintUpdateAnimation()
   n1 = AND Boolean(A=NOT Boolean(A=Get bIsJumping()), B=NOT Boolean(A=Get bIsFalling()))
   n2 = Get bIsSprinting()
   > Set isSprinting(isSprinting:double="To Float (Boolean)"(InBool=AND Boolean(A=@n2, B=@n1))) -> (Output_Get:double)
   Event OnBeginSkydiving()
   n3 = "To Float (Boolean)"(InBool=Get bIsSkyDiving())
   > Set isSkydiving(isSkydiving:double=@n3) -> (Output_Get:double)
   > Set isSkydivingINV(isSkydivingINV:double=float - float(B=@n3) {is=title K2Node_CallFunction}) -> (Output_Get:double)
   n4 = Event OnEndPlayingEmote() {is=title K2Node_Event}
   Delay(execute=[@n4.then, @n5.then, @n6.then, @n8.then], Duration=0.1)
   > ResetDynamics(InTeleportType=TeleportPhysics)
   > ResetDynamics(InTeleportType=ResetPhysics)
   Event OnBeginCrouching()
   > n5 = Set isCrouching(isCrouching:double="To Float (Boolean)"(InBool=Get bIsCrouching())) -> (Output_Get:double)
   Event OnEndCrouching()
   > n6 = Set isCrouching(isCrouching="To Float (Boolean)"(InBool=Get bIsCrouching()))
   Event OnEndSkydiving()
   n7 = "To Float (Boolean)"(InBool=Get bIsSkyDiving())
   > Set isSkydiving(isSkydiving=@n7)
   > n8 = Set isSkydivingINV(isSkydivingINV=float - float(B=@n7))
  ]}
]
variables [
 isCrouching {type=number}
 isSkydiving {type=number}
 isSkydivingINV {type=number, default=1}
 isSprinting {type=number}
]
